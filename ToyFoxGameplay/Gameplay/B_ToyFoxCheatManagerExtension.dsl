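; Cheat manager extension with debug commands for the ToyFox giant event: giant control, objectives, enemy, power level, rewards and spawning.
; Commands validate cheat data, and most forward to the ToyFox game state component.
; Become-giant commands gather eligible nearby teammates, filtered by gameplay tags, and activate the event with a giant above the player.
; Spawn commands for vehicle, lightning, supply drop and pedestal target camera-look hits; the pedestal spawns directly.
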(event Custom|SpawnAndPossessGiant
  (params (PrimaryPlayer :type Fort%20Player%20Pawn%20Object%20Reference) (SecondaryPlayers :type Array%20of%20Fort%20Player%20Pawn%20Object%20References))
  ;[0.15 0 0.1313 0.5] SpawnAndPossessGiant: Runs a cheat only logic to spawn and possess giant.
  (bind _component (Utilities|GetGameStateComponent)
    (:Valid
      (bind _returnvalue (Utilities|LoadClassAssetBlocking (Variables|Default|GetGiantPawnClass)))
      (bind _astoy_fox_giant_player_pawn (Utilities|Casting|CastToToyFoxGiantPlayerPawnClass _returnvalue))
      (Class|BToyFoxGameStateComponent|CheatActivateEventatLocationwithDirectGameplay _component _astoy_fox_giant_player_pawn (Utilities|Operators|Add (Transformation|GetActorLocation PrimaryPlayer) (0.0 0.0 2.5e+03)) (Math|Rotator|MakeRotator 0.0 0.0 (.yaw (Transformation|GetActorRotation PrimaryPlayer))) PrimaryPlayer SecondaryPlayers))))

(fn HandleEnemySpawned
  (params (EnemyActor :type Actor%20Object%20Reference))
  (bind _component (Utilities|GetGameStateComponent)
    (:Valid
      (Enemy|HandleEnemySpawned _component EnemyActor))))

(fn IsPawnValidForEvent
  (params (PlayerPawn :type Fort%20Player%20Pawn%20Object%20Reference))
  (Utilities|FlowControl|Sequence
    (:then_0
      (Default|ValidateCheatData
        (:Valid
          (Utilities|IsValid PlayerPawn
            (:Is%20Valid
              (if (not (GameplayTags|HasAnyMatchingGameplayTags PlayerPawn (GameplayTags|MakeLiteralGameplayTagContainer "(GameplayTags=((TagName=\"ToyFox.Player.Pilot\"),(TagName=\"ToyFox.Pawn\"),(TagName=\"ToyFox.Player.PilotTeleport\")))")))
                (return true)))))))
    (:then_1
      (return false))))

(fn ToyFoxActivateObjectiveAtLocation
  ;[0.02603 0.1312 0.15 0.5] Can be used to activate the objective at the closest event actor location.
  (bind _playercontroller (Default|ValidateCheatData)
    (:Valid
      (bind _asfort_player_controller_athena (Utilities|Casting|CastToFortPlayerControllerAthena _playercontroller))
      (bind _returnvalue (Game|Player|GetPlayerPawn _asfort_player_controller_athena))
      (bind _component (Utilities|GetGameStateComponent)
        (:Valid
          (Class|BToyFoxGameStateComponent|CheatActivateObjectiveatPlayer _component _returnvalue))))))

(fn ToyFoxBecomeGiant
  (Default|ValidateCheatData
    (:Valid
      (CallFunction|ToyFoxBecomeGiantwithNearbySquad 0.0))))

(fn ToyFoxBecomeGiantWithNearbySquad
  (params (NearbyDistance :type Float%20%28double-precision%29))
  (bind _playercontroller (Default|ValidateCheatData)
    (:Valid
      (bind _asfort_player_controller_athena (Utilities|Casting|CastToFortPlayerControllerAthena _playercontroller))
      (bind _output_get (LocalVariables|SetPrimaryPlayerPawn (Game|Player|GetPlayerPawn _asfort_player_controller_athena)))
      (Utilities|IsValid _output_get
        (:Is%20Valid
          (if (Helper|IsPawnValidforEvent (Game|Player|GetPlayerPawn _asfort_player_controller_athena))
            (bind _outfortplayerpawns (Fortnite|GetFortPlayerPawnsInRange (Transformation|GetActorLocation (Game|Player|GetPlayerPawn _asfort_player_controller_athena)) NearbyDistance true))
            (for _array_element _outfortplayerpawns
              (Utilities|IsValid _array_element
                (:Is%20Valid
                  (bind _returnvalue (Fortnite|Utilities|AttemptGetPlayerStateforPawn _array_element))
                  (bind _asfort_player_state_athena (Utilities|Casting|CastToFortPlayerStateAthena _returnvalue))
                  (if (Math|Boolean|ANDBoolean (Class|FortPlayerStateAthena|IsAnAthenaGameParticipant _asfort_player_state_athena) (!= _array_element (LocalVariables|GetPrimaryPlayerPawn)) (== (Team|GetActorTeam _array_element) (Team|GetActorTeam (Game|Player|GetPlayerPawn _asfort_player_controller_athena))) (Helper|IsPawnValidforEvent _array_element))
                    (Utilities|Array|Add (LocalVariables|GetSecondaryPlayerPawns) _array_element)))))
            (CallFunction|SpawnandPossessGiant (LocalVariables|GetPrimaryPlayerPawn) (LocalVariables|GetSecondaryPlayerPawns))))))))

(fn ToyFoxBecomeGiantWithSquad
  (CallFunction|ToyFoxBecomeGiantwithNearbySquad 1e+07))

(fn ToyFoxExitGiant
  (bind _playercontroller (Default|ValidateCheatData)
    (:Valid
      (bind _asfort_player_controller (Utilities|Casting|CastToFortPlayerController _playercontroller))
      (bind _returnvalue (Game|Player|GetPlayerPawn _asfort_player_controller))
      (if (GameplayTags|HasMatchingGameplayTag _returnvalue "(TagName=\"ToyFox.Player.Pilot\")")
        (bind _component (Utilities|GetGameStateComponent)
          (:Valid
            (Class|BToyFoxGameStateComponent|ExitToyFoxGiant _component (Utilities|Array|MakeArray _returnvalue))))))))

(fn ToyFoxExitGiantWithSquad
  (bind _playercontroller (Default|ValidateCheatData)
    (:Valid
      (bind _asfort_player_controller (Utilities|Casting|CastToFortPlayerController _playercontroller))
      (bind _returnvalue (Game|Player|GetPlayerPawn _asfort_player_controller))
      (if (GameplayTags|HasMatchingGameplayTag _returnvalue "(TagName=\"ToyFox.Player.Pilot\")")
        (bind _outfortplayerpawns (Fortnite|GetFortPlayerPawns))
        (for _array_element _outfortplayerpawns
          (Utilities|IsValid _array_element
            (:Is%20Valid
              (if (and (GameplayTags|HasMatchingGameplayTag _array_element "(TagName=\"ToyFox.Player.Pilot\")") (Team|OnSameTeam _array_element _returnvalue))
                (Utilities|Array|Add (LocalVariables|GetPlayerPawns) _array_element)))))
        (bind _component (Utilities|GetGameStateComponent)
          (:Valid
            (Class|BToyFoxGameStateComponent|ExitToyFoxGiant _component (LocalVariables|GetPlayerPawns))))))))

(fn ToyFoxPilotTeleport
  ;[0 0.15 0.09216 0.5] ToyFoxPilotTeleport: Simulates the pilot teleport sequence that occurs for the brief duration after the pedestal interaction.
  (bind _playercontroller (Default|ValidateCheatData)
    (:Valid
      (bind _asfort_player_controller (Utilities|Casting|CastToFortPlayerController _playercontroller))
      (bind _returnvalue (Utilities|LoadClassAssetBlocking (Variables|Default|GetPilotTeleportGEClass)))
      (bind _asgameplay_effect (Utilities|Casting|CastToGameplayEffectClass _returnvalue))
      (Fortnite|ApplyGameplayEffectsToActor (Game|Player|GetPlayerPawn _asfort_player_controller) (Utilities|Array|MakeArray _asgameplay_effect) (0.0 0.0 0.0) false 0 (Ability|GetAbilitySystemComponent (Game|Player|GetPlayerPawn _asfort_player_controller)) "()"))))

(fn ToyFoxRewardBoon
  (bind _playercontroller (Default|ValidateCheatData)
    (:Valid
      (bind _component (Utilities|GetGameStateComponent)
        (:Valid
          (Reward|RewardGiantDamageBoontoPlayer _component (Utilities|Casting|CastToFortPlayerPawn (Pawn|GetControlledPawn _playercontroller))))))))

(fn ToyFoxSetPowerLevel
  (params (PowerLevel :type Float%20%28double-precision%29))
  (Default|ValidateCheatData
    (:Valid
      (bind _component (Utilities|GetGameStateComponent)
        (:Valid
          (if (Utilities|Enum|Equal(Enum) (Event|GetEventState _component) "NewEnumerator3")
            (Class|BToyFoxGameStateComponent|CheatSetPowerLevel _component (Math|Float|Clamp(Float) PowerLevel))))))))

(fn ToyFoxShowEventLocations
  (params (bShow :type Boolean))
  ;[0.02294 0.0172 0.15 0.5] ToyFoxShowEventLocations: Enables the minimap icon on all the event locations in the world.
  (Default|ValidateCheatData
    (:Valid
      (bind _component (Utilities|GetGameStateComponent)
        (:Valid
          (Class|BToyFoxGameStateComponent|CheatShowEventLocations _component bShow))))))

(fn ToyFoxSimulateIncomingPlayerDamageOnPowerLevel
  (params (NumberOfPlayers :type Integer) (DamagePerSecond :type Float%20%28double-precision%29))
  ;[0.1276 0 0.15 0.5] ToyFoxSimulateIncomingPlayerDamageOnPowerLevel: Simulate the incoming DPS on the power level.
  (Default|ValidateCheatData
    (:Valid
      (bind _component (Utilities|GetGameStateComponent)
        (:Valid
          (Class|BToyFoxGameStateComponent|CheatSimulateIncomingPlayerDamageonPowerLevel _component (* NumberOfPlayers DamagePerSecond)))))))

(fn ToyFoxSpawnEnemy
  ;[0.1209 0.04913 0.15 0.5] ToyFoxSpawnEnemy: Spawns the enemy boss at the closest event actor location.
  (bind _playercontroller (Default|ValidateCheatData)
    (:Valid
      (bind _asfort_player_controller_athena (Utilities|Casting|CastToFortPlayerControllerAthena _playercontroller))
      (bind _component (Utilities|GetGameStateComponent)
        (:Valid
          (Class|BToyFoxGameStateComponent|CheatSpawnEnemy _component (Game|Player|GetPlayerPawn _asfort_player_controller_athena)))))))

(fn ToyFoxSpawnHoagie
  (bind _playercontroller (Default|ValidateCheatData)
    (:Valid
      (bind _asfort_player_controller (Utilities|Casting|CastToFortPlayerController _playercontroller))
      (bind _component (Utilities|GetGameStateComponent)
        (:Valid
          (bind (_outhitresult _returnvalue) (Helpers|GetCameraLookHitResult _asfort_player_controller))
          (if _returnvalue
            (Class|BToyFoxGameStateComponent|SpawnHoagieVehicleatLocation _component (.ImpactPoint (Collision|BreakHitResult _outhitresult)))))))))

(fn ToyFoxSpawnLightningBoltAtLocation
  (bind _playercontroller (Default|ValidateCheatData)
    (:Valid
      (bind _asfort_player_controller_athena (Utilities|Casting|CastToFortPlayerControllerAthena _playercontroller))
      (bind (_outhitresult _returnvalue) (Helpers|GetCameraLookHitResult _asfort_player_controller_athena 1e+05))
      (if _returnvalue
        (bind _component (Utilities|GetGameStateComponent)
          (:Valid
            (Class|BToyFoxGameStateComponent|SpawnLightningBoltatLocation _component (.ImpactPoint (Collision|BreakHitResult _outhitresult)))))))))

(fn ToyFoxSpawnLightningGroundRumbleAtLocation
  (bind _playercontroller (Default|ValidateCheatData)
    (:Valid
      (bind _asfort_player_controller_athena (Utilities|Casting|CastToFortPlayerControllerAthena _playercontroller))
      (bind (_outhitresult _returnvalue) (Helpers|GetCameraLookHitResult _asfort_player_controller_athena 1e+05))
      (if _returnvalue
        (bind _component (Utilities|GetGameStateComponent)
          (:Valid
            (Class|BToyFoxGameStateComponent|SpawnLightningGroundRumbeActoratLocation _component (.ImpactPoint (Collision|BreakHitResult _outhitresult)))))))))

(fn ToyFoxSpawnPedestal
  (bind _playercontroller (Default|ValidateCheatData)
    (:Valid
      (bind _asfort_player_controller_athena (Utilities|Casting|CastToFortPlayerControllerAthena _playercontroller))
      (bind (_outhitresult _returnvalue) (Helpers|GetCameraLookHitResult _asfort_player_controller_athena 1e+05))
      (if _returnvalue
        (bind _returnvalue_2 (Utilities|LoadClassAssetBlocking (Variables|Default|GetPedestalClass)))
        (bind _asbuilding_actor (Utilities|Casting|CastToBuildingActorClass _returnvalue_2))
        (bind _returnvalue_3 (Game|SpawnActorfromClass _asbuilding_actor (Math|Conversions|ToTransform(Vector) (.ImpactPoint (Collision|BreakHitResult _outhitresult))) "AlwaysSpawn" "MultiplyWithRoot" 0 0 0))
        (Building|InitializeKismetSpawnedBuildingActor _returnvalue_3 0 0 false 0 false)))))

(fn ToyFoxSpawnSupplyDrop
  (bind _playercontroller (Default|ValidateCheatData)
    (:Valid
      (bind _asfort_player_controller_athena (Utilities|Casting|CastToFortPlayerControllerAthena _playercontroller))
      (bind (_outhitresult _returnvalue) (Helpers|GetCameraLookHitResult _asfort_player_controller_athena))
      (if _returnvalue
        (bind _component (Utilities|GetGameStateComponent)
          (:Valid
            (Class|BToyFoxGameStateComponent|SpawnSupplyDropatLocation _component (.ImpactPoint (Collision|BreakHitResult _outhitresult)))))))))
