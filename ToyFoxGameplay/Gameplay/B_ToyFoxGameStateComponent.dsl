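; Game state component that runs the ToyFox giant-robot event, from objective activation through pilot possession, boss spawn and outro.
; Server-authoritative events gate on community quest completion, pick a location by EQS with nearest-actor fallback, then spawn giant and pedestal.
; Pilots play a cinematic before possessing the giant; timers drive power drain and EQS-placed supply drops and helicopters.
; Soft references load asynchronously, keeping this always-loaded asset light; hotfix scalable floats tune behavior.
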(event EventBeginPlay
  (bind _returnvalue (GameplayEventRouter|GetGlobalEventRouter))
  (bind _returnvalue_2 (Fortnite|Athena|GameState|GetGameStateAthena))
  ;[0.008987 0.15 0.04059 0.5] BeginPlay: Start debugging timer in PIE and bind to know when game phase changes so analytics can be fired at the right time.
  (|CallParentFunction)
  (Utilities|FlowControl|Sequence
    (:then_0
      ;[1 1 1 1 50] NOTE: Try not to use expensive hard references in this blueprint since this asset is always loaded being a part of the game state component.
      (if (and (Editor|InEditororPIE) (Variables|Debug|GetbDebug))
        (Utilities|Time|SetTimerbyEvent (EventDispatchers|CreateEvent :!event PrintDebugStringsToScreen) 0.1 true true)))
    (:then_1
      (Networking|SwitchHasAuthority
        (:Authority
          (bind _returnvalue_3 (Class|FortQuestBlueprintLibrary|IsMeshNetworkedEventComplete self (Variables|Settings|GetCommunityQuestTag)))
          (if _returnvalue_3
            (Variables|Transient|SetbEventRequiredCommunityQuestComplete true)
            (Fortnite|FortShippingLog self "BeginPlay: Community request completed at startup!")
            (else
              (Utilities|IsValid _returnvalue
                (:Is%20Valid
                  (bind (_asynctaskproxy _payload) (GameplayEventRouter|StartListeningToEvent _returnvalue "/Script/FortniteGame.MeshNetworkedEventCompleted")
                    (:then
                      (Variables|Transient|Async|SetCommunityQuestCompletedAsyncTask _asynctaskproxy)
                      (Fortnite|FortShippingLog self "BeginPlay: Listening for community quest completion."))
                    (:OnEventReceived
                      (if (== (Utilities|Struct|BreakMeshNetworkedEventCompleted _payload) (Variables|Settings|GetCommunityWorldStateTag))
                        (Variables|Transient|SetbEventRequiredCommunityQuestComplete true)
                        (Fortnite|FortShippingLog self "BeginPlay: Community quest complete from delegate!")))))
                (:Is%20Not%20Valid
                  ;;[1 0 0.03209] Warning: GlobalEventRouter is null.
                  (Fortnite|FortShippingLog self "BeginPlay: GlobalEventRouter is null." true))))))))
    (:then_2
      (Networking|SwitchHasAuthority
        (:Authority
          (Class|FortAsyncActionCurrentPlaylistReady|PlaylistDataReadyAsync
            (:OnReady
              (Utilities|FlowControl|Sequence
                (:then_0
                  (Utilities|FlowControl|Sequence
                    (:then_0
                      (if (or (Editor|InEditororPIE) (ScalableFloat|GetValueasBool (Variables|Hotfix|GetHotfix_EventBypassCommunityQuestCheck)))
                        (Variables|Transient|SetbEventRequiredCommunityQuestComplete true)
                        (Fortnite|FortShippingLog self "BeginPlay: Bypassing community quest check.")))
                    (:then_1
                      (Utilities|IsValid _returnvalue_2
                        (:Is%20Valid
                          (bind _returnvalue_4 (GameplayEventRouter|GetEventRouter "/Script/FortniteGame.FortGameState" _returnvalue_2))
                          (Utilities|IsValid _returnvalue_4
                            (:Is%20Valid
                              (bind (_asynctaskproxy_2 _payload_2) (GameplayEventRouter|StartListeningToStatefulEvent _returnvalue_4 "/Script/FortniteGame.FortGamePhaseUpdatedEvent")
                                (:then
                                  (Variables|Transient|SetGamePhaseUpdatedAsyncTask _asynctaskproxy_2))
                                (:OnEventReceived
                                  _)
                                (:OnSavedEventStateExists
                                  _))
                              (Analytics|HandleGamePhaseChanged (Utilities|Struct|BreakFortGamePhaseUpdatedEvent _payload_2)))
                            (:Is%20Not%20Valid
                              ;;[1 0 0.03209] Warning: EventRouter is null.
                              (Fortnite|FortShippingLog self "BeginPlay: EventRouter is null." true))))))))
                (:then_1
                  (Variables|Transient|SetbIntermittentPawnAttach (> (ScalableFloat|GetCachedScalableFloatValue (Variables|Hotfix|GetHotfix_AttachPawnIntermittent_Enabled)) 0)))))))))))

(event EventEndPlay
  (params (EndPlayReason :type EEndPlayReason%20Enum :default "Destroyed"))
  ;[0.1302 0.09219 0.02218] EndPlay: Clean up binds and fire analytics logic if playing in editor.
  (Utilities|FlowControl|Sequence
    (:then_0
      (Utilities|FlowControl|Sequence
        (:then_0
          (Networking|SwitchHasAuthority
            (:Authority
              (if (Editor|InEditororPIE)
                (Analytics|SubmitAnalytics)
                (else
                  _))
              (bind _gamephaseupdatedasynctask (Variables|Transient|GetGamePhaseUpdatedAsyncTask)
                (:then
                  (GameplayEventRouter|StopListeningtoStatefulEvent _gamephaseupdatedasynctask))
                (:else
                  _))
              (bind _communityquestcompletedasynctask (Variables|Transient|Async|GetCommunityQuestCompletedAsyncTask)
                (:then
                  (GameplayEventRouter|StopListeningtoEvent _communityquestcompletedasynctask))
                (:else)))))
        (:then_1
          (Ability|Async|EndAction (Variables|Transient|Enemy|GetAsyncActions_EnemyStunned))
          (bind _enemy (Variables|Transient|Enemy|GetEnemy)
            (:then
              (EventDispatchers|UnbindEventfromOnAbilitySystemComponentInitialized _enemy (EventDispatchers|CreateEvent :!event Custom|HandleEnemyAbilitySystemComponentInitialized)))
            (:else)))))
    (:then_1
      (|CallParentFunction EndPlayReason))))

(event Custom|ActivateToyFoxObjective
  ;[0.00195 0.03125 0.001983 0.5 35] Objective activation
  ;;[0.009496 0.15 0.01522 0.5] ActivateToyFoxObjective: Activate the toyfox objective if it can be activated. This should run an EQS to figure out a random best place to start the event.
  (Networking|SwitchHasAuthority
    (:Authority
      (if (Objective|CanActivateObjective)
        (Fortnite|FortShippingLog self "ActivateToyFoxObjective: Invoked.")
        (Variables|Event|GetEventLocationEvnQueryInstance
          (:then)
          (:else
            (bind _object (Utilities|AsyncLoadAsset (Variables|Settings|GetEventLocationEnvQuery))
              (:Completed
                (bind _returnvalue (AI|EQS|RunEQSQuery (Utilities|Casting|CastToEnvQuery _object) self))
                (bind _output_get (Variables|Event|SetEventLocationEvnQueryInstance _returnvalue))
                (Utilities|IsValid _output_get
                  (:Is%20Valid
                    (AI|EQS|BindEventtoOnQueryFinished _output_get (EventDispatchers|CreateEvent :!event Custom|HandleEventLocationEQSQueryFinished))
                    (Fortnite|FortShippingLog self "ActivateToyFoxObjective: EnvQuery for EventLocation started.")))))))
        (else
          (Fortnite|FortShippingLog self (Utilities|String|Append "ActivateToyFoxObjective, Check failed! CommunityQuestBool: " (Utilities|String|ToString(Boolean) (Variables|Transient|GetbEventRequiredCommunityQuestComplete)))))))))

(event Custom|HandleEventLocationEQSQueryFinished
  (params (QueryInstance :type EQS%20Query%20Instance%20Object%20Reference) (QueryStatus :type EEnvQueryStatus%20Enum :default "Processing"))
  (bind _returnvalue (SafeZone|GetSafeZoneIndicator))
  ;[0.00195 0.03125 0.001983 0.5 35] Objective activation
  ;;[0.1313 0.02465 0.15 0.5] HandleEventLocationEQSQueryFinished: Activate the objective at the found EQS event location.
  (Networking|SwitchHasAuthority
    (:Authority
      (Utilities|IsValid QueryInstance
        (:Is%20Valid
          (Variables|Event|SetEventLocationEvnQueryInstance 0)
          (Fortnite|FortShippingLog self "HandleEventLocationEQSQueryFinished: Invoked.")
          (bind _resultactors (AI|EQS|GetQueryResultsasActors QueryInstance))
          (if (Utilities|Array|IsValidIndex _resultactors)
            (bind _asb_toy_fox_event_location (Utilities|Casting|CastToB_ToyFox_EventLocation (Utilities|Array|Get(acopy) _resultactors)))
            (CallFunction|ActivateToyFoxObjectiveatEventActor _asb_toy_fox_event_location)
            (else
              _)))
        (:Is%20Not%20Valid
          _))
      (Fortnite|FortShippingLog self "HandleEventLocationEQSQueryFinished: Failed to find location. Falling back to random location." true)
      ;;;[1 1 1] If Eqs fails, do a simple direct search.
      (Utilities|IsValid _returnvalue
        (:Is%20Valid
          (bind _outactors (Actor|GetAllActorsOfClass "/ToyFoxGameplay/Gameplay/Locations/B_ToyFox_EventLocation.B_ToyFox_EventLocation_C"))
          (bind _asb_toy_fox_event_location_2 (Utilities|Casting|CastToB_ToyFox_EventLocation (.ReturnValue (Actor|FindNearestActor (Transformation|GetActorLocation _returnvalue) _outactors))))
          (CallFunction|ActivateToyFoxObjectiveatEventActor _asb_toy_fox_event_location_2))))))

(event Custom|ActivateToyFoxObjectiveAtEventActor
  (params (EventLocationActor :type B%20Toy%20Fox%20Event%20Location%20Object%20Reference))
  (bind _name (Class|BToyFoxEventLocation|GetEventLocationName EventLocationActor))
  ;[0.00195 0.03125 0.001983 0.5 35] Objective activation
  ;;[0.009628 0.02084 0.15 0.5] ActivateToyFoxObjectiveAtEventLocation: Activates the objective at the location determined by the input event location actor.
  (Networking|SwitchHasAuthority
    (:Authority
      (Utilities|IsValid EventLocationActor
        (:Is%20Valid
          (Variables|Event|SetEventLocationActor EventLocationActor)
          (Fortnite|FortShippingLog self (Utilities|String|Append "ActivateToyFoxObjectiveAtEventActor: Invoked at location " _name))
          (bind _bvalid (Class|BToyFoxEventLocation|GetPedestalLocation EventLocationActor))
          (if _bvalid
            (bind (_bvalid_2 _location _rotation) (Class|BToyFoxEventLocation|GetPlayerLocationandRotation EventLocationActor))
            (if _bvalid_2
              (CallFunction|ActivateToyFoxObjectiveatLocation _location _rotation)
              (else
                (Fortnite|FortShippingLog self (Utilities|String|Append "ActivateToyFoxObjectiveAtEventActor: Player Location is not valid: " (Class|BToyFoxEventLocation|GetEventLocationName EventLocationActor)) true)))
            (else
              (Fortnite|FortShippingLog self (Utilities|String|Append "ActivateToyFoxObjectiveAtEventActor: Pedestal Location is not valid: " _name) true))))))))

(event Custom|ActivateToyFoxObjectiveAtLocation
  (params (Location :type Vector) (Rotation :type Rotator))
  (bind _toyfoxgiantpawn (Variables|Transient|GetToyFoxGiantPawn))
  ;[0.00195 0.03125 0.001983 0.5 35] Objective activation
  ;;[0.06469 0.1447 0.15 0.5] ActivateToyFoxObjectiveAtLocation: Start the objective at the specified location.
  (Networking|SwitchHasAuthority
    (:Authority
      (if (Objective|CanActivateObjective)
        (Variables|Transient|SetbObjectiveActive true)
        (Utilities|FlowControl|Sequence
          (:then_0
            ;;;[0.15 0.06224 0.1161 0.5] Async load classes that will be used soon.
            (bind _class (Utilities|AsyncLoadClassAsset (Variables|Settings|GetPrimaryPilotGameplayEffectClass))
              (:then
                (bind _class_2 (Utilities|AsyncLoadClassAsset (Variables|Settings|GetSecondaryPilotGameplayEffectClass))
                  (:then
                    (bind _class_3 (Utilities|AsyncLoadClassAsset (Variables|Settings|GetPilotStasisGameplayEffect))
                      (:then
                        (bind _class_4 (Utilities|AsyncLoadClassAsset (Variables|Settings|GetGiantOutroGameplayEffectClass))
                          (:then
                            (bind _class_5 (Utilities|AsyncLoadClassAsset (Variables|Settings|GetPilotTeleportGameplayEffectClass))
                              (:then
                                (bind _class_6 (Utilities|AsyncLoadClassAsset (Variables|Settings|GetStandInPilotClass))
                                  (:then
                                    (bind _class_7 (Utilities|AsyncLoadClassAsset (Variables|SupplyDrop|GetSupplyDropSoftClass))
                                      (:then
                                        (bind _object (Utilities|AsyncLoadAsset (Variables|Settings|GetIntroAbilitySet))
                                          (:then
                                            (bind _class_8 (Utilities|AsyncLoadClassAsset (Variables|Settings|GetSecondaryPilotFinisherGameplayEffectClass))
                                              (:Completed
                                                (Utilities|CacheAsyncLoadedClass _class_8))))
                                          (:Completed
                                            (Utilities|CacheAsyncLoadedObject _object))))
                                      (:Completed
                                        (Utilities|CacheAsyncLoadedClass _class_7))))
                                  (:Completed
                                    (Utilities|CacheAsyncLoadedClass _class_6))))
                              (:Completed
                                (Utilities|CacheAsyncLoadedClass _class_5))))
                          (:Completed
                            (Utilities|CacheAsyncLoadedClass _class_4))))
                      (:Completed
                        (Utilities|CacheAsyncLoadedClass _class_3))))
                  (:Completed
                    (Utilities|CacheAsyncLoadedClass _class_2))))
              (:Completed
                (Utilities|CacheAsyncLoadedClass _class))))
          (:then_1
            ;;;[0.15 0.05428 0.1465 0.5] Spawn the event location mark actor.
            (bind _returnvalue (Game|SpawnActorfromClass (Variables|Event|GetEventLocationMarkActorClass) (Event|GetEventLocationMarkActorSpawnTransform) "AlwaysSpawn" "MultiplyWithRoot" 0 0 0))
            (Variables|Transient|SetEventLocationMarkActor _returnvalue)
            (Utilities|FlowControl|Delay 1.0) ; Brief duration to allow \nthe location mark actor to replicate.
            (Event|SetEventState "NewEnumerator5")
            ;;;[1 1 1] Spawn the analytics volume actor to track how much players are fighting around the event point.
            (bind _class_9 (Utilities|AsyncLoadClassAsset (Variables|Analytics|GetEventLocationCombatAnalyticsVolumeActorClass))
              (:then
                (bind (_bvalid _location) (Class|BToyFoxEventLocation|GetEnemyLocationandRotation (Variables|Event|GetEventLocationActor)))
                (CallFunction|SpawnLightningBoltatLocation _location)
                (CallFunction|PlayVOAtLocation "(TagName=\"ToyFox.VO.Rita.Laugh\")" (+ _location "0.000000,0.000000,10000.000000"))
                (CallFunction|SpawnLightningGroundRumbeActoratLocation _location)
                (Utilities|FlowControl|Sequence
                  (:then_0
                    (Utilities|FlowControl|Delay (ScalableFloat|GetCachedScalableFloatValue (Variables|Hotfix|GetHotfix_DelayBetweenEventIntroAndGiantIntro))) ; Delay between the danger callout (event intro) and giant intro. 
                    ;;;[1 1 1] Spawn the giant pawn.
                    (Utilities|SpawnToyFoxGiantPawn Location Rotation
                      (:TaskComplete
                        (Variables|Transient|GetToyFoxGiantPawn
                          (:then
                            (Fortnite|FortShippingLog self "ActivateToyFoxObjectiveAtLocation: Giant pawn spawn successful..")
                            (Event|SetEventState "NewEnumerator1") ; Set GiantIntro only after Giant is spawned.
                            ;;;[0.02788 0.08145 0.15 0.5] Provide the pawn with the intro ability set.
                            (bind _returnvalue_2 (Utilities|LoadAssetBlocking (Variables|Settings|GetIntroAbilitySet))) ; This will already be async loaded at this point
                            (bind _returnvalue_3 (Ability|EquipFortAbilitySet _toyfoxgiantpawn (Utilities|Casting|CastToFortAbilitySet _returnvalue_2)))
                            (Variables|Transient|SetIntroAbilitySetHandle _returnvalue_3)
                            (Ability|SendGameplayEventtoActor _toyfoxgiantpawn "(TagName=\"ToyFox.Event.Intro.Invoked\")") ; Fire the intro gameplay ability.
                            ;;;[1 1 1] Load the pilot cinematic level sequence so that it is ready when needed.
                            (bind _object_2 (Utilities|AsyncLoadAsset (Variables|PilotCinematic|GetPilotCinematicLevelSequenceSoftRef))
                              (:Completed
                                (bind _aslevel_sequence (Utilities|Casting|CastToLevelSequence _object_2))
                                (Variables|Transient|PilotCinematic|SetPilotCinematicLevelSequence _aslevel_sequence)
                                (Fortnite|FortShippingLog self "ActivateToyFoxObjectiveAtLocation: Pilot Cinematic Level Sequence async loaded."))))
                          (:else
                            (Fortnite|FortShippingLog self "ActivateToyFoxObjectiveAtLocation: Failed to spawn giant pawn, event failed." true))))) ; Async event.
                  )
                  (:then_1
                    ;;;[0 0.01262 0.15 0.5] Broadcast alert (HUD notification) that the queen has been detected.
                    (Utilities|FlowControl|Delay (ScalableFloat|GetCachedScalableFloatValue (Variables|Hotfix|GetHotfix_AlertDelayBeforeNotificationQueenDetected)))
                    (CallFunction|BroadcastAlertNotification 0 (struct :AlertGameplayTag_12_6C55428345DE14476CEF4D9FECB3F891 "(TagName=\"ToyFox.HUD.Alert.Queen.Detected\")"))
                    ;;;[0 0.03551 0.15 0.5] Ping and enable the map icon for the event location actor.
                    (bind _eventlocationmarkactor (Variables|Transient|GetEventLocationMarkActor)
                      (:then
                        (Class|BGAEventLocationMarkActor|SetMapIconEnable _eventlocationmarkactor true)
                        (Class|BGAEventLocationMarkActor|MarkActorforPlayers _eventlocationmarkactor))
                      (:else)))))
              (:Completed
                (bind (_bvalid_2 _location_2) (Class|BToyFoxEventLocation|GetPedestalLocation (Variables|Event|GetEventLocationActor)))
                (bind _returnvalue_4 (Game|SpawnActorfromClass (Utilities|Casting|CastToActorClass _class_9) (Math|Conversions|ToTransform(Vector) _location_2) "AlwaysSpawn" "MultiplyWithRoot" 0 0))
                (Variables|Analytics|Transient|SetEventLocationCombatAnalyticsEventActor _returnvalue_4)))))))))

(event Custom|SpawnToyFoxPedestalAtEventLocation
  ;[0.08652 0.04659 0.15 0.5] SpawnToyFoxPedestalAtEventLocation: This is invoked from GA_ToyFox_Intro, once giant pawn has landed.
  (Networking|SwitchHasAuthority
    (:Authority
      (Fortnite|FortShippingLog self "SpawnToyFoxPedestalAtEventLocation: Invoked.")
      (bind _eventlocationactor (Variables|Event|GetEventLocationActor)
        (:then
          (bind (_bvalid _location) (Class|BToyFoxEventLocation|GetPedestalLocation _eventlocationactor))
          (if _bvalid
            (bind _class (Utilities|AsyncLoadClassAsset (Variables|Settings|GetPedestalClassSoft))
              (:then
                (Event|SetEventState "NewEnumerator2"))
              (:Completed
                (bind _asbuilding_gameplay_actor (Utilities|Casting|CastToBuildingGameplayActorClass _class))
                (bind _returnvalue (Game|SpawnActorfromClass _asbuilding_gameplay_actor (Math|Conversions|ToTransform(Vector) _location) "AlwaysSpawn" "MultiplyWithRoot" 0 0 0))
                (Utilities|IsValid _returnvalue
                  (:Is%20Valid
                    (bind _output_get (Variables|Transient|SetTempPlayerControllers (Fortnite|GetAllFortPlayerControllers true)))
                    (for _array_element _output_get
                      (MarkerComponent|MarkActoronServer (Actor|GetComponentbyClass _array_element "/Script/FortniteGame.AthenaMarkerComponent") _returnvalue false false))))))
            (else
              (Fortnite|FortShippingLog self "SpawnToyFoxPedestalAtEventLocation: Pedestal location is invalid." true))))
        (:else
          (Fortnite|FortShippingLog self "SpawnToyFoxPedestalAtEventLocation: EventLocationActor is invalid." true))))))

(event Custom|HandleGiantPawnPossessed
  (params (Pawn :type Pawn%20Object%20Reference))
  ;[0 0.05552 0.05729 0.5 35] GiantPawnPossession
  ;;[0.15 0.003661 0.1194 0.5] HandleGiantPawnPossessed: Once the giant pawn is possessed, remove the intro ability set.
  (Utilities|SwitchHasAuthority
    (:Authority
      (bind _toyfoxgiantpawn (Variables|Transient|GetToyFoxGiantPawn)
        (:then
          (Utilities|IsValid (Pawn|GetController _toyfoxgiantpawn)
            (:Is%20Valid
              (Fortnite|FortShippingLog self (Utilities|String|Append "HandleGiantPawnPossessed: Giant pawn possessed by controller: " (Utilities|GetDisplayName (Pawn|GetController _toyfoxgiantpawn)) " ; IsPlayerController? " (Utilities|String|ToString(Boolean) (Pawn|IsPlayerController (Pawn|GetController _toyfoxgiantpawn)))))
              (if (Pawn|IsPlayerController (Pawn|GetController _toyfoxgiantpawn))
                (Ability|UnequipFortAbilitySet (Variables|Transient|GetIntroAbilitySetHandle))
                (Game|Player|UnbindEventfromOnPawnPossessed _toyfoxgiantpawn (delegate Custom|HandleGiantPawnPossessed))
                (bind _output_get (Variables|Transient|SetGiantPossessingPlayerState (Utilities|Casting|CastToFortPlayerStateAthena (Variables|/Script/Engine.Pawn|GetPlayerState _toyfoxgiantpawn))))
                (Fortnite|FortShippingLog self (Utilities|String|Append "HandleGiantPawnPossessed: GiantPossessingPlayerState is " (Utilities|GetDisplayName _output_get)) (not (.bSuccess (Utilities|Casting|CastToFortPlayerStateAthena (Variables|/Script/Engine.Pawn|GetPlayerState _toyfoxgiantpawn)))))
                (Event|SetEventState "NewEnumerator3")
                ;;[0.01232 0.09545 0.15 0.5] Start timer to passively reduce power level.
                (Utilities|Time|SetTimerbyEvent (EventDispatchers|CreateEvent :!event ReducePowerLevelPassive) 1.0 true)))))
        (:else)))))

(event Custom|BeginToyFoxEnterSequence
  (params (PrimaryPlayer :type Fort%20Player%20Pawn%20Object%20Reference) (SecondaryPlayers :type Array%20of%20Fort%20Player%20Pawn%20Object%20References) (bSkipCinematic :type Boolean))
  ;[0.0625 0.002924 0.01081 0.5 35] Intro sequences and cinematics.
  ;;[0.02663 0.15 0.02841 0.5] BeginToyFoxEnterSequence: Handles transitioning the input players from their BR gameplay state to controlling ToyFox. This includes playing the pilot cinematic sequence.
  ;; Invoked from B_ToyFoxPedestal after successful interaction.
  (Networking|SwitchHasAuthority
    (:Authority
      (Utilities|FlowControl|Sequence
        (:then_0
          (Variables|Transient|Pilots|SetPrimaryPilotPlayerPawn PrimaryPlayer)
          (Variables|Transient|Pilots|SetSecondaryPilotPlayerPawns SecondaryPlayers)
          (Analytics|CachePilotAnalyticsInfo)
          (Pilots|SetupStandinPilots)
          (Pilots|CreatePilotActorsArray)
          (Cinematics|PlayPilotTeleport)
          (Cinematics|SetEnableCinematicHideHUDComponents true)
          (Pilots|CreateControllerComponents)
          (Utilities|FlowControl|Delay (Variables|Settings|GetPilotTeleportSequenceDuration))
          (CallFunction|BroadcastAlertNotification 0 (struct :AlertGameplayTag_12_6C55428345DE14476CEF4D9FECB3F891 "(TagName=\"ToyFox.HUD.Alert.Megazord.Claimed\")")) ; Needs to be called after a slight delay \nso that the clients have the pilot information replicated to them beforehand.
          (Pilots|ApplyPilotStatus)
          (if (not bSkipCinematic)
            (Variables|PilotCinematic|Spectate|SetbEnableSpectateTeleportCamera true)
            ;;;[1 1 1] Spawn outside gameplay space otherwise it is also replicating to nearby players.
            (bind _returnvalue (Game|SpawnActorfromClass "/ToyFoxGameplay/Gameplay/PilotCinematic/ToyFoxPilotCinematicSequenceActor.ToyFoxPilotCinematicSequenceActor_C" (struct :Location "0.000000,0.000000,50000.000000" :Rotation "0, 0, 0" :Scale "1.000000,1.000000,1.000000") "AlwaysSpawn" "MultiplyWithRoot" 0 (Variables|PilotCinematic|GetPilotCinematicPlaybackSettings) (Variables|Transient|PilotCinematic|GetPilotCinematicLevelSequence) (Variables|PilotCinematic|GetPilotCinematicCameraSettings) true 0))
            (Utilities|IsValid _returnvalue
              (:Is%20Valid
                (bind _output_get (Variables|Transient|PilotCinematic|SetPilotCinematicSequenceActor _returnvalue))
                (Class|ToyFoxPilotCinematicSequenceActor|AddPilotPlayerStates _output_get (Pilots|GetPilotPlayerStates))
                (bind _sequenceplayer (Variables|/Script/LevelSequence.LevelSequenceActor|GetSequencePlayer _output_get)
                  (:then
                    (Sequencer|Player|BindEventtoOnStop _sequenceplayer (EventDispatchers|CreateEvent :!event Custom|HandleCinematicSequenceComplete))
                    (Sequencer|Player|SetPlaybackPosition _sequenceplayer (Utilities|Struct|MakeMovieSceneSequencePlaybackParams (struct :FrameNumber (struct :Value 60) :SubFrame 0.0) 0.0 "" "(Hours=0,Minutes=0,Seconds=0,Frames=0,Subframe=0.000000,bDropFrameFormat=False)" "Frame" "Play" false))
                    (Sequencer|Player|Play _sequenceplayer :!class /Script/MovieScene.MovieSceneSequencePlayer)
                    (Cinematics|SetEnableCinematicCameraHandlerComponents true))
                  (:else)))
              (:Is%20Not%20Valid
                _))
            (else
              _))
          ;;;[0.15 0.05692 0.01203 0.5] When the intro sequence is skipped, as when using a cheat, player possession can be swapped before certain animations are finished playing, resulting in a bug related to weapon holstering (FORT-197826).
          ;;;
          ;;; This unfortunate delay helps prevent that race condition. -Andrew G, SkyMango event team
          (Utilities|FlowControl|Delay (ScalableFloat|GetValueatLevel (Variables|Hotfix|GetHotfix_IntroSequenceSkipDelay)))
          (CallFunction|HandleCinematicSequenceComplete))
        (:then_1
          ;;;[0.1163 0 0.15 0.5] Post possession VO and FX sequence, primarily for ground players.
          ;;; Delays are very intentional with timings for feel as well as cinematic alignment.
          ;;;[1 1 1] Do if we're spawning the boss
          (if (ScalableFloat|GetValueasBool (Variables|Hotfix|GetHotfix_EventBossShouldSpawn))
            (Utilities|FlowControl|Delay 2.75)
            (bind (_bvalid _location) (Class|BToyFoxEventLocation|GetEnemyLocationandRotation (Variables|Event|GetEventLocationActor)))
            (CallFunction|PlayVOAtLocation "(TagName=\"ToyFox.VO.Rita.PreEnemySpawn\")" (+ _location "0.000000,0.000000,10000.000000"))
            (Utilities|FlowControl|Delay 1.0)
            (CallFunction|SpawnLightningBoltatLocation _location)
            (CallFunction|SpawnLightningGroundRumbeActoratLocation _location)
            (Utilities|FlowControl|Delay (ScalableFloat|GetCachedScalableFloatValue (Variables|Hotfix|GetHotfix_MakeMyMonster_VODelay)))
            (CallFunction|PlayVOAtLocation "(TagName=\"ToyFox.VO.Rita.MakeMyMonster\")" (+ _location "0.000000,0.000000,10000.000000") true)
            (Utilities|FlowControl|Delay (ScalableFloat|GetCachedScalableFloatValue (Variables|Hotfix|GetHotfix_MakeMyMonster_LightningDelay)))
            (CallFunction|SpawnLightningBoltatLocation _location)))
        (:then_2
          ;;[1 1 1] Manually attach players to giant near end of cosmetics to avoid camera issues
          (Utilities|FlowControl|Delay 1e+01)
          (Pilots|AttachPlayerstoGiant))))))

(event Custom|HandleCinematicSequenceComplete
  ;[0.0625 0.002924 0.01081 0.5 35] Intro sequences and cinematics.
  ;;[0.03333 0.1151 0.15 0.5] HandleCinematicSequenceComplete: Once the cinematic sequence is complete, have the pilot players possess the giant pawn.
  (Networking|SwitchHasAuthority
    (:Authority
      (bind _pilotcinematicsequenceactor (Variables|Transient|PilotCinematic|GetPilotCinematicSequenceActor)
        (:then
          (bind _sequenceplayer (Variables|/Script/LevelSequence.LevelSequenceActor|GetSequencePlayer _pilotcinematicsequenceactor)
            (:then
              (Sequencer|Player|Stop _sequenceplayer :!class /Script/MovieScene.MovieSceneSequencePlayer)
              (Actor|SetLifeSpan _pilotcinematicsequenceactor 1.0))
            (:else
              _))
          (Pilots|PossessGiantandEnablePilotControls)
          (Cinematics|SetEnableCinematicCameraHandlerComponents false)
          (Cinematics|SetEnableCinematicHideHUDComponents false)
          (Variables|PilotCinematic|Spectate|SetbEnableSpectateTeleportCamera false)
          (Utilities|IsValid (Variables|Transient|Pilots|GetPrimaryPilotPlayerPawn)
            (:Is%20Valid
              (if (ScalableFloat|GetValueasBool (Variables|Hotfix|GetHotfix_EventBossShouldSpawn))
                (Giant|SpawnEnemyBoss)))
            (:Is%20Not%20Valid
              (Variables|PowerLevel|SetPowerLevel 0.0) ; no valid primary pilot, ending the event immediately now
            )))
        (:else
          _)))))

(event Custom|HandleOutroFinished
  ;[1 1 1] HandleOutroFinished: event ends officially now that outro anim finished
  (Utilities|FlowControl|Sequence
    (:then_0
      (if (Utilities|Enum|NotEqual(Enum) (Variables|Transient|GetEventState) "NewEnumerator2")
        (Utilities|FlowControl|Sequence
          (:then_0
            ;;[1 1 1] When primary pilot is invalid, it's already disconnected and we only need to destroy the giant
            (Utilities|IsValid (Variables|Transient|Pilots|GetPrimaryPilotPlayerPawn)
              (:Is%20Not%20Valid
                (Class|ToyFoxGiantPlayerPawn|DestroyDelayed (Variables|Transient|GetToyFoxGiantPawn) 0.1))))
          (:then_1
            (CallFunction|ExitAllToyFoxPilots)))
        (else
          (Class|ToyFoxGiantPlayerPawn|DestroyDelayed (Variables|Transient|GetToyFoxGiantPawn) 0.1))))
    (:then_1
      (Default|CallOnGiantEnded)
      (if (ScalableFloat|GetValueasBool (Variables|Hotfix|GetHotfix_EventBypassCheck))
        (CallFunction|ResetObjective)))))

(event Custom|ExitToyFoxGiant
  (params (PlayerPawns :type Array%20of%20Fort%20Player%20Pawn%20Object%20References))
  ;[0 0.03491 0.15 0.5] ExitToyFoxGiant: Exits the given player pawns from the giant.
  ; Logic handles both primary and secondary pilots.
  (Networking|SwitchHasAuthority
    (:Authority
      (for _array_element PlayerPawns
        (Pilots|HandlePlayerExit _array_element)))))

(event Custom|ExitAllToyFoxPilots
  ;[0.01147 0.15 0.1204 0.5] ExitAllToyFoxPilots: exits all primary and secondary pilot from the giant
  (Networking|SwitchHasAuthority
    (:Authority
      (Pilots|HandlePlayerExit (Variables|Transient|GetToyFoxGiantPawn))
      (CallFunction|ExitToyFoxGiant (Variables|Transient|Pilots|GetSecondaryPilotPlayerPawns)))))

(event Custom|ExitSecondaryPilots
  ;[0.07804 0.01367 0.15 0.5] ExitSecondaryPilots: Exits the secondary pilots from their cockpit.
  (Networking|SwitchHasAuthority
    (:Authority
      (CallFunction|ExitToyFoxGiant (Variables|Transient|Pilots|GetSecondaryPilotPlayerPawns)))))

(event Custom|SetEnableSpawnSupplyDrops
  (params (bEnable :type Boolean) (NumToInstantSpawn :type Integer))
  ;[0.131 0.03015 0.15 0.5] Supply Drops.
  ;;[0.03576 0.05956 0.15 0.5] Enable or disable the supply drop spawn timer.
  (Networking|SwitchHasAuthority
    (:Authority
      (if bEnable
        (for _ (range (+ (- NumToInstantSpawn 1) 1))
          (CallFunction|RunEqstoSpawnSupplyDrop))
        (bind _returnvalue (Utilities|Time|SetTimerbyEvent (EventDispatchers|CreateEvent :!event Custom|RunEqsToSpawnSupplyDrop) (ScalableFloat|GetCachedScalableFloatValue (Variables|Hotfix|GetHotfix_SupplyDropDelayBetweenSpawns)) true))
        (Variables|Transient|SupplyDrop|SetSupplyDropSpawnTimerHandle _returnvalue)
        (else
          (Utilities|Time|ClearandInvalidateTimerbyHandle (Variables|Transient|SupplyDrop|GetSupplyDropSpawnTimerHandle)))))))

(event Custom|RunEqsToSpawnSupplyDrop
  (bind _pawn (Giant|GetGiantPawn))
  ;[0.131 0.03015 0.15 0.5] Supply Drops.
  ;;[0.03372 0.07767 0.15 0.5] Run the EQS to find a spawn location for the supply drop.
  (Networking|SwitchHasAuthority
    (:Authority
      (Utilities|IsValid _pawn
        (:Is%20Valid
          (bind _returnvalue (AI|EQS|RunEQSQuery (Variables|SupplyDrop|GetEnvQuerySupplyDropSpawn) _pawn "RandomBest25Pct"))
          (Utilities|IsValid _returnvalue
            (:Is%20Valid
              (AI|EQS|BindEventtoOnQueryFinished _returnvalue (EventDispatchers|CreateEvent :!event HandleSupplyDropSpawnEQSQueryComplete)))))
        (:Is%20Not%20Valid
          (CallFunction|SetEnableSpawnSupplyDrops false))))))

(event Custom|HandleSupplyDropSpawnEqsQueryComplete
  (params (QueryInstance :type EQS%20Query%20Instance%20Object%20Reference) (QueryStatus :type EEnvQueryStatus%20Enum :default "Processing"))
  ;[0.131 0.03015 0.15 0.5] Supply Drops.
  ;;[0.02934 0.15 0.08064 0.5] Spawn the supply drop at the location found by the EQS.
  (Networking|SwitchHasAuthority
    (:Authority
      (Utilities|IsValid QueryInstance
        (:Is%20Valid
          (AI|EQS|UnbindEventfromOnQueryFinished QueryInstance (delegate Custom|HandleSupplyDropSpawnEqsQueryComplete))
          (bind (_resultlocations _returnvalue) (AI|EQS|GetQueryResultsasLocations QueryInstance))
          (if _returnvalue
            (if (Utilities|Array|IsValidIndex _resultlocations)
              (if (< (Variables|Transient|SupplyDrop|GetNumOfSupplyDropsSpawned) (ScalableFloat|GetCachedScalableFloatValue (Variables|Hotfix|GetHotfix_SupplyDropMaxNumToSpawn)))
                (CallFunction|SpawnSupplyDropatLocation (AI|Navigation|GetRandomReachablePointinRadius (Utilities|Array|Get(aref) _resultlocations) (ScalableFloat|GetCachedScalableFloatValue (Variables|Hotfix|GetHotfix_SupplyDropMaxRadiusAroundSelectedLocation))))
                (Math|Integer|IncrementInt (Variables|Transient|SupplyDrop|GetNumOfSupplyDropsSpawned))
                (else
                  (CallFunction|SetEnableSpawnSupplyDrops false))))))))))

(event Custom|SpawnLightningBoltAtLocation
  (params (InLocation :type Vector))
  ;[0 0.03701 0.15 0.5] SpawnLightningBoltAtLocation: Spawn the lightning bolt at the desired location. It cleans up own its own.
  (Networking|SwitchHasAuthority
    (:Authority
      (bind _class (Utilities|AsyncLoadClassAsset (Variables|LightningSequence|GetLightningBoltActorSoft))
        (:Completed
          (Game|SpawnActorfromClass (Utilities|Casting|CastToActorClass _class) (Math|Conversions|ToTransform(Vector) InLocation) "AlwaysSpawn" "MultiplyWithRoot" 0 0) ; Actor is temporary, removes itself after 
        )))))

(event Custom|SpawnMusicComponentForLocalPlayerControllers
  ;[0.15 0.02421 0.1482 0.5] SpawnMusicComponentForLocalPlayerControllers: Add music manager component for local players if they do not have one already.
  (Networking|CanExecuteCosmeticEvents
    (:True
      (bind _class (Utilities|AsyncLoadClassAsset (Variables|Music|GetMusicManagerControllerComponent))
        (:Completed
          (bind _asactor_component (Utilities|Casting|CastToActorComponentClass _class))
          (for _array_element (Fortnite|GetLocalFortPlayerControllers true true)
            (Utilities|IsValid _array_element
              (:Is%20Valid
                (Utilities|IsValid (Actor|GetComponentbyClass _array_element _asactor_component)
                  (:Is%20Not%20Valid
                    (Game|AddComponentbyClass _array_element _asactor_component)))))))))))

(event Custom|SetEnableHoagieSpawn
  (params (bEnable :type Boolean))
  ;[0.003247 0.01979 0.08854 0.5] Hoagie (Helicopters!)
  ;;[0 0.15 0.13 0.5] SetEnableHoagieSpawn: Start or stop a timer that attemps to rfit in a hoagie vehicle.
  (Networking|SwitchHasAuthority
    (:Authority
      (Fortnite|FortShippingLog self (Utilities|String|Append "SetEnableHoagieSpawn: " (Utilities|String|ToString(Boolean) bEnable)))
      (if bEnable
        (if (> (Math|Float|Truncate (ScalableFloat|GetCachedScalableFloatValue (Variables|Hotfix|GetHotfix_HoagieMaxSpawnCount))) 0)
          (bind _class (Utilities|AsyncLoadClassAsset (Variables|Hoagie|GetHoagieSpawnerActorClass))
            (:Completed
              (Utilities|CacheAsyncLoadedClass _class
                (:Done
                  (bind _returnvalue (Utilities|Time|SetTimerbyEvent (EventDispatchers|CreateEvent :!event Custom|RunEqsToSpawnHoagie) (ScalableFloat|GetCachedScalableFloatValue (Variables|Hotfix|GetHotfix_HoagieDelayBetweenSpawnAttempt)) true))
                  (Variables|Transient|Hoagie|SetHoagieSpawnTimerHandle _returnvalue)))))
          (else
            (Fortnite|FortShippingLog self "SetEnableHoagieSpawn: Disabling as max hoagie spawn count is invalid.")))
        (else
          _))
      (Utilities|Time|ClearandInvalidateTimerbyHandle (Variables|Transient|Hoagie|GetHoagieSpawnTimerHandle)))))

(event Custom|RunEqsToSpawnHoagie
  (bind _pawn (Giant|GetGiantPawn))
  ;[0.003247 0.01979 0.08854 0.5] Hoagie (Helicopters!)
  ;;[0.15 0.09451 0.0469 0.5] RunEqsToSpawnHoagie: Run an EQS query to figure out the best place to rift in the helicopter.
  (Networking|SwitchHasAuthority
    (:Authority
      (Utilities|IsValid _pawn
        (:Is%20Valid
          (bind _returnvalue (AI|EQS|RunEQSQuery (Variables|Hoagie|GetHoagieSpawnEnvQuery) _pawn "RandomBest5Pct"))
          (Utilities|IsValid _returnvalue
            (:Is%20Valid
              (AI|EQS|BindEventtoOnQueryFinished _returnvalue (EventDispatchers|CreateEvent :!event Custom|HandleHoagieSpawnEqsQueryComplete)))))
        (:Is%20Not%20Valid
          (CallFunction|SetEnableHoagieSpawn false))))))

(event Custom|HandleHoagieSpawnEqsQueryComplete
  (params (QueryInstance :type EQS%20Query%20Instance%20Object%20Reference) (QueryStatus :type EEnvQueryStatus%20Enum :default "Processing"))
  ;[0.003247 0.01979 0.08854 0.5] Hoagie (Helicopters!)
  ;;[0 0.15 0.07702 0.5] HandleHoagieSpawnEqsQueryComplete: Query complete, spawn the hoagie at the result location.
  (Networking|SwitchHasAuthority
    (:Authority
      (Utilities|IsValid QueryInstance
        (:Is%20Valid
          (AI|EQS|UnbindEventfromOnQueryFinished QueryInstance (delegate Custom|HandleHoagieSpawnEqsQueryComplete))
          (bind (_resultlocations _returnvalue) (AI|EQS|GetQueryResultsasLocations QueryInstance))
          (if _returnvalue
            (if (Utilities|Array|IsValidIndex _resultlocations)
              (if (< (Variables|Transient|Hoagie|GetHoagieSpawnCount) (Math|Float|Truncate (ScalableFloat|GetCachedScalableFloatValue (Variables|Hotfix|GetHotfix_HoagieMaxSpawnCount))))
                (if (< (Utilities|Array|Length (Variables|Transient|Hoagie|GetHoagieVehicles)) (Math|Float|Truncate (ScalableFloat|GetCachedScalableFloatValue (Variables|Hotfix|GetHotfix_HoagieMaxAliveCount))))
                  (CallFunction|SpawnHoagieVehicleatLocation (Utilities|Array|Get(aref) _resultlocations)))
                (else
                  (CallFunction|SetEnableHoagieSpawn false))))))))))

(event Custom|SpawnHoagieVehicleAtLocation
  (params (Location :type Vector))
  ;[0.003247 0.01979 0.08854 0.5] Hoagie (Helicopters!)
  ;;[0.08333 0.01666 0.06709 0.5] SpawnHoagieVehicleAtLocation: Spawn the hoagie rift spawner at the desired location.
  (Networking|SwitchHasAuthority
    (:Authority
      (bind _returnvalue (Utilities|LoadClassAssetBlocking (Variables|Hoagie|GetHoagieSpawnerActorClass))) ; Should already be async loaded by this time.
      (Game|SpawnActorfromClass (Utilities|Casting|CastToActorClass _returnvalue) (Math|Transform|MakeTransform Location) "AlwaysSpawn" "MultiplyWithRoot" 0 0)
      (Math|Integer|IncrementInt (Variables|Transient|Hoagie|GetHoagieSpawnCount)))))

(event Custom|SetEnableScanNearbyPlayersForRitaTakeDownAlert
  (params (bEnable :type Boolean))
  ;[0 0.06462 0.15 0.5] Rita Repulsa Take Down Notification/Alert.
  ;;[1 1 1] SetEnableScanNearbyPlayersForRitaTakeDownPrompt: Start looking for nearby players to prompt them to attack megazord.
  (Networking|SwitchHasAuthority
    (:Authority
      (if bEnable
        (bind _returnvalue (Utilities|Time|SetTimerbyEvent (EventDispatchers|CreateEvent :!event Custom|ScanNearbyPlayersForRitaTakeDownAlert) (ScalableFloat|GetCachedScalableFloatValue (Variables|Hotfix|GetHotfix_AlertRitaTakeDownMegazordDistanceTestTimerFreq)) true false 3.0))
        (Variables|Transient|Alert|SetScanNearbyPlayersForRitaTakeDownAlertTimerHandle _returnvalue)
        (else
          (Utilities|Time|ClearandInvalidateTimerbyHandle (Variables|Transient|Alert|GetScanNearbyPlayersForRitaTakeDownAlertTimerHandle)))))))

(event Custom|ScanNearbyPlayersForRitaTakeDownAlert
  (bind _pawn (Giant|GetGiantPawn))
  ;[0 0.06462 0.15 0.5] Rita Repulsa Take Down Notification/Alert.
  ;;[1 1 1] ScanNearbyPlayersForRitaTakeDownAlert: Scan nearby players and let them know about taking down megazord.
  (Networking|SwitchHasAuthority
    (:Authority
      (if (> (PowerLevel|GetPowerLevel) (ScalableFloat|GetCachedScalableFloatValue (Variables|Hotfix|GetHotfix_AlertRitaTakeDownMegazordPowerLevelThreshold)))
        (Utilities|IsValid _pawn
          (:Is%20Valid
            (bind _outfortplayerpawns (Fortnite|GetFortPlayerPawnsInRange (Transformation|GetActorLocation _pawn) (ScalableFloat|GetCachedScalableFloatValue (Variables|Hotfix|GetHotfix_AlertRitaTakeDownMegazordDistanceTestValue)) true))
            (for _array_element _outfortplayerpawns
              (Utilities|IsValid _array_element
                (:Is%20Valid
                  (if (not (GameplayTags|HasMatchingGameplayTag _array_element "(TagName=\"ToyFox.HUD.Alert.Rita.TakeDownMegazord\")"))
                    (Ability|GameplayEffect|AddLooseGameplayTags _array_element (GameplayTags|MakeLiteralGameplayTagContainer "(GameplayTags=((TagName=\"ToyFox.HUD.Alert.Rita.TakeDownMegazord\")))"))
                    (Fortnite|ApplyGameplayEffectsToActor _array_element (Utilities|Array|MakeArray (Variables|Alert|GetRitaTakeDownMegazordGameplayEffectClass)) (0.0 0.0 0.0) false 0 (Ability|GetFortAbilitySystemComponent _array_element) (GameplayEffects|MakeEffectContext (Ability|GetFortAbilitySystemComponent _array_element))))))))
          (:Is%20Not%20Valid
            (CallFunction|SetEnableScanNearbyPlayersforRitaTakeDownAlert false)))))))

(event Custom|BroadcastAlertNotification
  (params (Alert :type Alert%20Notification%20Struct%20Structure))
  ;[0.0952 0 0.15 0.5] Send alerts for HUD notification.
  (Networking|SwitchHasAuthority
    (:Authority
      (CallFunction|MulticastBroadcastAlertNotification Alert))
    (:Remote
      (Default|CallOnAlertReceived Alert))))

(event Custom|MulticastBroadcastAlertNotification
  (params (AlertNotification :type Alert%20Notification%20Struct%20Structure))
  ;[0.0952 0 0.15 0.5] Send alerts for HUD notification.
  (Networking|CanExecuteCosmeticEvents
    (:True
      _)
    (:False
      _))
  (Default|CallOnAlertReceived AlertNotification))

(event Custom|BroadcastAlertGiantFinisherComplete
  ;[0.0952 0 0.15 0.5] Send alerts for HUD notification.
  (Networking|SwitchHasAuthority
    (:Authority
      (Utilities|FlowControl|Delay (ScalableFloat|GetCachedScalableFloatValue (Variables|Hotfix|GetHotfix_AlertDelayBeforeNotificationQueenDefeated)))
      (CallFunction|BroadcastAlertNotification 0 (struct :AlertGameplayTag_12_6C55428345DE14476CEF4D9FECB3F891 "(TagName=\"ToyFox.HUD.Alert.Zordon.Queen.Defeated\")")))))

(event Custom|BroadcastAlertMegazordEixt
  ;[0.0952 0 0.15 0.5] Send alerts for HUD notification.
  (Networking|SwitchHasAuthority
    (:Authority
      (Utilities|FlowControl|Delay (ScalableFloat|GetCachedScalableFloatValue (Variables|Hotfix|GetHotfix_AlertDelayBeforeNotificationMegazordExit)))
      (CallFunction|BroadcastAlertNotification 0 (struct :AlertGameplayTag_12_6C55428345DE14476CEF4D9FECB3F891 "(TagName=\"ToyFox.HUD.Alert.Megazord.Exit\")")))))

(event Custom|HandleEnemyAbilitySystemComponentInitialized
  (bind _asyncactions_enemystunned (Variables|Transient|Enemy|GetAsyncActions_EnemyStunned))
  (bind _enemystunnedgameplaytagquery (Variables|Enemy|GetEnemyStunnedGameplayTagQuery))
  ;[0 0.15 0.1377 0.5] HandleEnemyAbilitySystemComponentInitialized: Bind to know when the enemy is stunned or not and fire delegate.
  (Ability|Async|EndAction _asyncactions_enemystunned)
  (Utilities|Array|Clear _asyncactions_enemystunned)
  (bind _enemy (Variables|Transient|Enemy|GetEnemy)
    (:then
      (bind _asynctaskproxy (Ability|Tasks|WaitGameplayTagQueryOnActor _enemy _enemystunnedgameplaytagquery)
        (:then
          (Utilities|Array|Add _asyncactions_enemystunned _asynctaskproxy)
          (bind _asynctaskproxy_2 (Ability|Tasks|WaitGameplayTagQueryOnActor (Variables|Transient|Enemy|GetEnemy) _enemystunnedgameplaytagquery "WhenFalse")
            (:then
              (Utilities|Array|Add _asyncactions_enemystunned _asynctaskproxy_2)
              (bind _asynctaskproxy_3 (Ability|Tasks|WaitGameplayTagQueryOnActor (Variables|Transient|Enemy|GetEnemy) (Variables|Enemy|GetEnemySpawnInGameplayTagQuery))
                (:then
                  (Utilities|Array|Add _asyncactions_enemystunned _asynctaskproxy_3))
                (:Triggered
                  (Default|CallOnEnemySpawninComplete))))
            (:Triggered
              (Default|CallOnEnemyStunnedActiveChanged false))))
        (:Triggered
          (Default|CallOnEnemyStunnedActiveChanged true))))
    (:else)))

(event Custom|PlayVOAtLocation
  (params (Tag :type Gameplay%20Tag%20Structure) (Location :type Vector) (bIncludePilots :type Boolean))
  (bind _vomap (Variables|VO|GetVOMap))
  (bind (_value _returnvalue) (Utilities|Map|Find _vomap Tag))
  ;[0.07741 0.15 0.0208 0.5] PlayVoAtLocation: Function to help play the desired VO at a location.
  (Networking|CanExecuteCosmeticEvents
    (:True
      (if (not bIncludePilots)
        (Variables|Transient|PilotCinematic|GetPilotCinematicSequenceActor
          (:then)
          (:else
            _)) ; Only play the VO if the cinematic is not active for the player.
      )
      (if _returnvalue
        (bind _object (Utilities|AsyncLoadAsset _value)
          (:Completed
            (bind _assound_base (Utilities|Casting|CastToSoundBase _object))
            (Audio|PlaySoundatLocation _assound_base Location "0, 0, 0" 1.0 1.0 0.0 "/ToyFoxGameplay/Assets/Sound/ATT_Toyfox_Spawn_Assemble.ATT_Toyfox_Spawn_Assemble")))))))

(event Custom|SpawnLightningGroundRumbeActorAtLocation
  (params (Location :type Vector))
  ;[0 0.1086 0.15 0.5] SpawnLightningGroundRumbeActorAtLocation: Spawns ground rumble actor at input location.
  (Networking|SwitchHasAuthority
    (:Authority
      (bind _class (Utilities|AsyncLoadClassAsset (Variables|LightningSequence|GetLightningGroundRumbleActorSoft))
        (:Completed
          (bind _returnvalue (Game|SpawnActorfromClass (Utilities|Casting|CastToActorClass _class) (Math|Conversions|ToTransform(Vector) Location) "AlwaysSpawn" "MultiplyWithRoot" 0 0))
          (Utilities|Array|Add (Variables|Transient|GetLightningGroundRumbleActors) _returnvalue))) ; Ground rumble is a continuous event so spawning an actor to take care of the visuals. \nActor is destroyed once the bug is spawned.
    )))

(event Custom|SpawnSupplyDropAtLocation
  (params (Location :type Vector))
  ;[0.15 0.06515 0.1288 0.5] SpawnSupplyDropAtLocation: Spawn the supply drop at the desired location.
  (Networking|SwitchHasAuthority
    (:Authority
      (bind _returnvalue (Utilities|LoadClassAssetBlocking (Variables|SupplyDrop|GetSupplyDropSoftClass)))
      (Game|SpawnActorfromClass (Utilities|Casting|CastToActorClass _returnvalue) (Math|Transform|MakeTransform Location (Math|Rotator|MakeRotator (Math|Random|RandomFloatinRange (* (ScalableFloat|GetCachedScalableFloatValue (Variables|Hotfix|GetHotfix_SupplyDropRotationRollPitchMax)) -1.0) (ScalableFloat|GetCachedScalableFloatValue (Variables|Hotfix|GetHotfix_SupplyDropRotationRollPitchMax))) (Math|Random|RandomFloatinRange (* (ScalableFloat|GetCachedScalableFloatValue (Variables|Hotfix|GetHotfix_SupplyDropRotationRollPitchMax)) -1.0) (ScalableFloat|GetCachedScalableFloatValue (Variables|Hotfix|GetHotfix_SupplyDropRotationRollPitchMax))) (Math|Random|RandomFloatinRange 0.0 3.6e+02)) "1.000000,1.000000,1.000000") "AlwaysSpawn" "MultiplyWithRoot" 0 0))))

(event Custom|Cleanup%20Self
  (if (not (ScalableFloat|GetValueasBool (Variables|Hotfix|GetHotfix_EventBypassCheck)))
    (Utilities|FlowControl|Delay 1e+01)
    (Components|DestroyComponent)))

(event Custom|UpdatePlayerDamageTrackingInfo
  ;[1 1 1] Manually update the player damage tracking info to clients to save perf
  (CallFunction|MulticastPlayerDamageTrackingInfo (Variables|Event|GetPlayerDamageTrackingInfo)))

(event Custom|Multicast_PlayerDamageTrackingInfo
  (params (NewTrackingInfo :type Array%20of%20Toy%20Fox%20Player%20Damage%20Struct%20Structures))
  ;[1 1 1] Manually update the player damage tracking info to clients to save perf
  (Variables|Event|SetPlayerDamageTrackingInfo NewTrackingInfo)
  (Default|CallOnPlayerDamageTrackingInfoUpdated))

(event Custom|CheatEndToyFox
  (if (Variables|Transient|GetbObjectiveActive)
    (CallFunction|CheatSetPowerLevel 0.0)))

(event Custom|CheatEventIntro
  (Event|SetEventState "NewEnumerator5"))

(event Custom|CheatBroadcastToyFoxAvailable
  ;[1 1 1] Manually notify players of ToyFox availability in case cheat logic is used to initiate the event
  (CallFunction|BroadcastAlertNotification 0 (struct :AlertGameplayTag_12_6C55428345DE14476CEF4D9FECB3F891 "(TagName=\"ToyFox.HUD.Alert.Megazord.Ready\")")))

(event Custom|CheatActivateEventSoloBypassingCheck
  (params (Spawn%20Transform%20Location :type Vector) (Spawn%20Transform%20Rotation :type Rotator) (PrimaryPlayerPilot :type Fort%20Player%20Pawn%20Object%20Reference))
  ;[1 1 1] Cheat used to repurpose ToyFox giant in other plugins, skips the usual event check but confirms another giant isn't present.
  (Networking|SwitchHasAuthority
    (:Authority
      (if (ScalableFloat|GetValueasBool (Variables|Hotfix|GetHotfix_EventBypassCheck))
        (if (not (Variables|Transient|GetbObjectiveActive))
          (Variables|Transient|GetToyFoxGiantPawn
            (:then)
            (:else
              (bind _returnvalue (Utilities|LoadClassAssetBlocking (Variables|Settings|GetGiantPawnClass)))
              (bind _astoy_fox_giant_player_pawn (Utilities|Casting|CastToToyFoxGiantPlayerPawnClass _returnvalue))
              (bind _returnvalue_2 (Game|SpawnActorfromClass _astoy_fox_giant_player_pawn (Math|Transform|MakeTransform Spawn%20Transform%20Location Spawn%20Transform%20Rotation) "AlwaysSpawn" "MultiplyWithRoot" 0 0))
              (Variables|Transient|SetToyFoxGiantPawn _returnvalue_2)
              (Event|SetEventState "NewEnumerator1")
              (Event|SetEventState "NewEnumerator2")
              (Default|CallOnGiantStartedbyCheat)
              (CallFunction|BeginToyFoxEnterSequence PrimaryPlayerPilot (Utilities|Array|MakeArray 0) true)
              (Variables|Transient|SetbObjectiveActive true))))))))

(event Custom|CheatActivateEventAtLocationWithDirectGameplay
  (params (GiantClass :type Toy%20Fox%20Giant%20Player%20Pawn%20Class%20Reference) (Spawn%20Transform%20Location :type Vector) (Spawn%20Transform%20Rotation :type Rotator) (PrimaryPlayerPilot :type Fort%20Player%20Pawn%20Object%20Reference) (SecondaryPlayerPilots :type Array%20of%20Fort%20Player%20Pawn%20Object%20References))
  ;[0 0 0 0.5] CheatActivateEventAtLocationWithDirectGameplay: Function to help cheat and start the objective directly with the giant possessed.
  (Networking|SwitchHasAuthority
    (:Authority
      (if (Objective|CanActivateObjective)
        (Variables|Transient|SetbObjectiveActive true)
        (bind _returnvalue (Game|SpawnActorfromClass GiantClass (Math|Transform|MakeTransform Spawn%20Transform%20Location Spawn%20Transform%20Rotation) "AlwaysSpawn" "MultiplyWithRoot" 0 0))
        (Variables|Transient|SetToyFoxGiantPawn _returnvalue)
        (Event|SetEventState "NewEnumerator1")
        (Event|SetEventState "NewEnumerator2")
        (CallFunction|BeginToyFoxEnterSequence PrimaryPlayerPilot SecondaryPlayerPilots true)
        (Default|CallOnGiantStartedbyCheat)))))

(event Custom|CheatSetPowerLevel
  (params (PowerLevel :type Float%20%28double-precision%29))
  ;[0 0 0 0.5] CheatSetPowerLevel: Adjust the power level to the desired value.
  (Networking|SwitchHasAuthority
    (:Authority
      (PowerLevel|ReducePowerLevel (- (PowerLevel|GetPowerLevel) (Math|Float|Clamp(Float) PowerLevel))))))

(event Custom|CheatSpawnEnemy
  (params (FortPlayerPawn :type Fort%20Player%20Pawn%20Object%20Reference))
  ;[0 0 0 0.5] CheatSpawnEnemy: Spawn one enemy for the input player pawn closest location
  (Utilities|IsValid FortPlayerPawn
    (:Is%20Valid
      (bind _outactors (Actor|GetAllActorsOfClass "/ToyFoxGameplay/Gameplay/Locations/B_ToyFox_EventLocation.B_ToyFox_EventLocation_C"))
      (bind _asb_toy_fox_event_location (Utilities|Casting|CastToB_ToyFox_EventLocation (Fortnite|GetClosestActorfromArray FortPlayerPawn _outactors)))
      (Default|BindEventtoOnEnemySpawned _asb_toy_fox_event_location (EventDispatchers|CreateEvent :!event HandleEnemySpawned))
      (Class|BToyFoxEventLocation|SpawnEnemy _asb_toy_fox_event_location)
      (Enemy|HandleEnemySpawned _asb_toy_fox_event_location))))

(event Custom|CheatShowEventLocations
  (params (bShow :type Boolean))
  ;[0 0 0 0.5] CheatShowEventLocations: show all event locations
  (bind _outactors (Actor|GetAllActorsOfClass "/ToyFoxGameplay/Gameplay/Locations/B_ToyFox_EventLocation.B_ToyFox_EventLocation_C"))
  (for _array_element _outactors
    (Utilities|IsValid _array_element
      (:Is%20Valid
        (if (Class|BToyFoxEventLocation|IsEnabled _array_element)
          (Class|BToyFoxEventLocation|SetMapIconActive _array_element bShow))))))

(event Custom|CheatActivateObjectiveAtPlayer
  (params (FortPlayerPawn :type Fort%20Player%20Pawn%20Object%20Reference))
  ;[0 0 0 0.5] CheatActivateObjectiveAtPlayer: activate objective at given player's location
  (Utilities|IsValid FortPlayerPawn
    (:Is%20Valid
      (bind _outactors (Actor|GetAllActorsOfClass "/ToyFoxGameplay/Gameplay/Locations/B_ToyFox_EventLocation.B_ToyFox_EventLocation_C"))
      (bind _asb_toy_fox_event_location (Utilities|Casting|CastToB_ToyFox_EventLocation (Fortnite|GetClosestActorfromArray FortPlayerPawn _outactors)))
      (CallFunction|ActivateToyFoxObjectiveatEventActor _asb_toy_fox_event_location))))

(event Custom|CheatSimulateIncomingPlayerDamageOnPowerLevel
  (params (DamagePerSecond :type Float%20%28double-precision%29))
  ;[0 0 0 0.5] CheatSimulateIncomingPlayerDamageOnPowerLevel: Simulate power level reduction as if the giant is taking input DPS. Used to help balance event life time.
  (Networking|SwitchHasAuthority
    (:Authority
      (if (> DamagePerSecond 0)
        (bind _returnvalue (Utilities|Time|SetTimerbyEvent (delegate Custom|CheatApplySimulateIncomingDamage) 1.0 true))
        (Variables|Cheat|SetCheatSimulateIncomingPlayerDamageTimerHandle _returnvalue)
        (else
          (Utilities|Time|ClearandInvalidateTimerbyHandle (Variables|Cheat|GetCheatSimulateIncomingPlayerDamageTimerHandle)))))))

(event Custom|CheatApplySimulateIncomingDamage
  ;[0 0 0 0.5] CheatSimulateIncomingPlayerDamageOnPowerLevel: Simulate power level reduction as if the giant is taking input DPS. Used to help balance event life time.
  (Networking|SwitchHasAuthority
    (:Authority
      (if (Utilities|Enum|Equal(Enum) (Event|GetEventState) "NewEnumerator3")
        (PowerLevel|ReducePowerLevelbyDamage (stale Custom|CheatSimulateIncomingPlayerDamageOnPowerLevel DamagePerSecond))))))

(fn ActivateObjective
  ;[0 0.15 0.02993 0.5] ActivateObjective: Invoked when the safe zone timings have been reached to activate the objective.
  (|CallParentFunction)
  (Networking|SwitchHasAuthority
    (:Authority
      (CallFunction|ActivateToyFoxObjective))))

(fn AddAnalyticsAttribute
  (params (AttributeName :type String) (StringValue :type String) (FGuidValue :type Guid%20Structure) (IntValue :type Integer :default -1) (floatValue :type Float%20%28single-precision%29 :default -1.0) (BoolValue :type Boolean))
  ;[0.009784 0.02039 0.15 0.5] AddAnalyticsAttribute: Add the desired analytic attribute value to the array and log it.
  (Networking|SwitchHasAuthority
    (:Authority
      (bind _returnvalue (Analytics|ConstructAnalyticsAttribute AttributeName StringValue FGuidValue IntValue floatValue BoolValue))
      (Utilities|Array|Add (Variables|Analytics|Transient|GetAnalyticsEventAttributes) _returnvalue)
      (Utilities|FlowControl|Sequence
        (:then_0
          (if (not (Utilities|String|IsEmpty (Variables|GetStringValue)))
            (LocalVariables|SetAttributeValue (Variables|GetStringValue))
            (elif (Guid|IsValidGuid (Variables|GetFGuidValue))
              (LocalVariables|SetAttributeValue (Guid|ToString(Guid) (Variables|GetFGuidValue)))
              (elif (!= (Variables|GetIntValue) -1)
                (LocalVariables|SetAttributeValue (Utilities|String|ToString(Integer) (Variables|GetIntValue)))
                (elif (not (Math|Float|NearlyEqual(Float) (Variables|GetfloatValue) -1.0))
                  (LocalVariables|SetAttributeValue (Utilities|String|ToString(Float) (Variables|GetfloatValue)))
                  (else
                    (LocalVariables|SetAttributeValue (Utilities|String|ToString(Boolean) (Variables|GetBoolValue)))))))))
        (:then_1
          (Fortnite|FortShippingLog self (Utilities|String|Append "AddAnalyticsAttribute: " AttributeName " : " (LocalVariables|GetAttributeValue))))))))

(fn ApplyPilotStatus
  (bind _primarypilotplayerpawn (Variables|Transient|Pilots|GetPrimaryPilotPlayerPawn))
  ;[0 0.03784 0.1354 0.5] ApplyPilotStatus: Secondary pilots enter stasis and get attached to the giant pawn mesh. Primary need to apply after giant pawn possession
  (Networking|SwitchHasAuthority
    (:Authority
      (Utilities|IsValid _primarypilotplayerpawn
        (:Is%20Valid
          (bind _returnvalue (Utilities|LoadClassAssetBlocking (Variables|Settings|GetPilotStasisGameplayEffect))) ; In regular event flow, these will already be async loaded by this point.
          (bind _returnvalue_2 (Utilities|LoadClassAssetBlocking (Variables|Settings|GetPrimaryPilotGameplayEffectClass))) ; In regular event flow, these will already be async loaded by this point.
          (Fortnite|ApplyGameplayEffectsToActor _primarypilotplayerpawn (Utilities|Array|MakeArray (Utilities|Casting|CastToGameplayEffectClass _returnvalue_2) (Utilities|Casting|CastToGameplayEffectClass _returnvalue)) (0.0 0.0 0.0) false 0 (Ability|GetFortAbilitySystemComponent _primarypilotplayerpawn) (GameplayEffects|MakeFortEffectContext (Ability|GetFortAbilitySystemComponent _primarypilotplayerpawn) (Giant|GetGiantPawn) 0 0 1))
          (for _array_element (Variables|Transient|Pilots|GetSecondaryPilotPlayerPawns)
            (Utilities|IsValid _array_element
              (:Is%20Valid
                (bind _returnvalue_3 (Utilities|LoadClassAssetBlocking (Variables|Settings|GetSecondaryPilotGameplayEffectClass))) ; In regular event flow, these will already be async loaded by this point.
                (bind _returnvalue_4 (Utilities|LoadClassAssetBlocking (Variables|Settings|GetPilotStasisGameplayEffect))) ; In regular event flow, these will already be async loaded by this point.
                (Fortnite|ApplyGameplayEffectsToActor _array_element (Utilities|Array|MakeArray (Utilities|Casting|CastToGameplayEffectClass _returnvalue_4) (Utilities|Casting|CastToGameplayEffectClass _returnvalue_3)) (0.0 0.0 0.0) false 0 (Ability|GetFortAbilitySystemComponent _array_element) (GameplayEffects|MakeFortEffectContext (Ability|GetFortAbilitySystemComponent _array_element) (Giant|GetGiantPawn) 0 0 1))))))))))

(fn AttachPlayersToGiant
  (if (Variables|Transient|GetbIntermittentPawnAttach)
    (for _array_element (Variables|Transient|Pilots|GetSecondaryPilotPlayerPawns)
      (Class|ToyFoxGiantPlayerPawn|AttachActortoSelf (Giant|GetGiantPawn) _array_element))
    (else
      (bind _output_get (LocalVariables|SetAllPlayerPawns (Variables|Transient|Pilots|GetSecondaryPilotPlayerPawns)))
      (Utilities|Array|Add _output_get (Variables|Transient|Pilots|GetPrimaryPilotPlayerPawn))
      (for _array_element_2 (LocalVariables|GetAllPlayerPawns)
        (Transformation|AttachActorToActor _array_element_2 (Giant|GetGiantPawn) "None" "SnapToTarget" "KeepRelative" "KeepRelative" true)))))

(fn CachePilotAnalyticsInfo
  (bind _secondarypilotplayerpawns (Variables|Transient|Pilots|GetSecondaryPilotPlayerPawns))
  ;[0.1035 0 0.15 0.5] CachePilotAnalyticsInfo: Add primary and secondary pilots info into the analytics.
  (Networking|SwitchHasAuthority
    (:Authority
      (Utilities|FlowControl|Sequence
        (:then_0
          (bind _returnvalue (Fortnite|Utilities|AttemptGetPlayerStateforPawn (Variables|Transient|Pilots|GetPrimaryPilotPlayerPawn)))
          (Utilities|IsValid _returnvalue
            (:Is%20Valid
              (bind _returnvalue_2 (PlayerInfo|GetDebugStringforUniqueId (PlayerState|GetUniqueNetId _returnvalue)))
              (Analytics|AddAnalyticsAttribute "PrimaryPilotUniqueNetId" _returnvalue_2))))
        (:then_1
          (for _index (range 4)
            (if (Utilities|Array|IsValidIndex _secondarypilotplayerpawns _index)
              (bind _returnvalue_3 (Fortnite|Utilities|AttemptGetPlayerStateforPawn (Utilities|Array|Get(acopy) _secondarypilotplayerpawns _index)))
              (Utilities|IsValid _returnvalue_3
                (:Is%20Valid
                  (bind _returnvalue_4 (PlayerInfo|GetDebugStringforUniqueId (PlayerState|GetUniqueNetId _returnvalue_3)))
                  (Analytics|AddAnalyticsAttribute (Utilities|String|Append "SecondaryPilotUniqueNetId_" (Utilities|String|ToString(Integer) _index)) _returnvalue_4))
                (:Is%20Not%20Valid
                  _))
              (else
                _))
            (Analytics|AddAnalyticsAttribute (Utilities|String|Append "SecondaryPilotUniqueNetId_" (Utilities|String|ToString(Integer) _index)) "NOPILOT")))))))

(fn CanActivateObjective
  ;[0.15 0.02987 0.148 0.5] Returns if the objective can be activated.
  ; It can only be activated if there's a valid giant player pawn and the objective is not already activated.
  (Utilities|FlowControl|Sequence
    (:then_0
      (Utilities|SwitchHasAuthority
        (:Authority
          (if (not (Variables|Transient|GetbObjectiveActive))
            (Variables|Transient|GetToyFoxGiantPawn
              (:then)
              (:else
                (if (Variables|Transient|GetbEventRequiredCommunityQuestComplete)
                  (if (<= (Math|Random|RandomFloat) (ScalableFloat|GetCachedScalableFloatValue (Variables|Hotfix|GetHotfix_EventSpawnRate)))
                    (return true)
                    (else
                      (Fortnite|FortShippingLog self "Hotfix spawn rate check failed, will not activate the event"))))))))))
    (:then_1
      (return false))))

(fn ClearAllPowerLevel
  (Networking|SwitchHasAuthority
    (:Authority
      (Variables|PowerLevel|SetPowerLevel 0.0))))

(fn CreateControllerComponents
  (Game|AddComponentbyClass (Pawn|GetController (Variables|Transient|Pilots|GetPrimaryPilotPlayerPawn)) "/ToyFoxGameplay/Gameplay/ControllerComp/B_ToyFoxControllerComp_Primary.B_ToyFoxControllerComp_Primary_C")
  (for _array_element (Variables|Transient|Pilots|GetSecondaryPilotPlayerPawns)
    (Game|AddComponentbyClass (Pawn|GetController _array_element) "/ToyFoxGameplay/Gameplay/ControllerComp/B_ToyFoxControllerComp_Secondary.B_ToyFoxControllerComp_Secondary_C")))

(fn CreatePilotActorsArray
  ;[0.07487 0.004033 0.15 0.5] Create pilot actors array from the primary, secondary, and stand in pilots.
  ; This is replicated to the clients.
  (Networking|SwitchHasAuthority
    (:Authority
      (bind _primarypilotplayerpawn (Variables|Transient|Pilots|GetPrimaryPilotPlayerPawn)
        (:then
          (Utilities|Array|Add (Variables|Transient|Pilots|GetPilotActors) _primarypilotplayerpawn)
          (for _array_element (Variables|Transient|Pilots|GetSecondaryPilotPlayerPawns)
            (Utilities|IsValid _array_element
              (:Is%20Valid
                (Utilities|Array|Add (Variables|Transient|Pilots|GetPilotActors) _array_element))))
          (for _array_element_2 (Variables|Transient|Pilots|GetStandInPilots)
            (Utilities|IsValid _array_element_2
              (:Is%20Valid
                (Utilities|Array|Add (Variables|Transient|Pilots|GetPilotActors) _array_element_2)))))
        (:else)))))

(fn EnableEnemyRelocationDataLayer
  ;[0.0301 0.1335 0.15 0.5] Enable the datalayers that house actors to faclitate relocation of the enemy boss.
  (Networking|SwitchHasAuthority
    (:Authority
      (bind _returnvalue (WorldPartition|GetDataLayerManager))
      (Utilities|IsValid _returnvalue
        (:Is%20Valid
          (DataLayers|SetDataLayerRuntimeState _returnvalue (Variables|Enemy|GetEnemyRelocationDataLayerAsset) "Activated" false :!class /Script/Engine.DataLayerManager))))))

(fn EnterState_End
  ;[0.1013 0 0.15 0.5] EnterState_End: Stop spawning supply drops, reward the medallion, and destroy the standin pilots.
  (Networking|SwitchHasAuthority
    (:Authority
      (Utilities|FlowControl|Sequence
        (:then_0
          (CallFunction|SetEnableScanNearbyPlayersforRitaTakeDownAlert false)
          (CallFunction|SetEnableSpawnSupplyDrops false)
          (CallFunction|SetEnableHoagieSpawn false)
          (Reward|RewardGiantDamageMedallion)
          (CallFunction|BroadcastAlertNotification 0 (Utilities|Struct|MakeAlertNotificationStruct "(TagName=\"ToyFox.HUD.Alert.Rita.Reward.Medallion\")" (Utilities|Array|MakeArray (Variables|Transient|Reward|GetGiantDamageMedallionRewardedPlayerState))))
          (Gameplay|SilentDie (Variables|Transient|Pilots|GetStandInPilots) false :!class /Script/FortniteGame.BuildingActor)
          (bind _output_get (Variables|Analytics|SetGiantPawnLifeTime (- (Fortnite|Sync|Time|GetServerWorldTimeSeconds) (Variables|Analytics|Transient|GetGiantPawnPossessStartServerGameTime)))) ; Analytics
          (Fortnite|FortShippingLog self (Utilities|String|Append "EnterState_End: GiantPawnEventLifeTime: " (Utilities|String|ToString(Float) _output_get)))
          (Analytics|SubmitAnalytics))
        (:then_1
          ;[1 1 1] Remove CriticalDamageOverride GEs
          (bind _output_get_2 (Variables|Transient|SetTempPlayerControllers (Fortnite|GetAllFortPlayerControllers true true)))
          (for _array_element _output_get_2
            (Game|Player|UnbindEventfromOnPlayerPawnPossessed _array_element (EventDispatchers|CreateEvent :!event HandlePawnPossessedSinceGiantEvent)))
          (bind _outfortplayerpawns (Fortnite|GetFortPlayerPawns)) ; This will include all player pawns, henchmen, human AIs
          (for _array_element_2 _outfortplayerpawns
            (GameplayEffects|RemoveActiveGameplayEffectbySourceEffect (Ability|GetFortAbilitySystemComponent _array_element_2) (Variables|PowerLevel|GetGE_CriticalDamageOverride) 0 -1)))
        (:then_2
          (CallFunction|CleanupSelf))))))

(fn EnterState_EventIntro
  (Networking|SwitchHasAuthority
    (:Remote
      (CallFunction|SpawnMusicComponentforLocalPlayerControllers))))

(fn EnterState_GiantAvailable
  ;[0.003489 0.15 0.05056 0.5] EnterState_GiantAvailable: The giant is now available to be possessed.
  (Networking|SwitchHasAuthority
    (:Authority
      (Variables|Analytics|Transient|SetGiantAvailableStartServerGameTime (Fortnite|Sync|Time|GetServerWorldTimeSeconds))
      ;[0 0.07144 0.15 0.5] Fire boardcast verb / quest for all players still in the match.
      (bind _output_get (Variables|Transient|SetTempPlayerControllers (Fortnite|GetAllFortPlayerControllers true)))
      (for _array_element _output_get
        (Utilities|IsValid _array_element
          (:Is%20Valid
            (bind _playerstate (Variables|/Script/Engine.Controller|GetPlayerState _array_element)
              (:then
                (bind _asfort_player_state_athena (Utilities|Casting|CastToFortPlayerStateAthena _playerstate))
                (if (<= (Variables|/Script/FortniteGame.FortPlayerStateAthena|GetPlace _asfort_player_state_athena) 0)
                  (CustomVerb|BroadcastCustomVerbMessage (CustomVerb|MakeCustomVerbMessage _array_element "(TagName=\"CustomVerb.ToyFox.PedestalSpawned\")"))))
              (:else))))))))

(fn EnterState_GiantIntro
  (bind _pawn (Giant|GetGiantPawn))
  (Networking|SwitchHasAuthority
    (:Authority
      ;[0.02724 0.09501 0.15 0.5] Bind to desired delegates after the giant has spawned.
      (EventDispatchers|BindEventtoOnDamaged _pawn (EventDispatchers|CreateEvent :!event HandleGiantPawnDamaged))
      (Game|Player|BindEventtoOnPawnPossessed _pawn (EventDispatchers|CreateEvent :!event HandleGiantPawnPossessed)))))

(fn EnterState_GiantPossessed
  ;[0.01261 0.1094 0 0.5] EnterState_GiantPossessed: Start spawning the supply drops.
  (Networking|SwitchHasAuthority
    (:Authority
      (Utilities|FlowControl|Sequence
        (:then_0
          (CallFunction|SetEnableScanNearbyPlayersforRitaTakeDownAlert true)
          (CallFunction|SetEnableSpawnSupplyDrops true (Math|Float|Truncate (ScalableFloat|GetCachedScalableFloatValue (Variables|Hotfix|GetHotfix_SupplyDropNumToSpawnAtStart))))
          (CallFunction|SetEnableHoagieSpawn true)
          (bind _output_get (Variables|Analytics|Transient|SetGiantPawnPossessStartServerGameTime (Fortnite|Sync|Time|GetServerWorldTimeSeconds)))
          (Variables|Analytics|SetGiantAvailableTime (- _output_get (Variables|Analytics|Transient|GetGiantAvailableStartServerGameTime))) ; Analytics
          (bind _eventlocationcombatanalyticseventactor (Variables|Analytics|Transient|GetEventLocationCombatAnalyticsEventActor)
            (:then
              (Actor|DestroyActor _eventlocationcombatanalyticseventactor) ; Destroying the actor ends the analytics capture\nafter submitting it to the game state component.
            )
            (:else
              _))
          (bind _eventlocationmarkactor (Variables|Transient|GetEventLocationMarkActor)
            (:then
              (Gameplay|SilentDie _eventlocationmarkactor false :!class /Script/FortniteGame.BuildingActor)
              (Networking|ForceNetUpdate _eventlocationmarkactor))
            (:else)))
        (:then_1
          ;[1 1 1] Apply CriticalDamageOverride GEs to all players in match, register pawn possession events to apply
          (bind _output_get_2 (Variables|Transient|SetTempPlayerControllers (Fortnite|GetAllFortPlayerControllers true true)))
          (for _array_element _output_get_2
            (Game|Player|BindEventtoOnPlayerPawnPossessed _array_element (EventDispatchers|CreateEvent :!event HandlePawnPossessedSinceGiantEvent)) ; In case this player dies and all GEs get refreshed
          )
          (bind _outfortplayerpawns (Fortnite|GetFortPlayerPawns)) ; This will include all player pawns, henchmen, human AIs
          (for _array_element_2 _outfortplayerpawns
            (Fortnite|ApplyGameplayEffectsToActor _array_element_2 (Utilities|Array|MakeArray "/ToyFoxGameplay/Gameplay/Weakpoints/GE_ToyFox_CriticalDamage.GE_ToyFox_CriticalDamage_C") (0.0 0.0 0.0) false 0 (Ability|GetFortAbilitySystemComponent _array_element_2) (GameplayEffects|MakeFortEffectContext (Ability|GetFortAbilitySystemComponent _array_element_2) (Giant|GetGiantPawn) (Giant|GetGiantPawn) 0 0))))))))

(fn GetDamageDoneToGiantByPlayer
  (params (Player :type Fort%20Player%20State%20Athena%20Object%20Reference))
  (Utilities|FlowControl|Sequence
    (:then_0
      (Utilities|IsValid Player
        (:Is%20Valid
          (bind _damageinfo (TrackedPlayerDamage|GetGiantDamageDonebyPlayersTrackingInfo))
          (for _array_element _damageinfo
            (if (== (Utilities|Struct|BreakToyFoxPlayerDamageStruct _array_element) Player)
              (return (.TotalDamage_5_F91677274558D42AB1D4FFBE0B7EFC33 (Utilities|Struct|BreakToyFoxPlayerDamageStruct _array_element))))))))
    (:then_1
      (return 0.0))))

(fn GetEnemy
  (return (Variables|Transient|Enemy|GetEnemy)))

(fn GetEnemyHealthPercent
  (Utilities|FlowControl|Sequence
    (:then_0
      (bind _enemy (Variables|Transient|Enemy|GetEnemy)
        (:then
          (return (Health|GetHealthPercent _enemy :!class /Script/FortniteGame.BuildingActor)))
        (:else)))
    (:then_1
      (return 1.0))))

(fn GetEventLocationMarkActorSpawnTransform
  (bind (_bvalid _location) (Class|BToyFoxEventLocation|GetPedestalLocation (Variables|Event|GetEventLocationActor)))
  (return (Math|Conversions|ToTransform(Vector) (Utilities|Operators|Add _location (Math|Vector|MakeVector 0.0 0.0 (ScalableFloat|GetCachedScalableFloatValue (Variables|Hotfix|GetHotfix_EventIntroMarkerHeight)))))))

(fn GetEventState
  (return (Variables|Transient|GetEventState)))

(fn GetGiantDamageDoneByPlayersTrackingInfo
  (return (Variables|Event|GetPlayerDamageTrackingInfo)))

(fn GetGiantPawn
  (return (Variables|Transient|GetToyFoxGiantPawn)))

(fn GetGiantPossessingPlayerState
  (return (Variables|Transient|GetGiantPossessingPlayerState)))

(fn GetHighestDamageDoneToGiantByAnyPlayer
  (params (ExcludingPlayerStates :type Array%20of%20Fort%20Player%20State%20Object%20References))
  (bind (_player_2_e4151eb841c86f790079458b29687f51 _totaldamage_5_f91677274558d42ab1d4ffbe0b7efc33) (Utilities|Struct|BreakToyFoxPlayerDamageStruct (LocalVariables|GetRetDamageInfo)))
  (Utilities|FlowControl|Sequence
    (:then_0
      (bind _damageinfo (TrackedPlayerDamage|GetGiantDamageDonebyPlayersTrackingInfo))
      (for _array_element _damageinfo
        (Utilities|IsValid (Utilities|Struct|BreakToyFoxPlayerDamageStruct _array_element)
          (:Is%20Valid
            (if (and (> (.TotalDamage_5_F91677274558D42AB1D4FFBE0B7EFC33 (Utilities|Struct|BreakToyFoxPlayerDamageStruct _array_element)) _totaldamage_5_f91677274558d42ab1d4ffbe0b7efc33) (not (Utilities|Array|ContainsItem ExcludingPlayerStates (Utilities|Struct|BreakToyFoxPlayerDamageStruct _array_element))))
              (LocalVariables|SetRetDamageInfo _array_element))))))
    (:then_1
      (return (LocalVariables|GetRetDamageInfo)))))

(fn GetPedestalInteractionInfo
  (return (Variables|Transient|GetPedestalInteractionInfo)))

(fn GetPilotActorPortraitInfo
  (params (Actor :type Actor%20Object%20Reference))
  (bind _playerstate (Giant|GetGiantPossessingPlayerState))
  (bind _returnvalue (Fortnite|GetPlayerNameSafe _playerstate))
  ;[0.0162 0 0.15 0.5] Returns the pilot actor portrait info based on if they are a pawn or a stand in pilot.
  (Utilities|FlowControl|Sequence
    (:then_0
      (Utilities|IsValid Actor
        (:Is%20Valid
          (if (Utilities|Array|ContainsItem (Pilots|GetPilotActors) Actor)
            (bind _asfort_player_pawn (Utilities|Casting|CastToFortPlayerPawn Actor)
              (:then
                (bind _output_get (LocalVariables|SetPilotName (Fortnite|GetPlayerNameSafe _asfort_player_pawn)))
                (if (not (Utilities|String|NotEqual,CaseInsenstive(String) _output_get "NONE"))
                  (if (and (== (Utilities|Array|FindItem (Pilots|GetPilotActors) Actor) 0) (Utilities|String|NotEqual,CaseInsenstive(String) _returnvalue "NONE"))
                    (LocalVariables|SetPilotName _returnvalue)))
                (bind _returnvalue_2 (Cosmetics|GetPawnCosmeticLoadout _asfort_player_pawn false true))
                (Utilities|IsValid (.Character (Utilities|Struct|BreakFortAthenaLoadout _returnvalue_2))
                  (:Is%20Valid
                    (return true (Class|FortItemFunctionLibrary|GetSmallPreviewImage (.Character (Utilities|Struct|BreakFortAthenaLoadout _returnvalue_2))) (Class|FortItemFunctionLibrary|GetLargePreviewImage (.Character (Utilities|Struct|BreakFortAthenaLoadout _returnvalue_2))) (Utilities|Text|ToText(String) (LocalVariables|GetPilotName))))))
              (:CastFailed
                (bind _characteritemdef (Class|BPISecondaryPilotStandIn|GetCharacterItemDef(Message) Actor))
                (Utilities|IsValid _characteritemdef
                  (:Is%20Valid
                    (return true (Class|FortItemFunctionLibrary|GetSmallPreviewImage _characteritemdef) (Class|FortItemFunctionLibrary|GetLargePreviewImage _characteritemdef) (Variables|/Script/ItemizationCoreRuntime.ItemDefinitionBase|GetItemName _characteritemdef))))))))))
    (:then_1
      (return false "" "" ""))))

(fn GetPilotActors
  (return (Variables|Transient|Pilots|GetPilotActors)))

(fn GetPilotCinematicCameraActorView
  (return (Variables|Transient|GetPilotCinematicCameraView)))

(fn GetPilotExitLocationAndRotation
  (params (PlayerPawn :type Fort%20Player%20Pawn%20Object%20Reference))
  ;[0.1264 0 0.15 0.5] GetPilotExitLocationAndRotation: Exit the pilot player pawn away from the giant, facing the giant's forward, and at a minimum height from the ground.
  (Utilities|FlowControl|Sequence
    (:then_0
      (Networking|SwitchHasAuthority
        (:Authority
          (Utilities|IsValid PlayerPawn
            (:Is%20Valid
              (bind _toyfoxgiantpawn (Variables|Transient|GetToyFoxGiantPawn)
                (:then
                  (bind (_returnvalue_x _returnvalue_y _returnvalue_z) (Math|Vector|BreakVector (Transformation|GetActorLocation _toyfoxgiantpawn)))
                  (bind _returnvalue (Math|Vector|MakeVector (.x (+ (* (Math|Vector|RotateVectorAroundAxis (Math|Vector|Normalize2D(Vector) (Transformation|GetActorForwardVector _toyfoxgiantpawn)) (Math|Random|RandomFloatinRange (* (ScalableFloat|GetCachedScalableFloatValue (Variables|Hotfix|GetHotfix_ExitPawnLocationMaxAngleOffset)) -1.0) (ScalableFloat|GetCachedScalableFloatValue (Variables|Hotfix|GetHotfix_ExitPawnLocationMaxAngleOffset))) "0.000000,0.000000,1.000000") (Math|Random|RandomFloatinRange (ScalableFloat|GetCachedScalableFloatValue (Variables|Hotfix|GetHotfix_ExitPawnLocationDistanceOffsetMin)) (ScalableFloat|GetCachedScalableFloatValue (Variables|Hotfix|GetHotfix_ExitPawnLocationDistanceOffsetMax)))) (Transformation|GetActorLocation _toyfoxgiantpawn))) (.y (+ (* (Math|Vector|RotateVectorAroundAxis (Math|Vector|Normalize2D(Vector) (Transformation|GetActorForwardVector _toyfoxgiantpawn)) (Math|Random|RandomFloatinRange (* (ScalableFloat|GetCachedScalableFloatValue (Variables|Hotfix|GetHotfix_ExitPawnLocationMaxAngleOffset)) -1.0) (ScalableFloat|GetCachedScalableFloatValue (Variables|Hotfix|GetHotfix_ExitPawnLocationMaxAngleOffset))) "0.000000,0.000000,1.000000") (Math|Random|RandomFloatinRange (ScalableFloat|GetCachedScalableFloatValue (Variables|Hotfix|GetHotfix_ExitPawnLocationDistanceOffsetMin)) (ScalableFloat|GetCachedScalableFloatValue (Variables|Hotfix|GetHotfix_ExitPawnLocationDistanceOffsetMax)))) (Transformation|GetActorLocation _toyfoxgiantpawn))) (+ _returnvalue_z (Class|ToyFoxGiantPlayerPawn|GetPawnExitTeleportExtraHeight _toyfoxgiantpawn))))
                  (bind _returnvalue_2 (Helpers|FindStaticGroundLocationUsingWorldContextAt self _returnvalue))
                  (bind _returnvalue_3 (Math|Vector|MakeVector (.x _returnvalue) (.y _returnvalue) (Math|Float|Max(Float) (+ (.z _returnvalue_2) (ScalableFloat|GetCachedScalableFloatValue (Variables|Hotfix|GetHotfix_ExitPawnLocationMinHeightFromGround))) (.z _returnvalue))))
                  (bind (_returnvalue_roll _returnvalue_pitch _returnvalue_yaw) (Math|Rotator|BreakRotator (Math|Rotator|FindLookatRotation _returnvalue_3 (Transformation|GetActorLocation _toyfoxgiantpawn))))
                  (return true _returnvalue_3 0.0 0.0 _returnvalue_yaw))
                (:else
                  (return true (.location (Transformation|GetActorTransform PlayerPawn)) (.rotation (Transformation|GetActorTransform PlayerPawn))))))))))
    (:then_1
      (return false "0, 0, 0" "0, 0, 0"))))

(fn GetPilotPlayerPawns
  ;[0.03215 0.03111 0.15 0.5] GetPilotPlayerPawns: Return all the pilot player pawn actors. Note: This only works on the server.
  (Utilities|FlowControl|Sequence
    (:then_0
      (Networking|SwitchHasAuthority
        (:Authority
          (bind _output_get (LocalVariables|SetRetPilots (Utilities|Array|MakeArray (Variables|Transient|Pilots|GetPrimaryPilotPlayerPawn))))
          (Utilities|Array|AppendArray _output_get (Variables|Transient|Pilots|GetSecondaryPilotPlayerPawns)))))
    (:then_1
      (return (LocalVariables|GetRetPilots)))))

(fn GetPilotPlayerStates
  ;[0.07342 0.05283 0.15 0.5] Return the player states of the pilot player pawns. Only works on the server.
  (Utilities|FlowControl|Sequence
    (:then_0
      (Networking|SwitchHasAuthority
        (:Authority
          (for _array_element (Pilots|GetPilotPlayerPawns)
            (bind _returnvalue (Fortnite|Utilities|AttemptGetPlayerStateforPawn _array_element))
            (bind _asfort_player_state_athena (Utilities|Casting|CastToFortPlayerStateAthena _returnvalue))
            (Utilities|Array|Add (LocalVariables|GetRetPlayerStates) _asfort_player_state_athena)))))
    (:then_1
      (return (LocalVariables|GetRetPlayerStates)))))

(fn GetPowerLevel
  (return (Variables|PowerLevel|GetPowerLevel)))

(fn GetPowerLevelLow
  (return (Variables|Transient|GetbGiantPowerLevelLow)))

(fn GetRegisteredEventLocationActors
  (return (Variables|Event|GetRegisteredEventLocations)))

(fn GetRewardBoonThresholdDamage
  (return (ScalableFloat|GetCachedScalableFloatValue (Variables|Hotfix|GetHotfix_DamageThresholdForBoonReward))))

(fn GetSecondaryPilotPlayerPawns
  ;[0.03215 0.03111 0.15 0.5] GetSecondaryPilotPlayerPawns: Return all the secondary pilot player pawn actors. Note: This only works on the server.
  (return (Variables|Transient|Pilots|GetSecondaryPilotPlayerPawns)))

(fn GetSelectedEventLocationActor
  ;[0 0.08731 0.15 0.5] Returns the selected event location actor.
  (return (Variables|Event|GetEventLocationActor)))

(fn HandleEnemyDamaged
  (params (DamagedActor :type Actor%20Object%20Reference) (Damage :type Float%20%28single-precision%29) (InstigatedBy :type Controller%20Object%20Reference) (DamageCauser :type Actor%20Object%20Reference) (HitLocation :type Vector) (FHitComponent :type Primitive%20Component%20Object%20Reference) (BoneName :type Name) (Momentum :type Vector))
  (Networking|SwitchHasAuthority))

(fn HandleEnemyDestroyed
  (params (DestroyedActor :type Actor%20Object%20Reference))
  ;[0.01324 0 0.15 0.5] HandleEnemyDestroyed: Called when the enemy boss is destroyed.
  (Default|CallOnEnemyDestroyed)
  (Utilities|FlowControl|Sequence
    (:then_0
      (Networking|SwitchHasAuthority
        (:Authority
          (Variables|Analytics|SetEnemyBossLifeTime (- (Fortnite|Sync|Time|GetServerWorldTimeSeconds) (Variables|Analytics|Transient|GetEnemySpawnedServerGameTime))) ; Analytics
          (if (> (ScalableFloat|GetCachedScalableFloatValue (Variables|Hotfix|GetHotfix_SupplyDropStopSpawningAfterEnemyDeath)) 0)
            (CallFunction|SetEnableSpawnSupplyDrops false)))))
    (:then_1
      (EventDispatchers|UnbindEventfromOnAbilitySystemComponentInitialized (Variables|Transient|Enemy|GetEnemy) (EventDispatchers|CreateEvent :!event HandleEnemyAbilitySystemComponentInitialized))
      (Ability|Async|EndAction (Variables|Transient|Enemy|GetAsyncActions_EnemyStunned)))))

(fn HandleEnemyHealthChanged
  (bind _enemy (Variables|Transient|Enemy|GetEnemy)
    (:then
      (Default|CallOnEnemyHealthPercentChanged (Building|GetBuildingHealthPercentage _enemy)))
    (:else)))

(fn HandleEnemySpawned
  (params (EnemyActor :type Actor%20Object%20Reference))
  ;[0 0.15 0.1308 0.5] HandleEnemySpawned: Invoked when the enemy boss spawns.
  (Networking|SwitchHasAuthority
    (:Authority
      (Utilities|IsValid EnemyActor
        (:Is%20Valid
          (Enemy|EnableEnemyRelocationDataLayer)
          (Utilities|FlowControl|Sequence
            (:then_0
              (bind _eventlocationactor (Variables|Event|GetEventLocationActor)
                (:then
                  (Default|UnbindEventfromOnEnemySpawned _eventlocationactor (EventDispatchers|CreateEvent :!event HandleEnemySpawned)))
                (:else)))
            (:then_1
              (bind _asbuilding_gameplay_actor (Utilities|Casting|CastToBuildingGameplayActor EnemyActor))
              (Variables|Transient|Enemy|SetEnemy _asbuilding_gameplay_actor)
              (Variables|Analytics|Transient|SetEnemySpawnedServerGameTime (Fortnite|Sync|Time|GetServerWorldTimeSeconds))
              (Actor|DestroyActor (Variables|Transient|GetLightningGroundRumbleActors)))))))))

(fn HandleGamePhaseChanged
  (params (GamePhase :type EAthenaGamePhase%20Enum))
  ;[0.0164 0.07935 0.1198] HandleGamePhaseChanged: Once we're in EndGame, submit the analytics.
  (Networking|SwitchHasAuthority
    (:Authority
      (if (Utilities|Enum|Equal(Enum) GamePhase "EndGame")
        (Analytics|SubmitAnalytics)
        (bind _gamephaseupdatedasynctask (Variables|Transient|GetGamePhaseUpdatedAsyncTask)
          (:then
            (GameplayEventRouter|StopListeningtoStatefulEvent _gamephaseupdatedasynctask))
          (:else))))))

(fn HandleGiantDamagedBuildingActor
  (params (DamagedActor :type Actor%20Object%20Reference) (Damage :type Float%20%28single-precision%29) (InstigatedBy :type Controller%20Object%20Reference) (DamageCauser :type Actor%20Object%20Reference) (HitLocation :type Vector) (FHitComponent :type Primitive%20Component%20Object%20Reference) (BoneName :type Name) (Momentum :type Vector)))

(fn HandleGiantPawnDamaged
  (params (DamagedActor :type Actor%20Object%20Reference) (Damage :type Float%20%28single-precision%29) (InstigatedBy :type Controller%20Object%20Reference) (DamageCauser :type Actor%20Object%20Reference) (HitLocation :type Vector) (FHitComponent :type Primitive%20Component%20Object%20Reference) (BoneName :type Name) (Momentum :type Vector))
  (Networking|SwitchHasAuthority
    (:Authority
      (TrackedPlayerDamage|UpdateDamageTakenbyGiantTracking DamagedActor InstigatedBy Damage))))

(fn HandleHoagieDestroyed
  (params (LastDamageInstigator :type Controller%20Object%20Reference) (VehicleActor :type Actor%20Object%20Reference))
  (bind _asfort_athena_vehicle (Utilities|Casting|CastToFortAthenaVehicle VehicleActor))
  ;[0.09859 0.02063 0.15 0.5] HandleHoagieDestroyed: Remove the hoagie vehicle from the list.
  (Networking|SwitchHasAuthority
    (:Authority
      (bind _returnvalue (Utilities|Array|RemoveItem (Variables|Transient|Hoagie|GetHoagieVehicles) _asfort_athena_vehicle))
      (if _returnvalue
        (Fortnite|FortShippingLog self (Utilities|String|Append "HandleHoagieDestroyed: " (Utilities|GetDisplayName _asfort_athena_vehicle)))))))

(fn HandleLocalViewTargetChanged
  (params (PlayerController :type Fort%20Player%20Controller%20Object%20Reference) (OldViewTarget :type Actor%20Object%20Reference) (NewViewTarget :type Actor%20Object%20Reference))
  (if (and (Utilities|Enum|Equal(Enum) (Networking|GetLocalRole NewViewTarget) "ROLE_SimulatedProxy") (Cinematics|Spectate|IsViewingPilotsorGiant NewViewTarget))
    (Cinematics|Spectate|SetSpectateTeleportCamera PlayerController true)))

(fn HandlePawnPossessedSinceGiantEvent
  (params (PossessedPawn :type Pawn%20Object%20Reference))
  (Fortnite|ApplyGameplayEffectsToActor PossessedPawn (Utilities|Array|MakeArray (Variables|PowerLevel|GetGE_CriticalDamageOverride)) (0.0 0.0 0.0) false 0 (Ability|GetFortAbilitySystemComponent PossessedPawn) (GameplayEffects|MakeFortEffectContext (Ability|GetFortAbilitySystemComponent PossessedPawn) (Giant|GetGiantPawn) (Giant|GetGiantPawn) 0 0)))

(fn HandlePlayerExit
  (params (PlayerPawn :type Fort%20Player%20Pawn%20Object%20Reference))
  (bind _humanpawn (LocalVariables|GetHumanPawn))
  (bind _returnvalue (Fortnite|GetFortPlayerControllerfromActor _humanpawn))
  (bind _humanpawn_2 (LocalVariables|GetHumanPawn))
  (bind _humanpawn_3 (LocalVariables|GetHumanPawn))
  (bind _humanpawn_4 (LocalVariables|GetHumanPawn))
  (bind _humanpawn_5 (LocalVariables|GetHumanPawn))
  ;[0.000121 0.15 0.117 0.5] Handle the exiting of the pilots (primary and secondary) from the giant experience.
  (Utilities|IsValid PlayerPawn
    (:Is%20Valid
      (LocalVariables|SetHumanPawn PlayerPawn)
      (Fortnite|FortShippingLog self (Utilities|String|Append "HandlePlayerExit: Handling exit for player " (Utilities|GetDisplayName PlayerPawn)))
      (Utilities|FlowControl|Sequence
        (:then_0
          (Utilities|FlowControl|Sequence
            (:then_0
              ;;[0.1187 0.03249 0.15 0.5] If the input player is a stand in player pawn, the giant, then use their replaced pawn. The replaced pawn is their actual HumanPawn.
              (bind _asfort_stand_in_player_pawn (Utilities|Casting|CastToFortStandInPlayerPawn (Variables|GetPlayerPawn)))
              (bind _asfort_player_pawn (Utilities|Casting|CastToFortPlayerPawn (Class|FortStandinPlayerPawn|GetReplacedPawn _asfort_stand_in_player_pawn)))
              (LocalVariables|SetHumanPawn _asfort_player_pawn))
            (:then_1
              ;;[1 1 1] Early return when the given pawn already exited
              (if (not (Utilities|Array|ContainsItem (Variables|Transient|Pilots|GetPilotActors) (LocalVariables|GetHumanPawn)))
                (return)))))
        (:then_1
          (if (Variables|Transient|GetbIntermittentPawnAttach)
            ;;[0.02404 0.02047 0.15 0.5] Detach the player pawn from the giant it is attached to.
            (Class|ToyFoxGiantPlayerPawn|DetachActorfromSelf (Variables|Transient|GetToyFoxGiantPawn) _humanpawn_3)
            (else
              (Transformation|DetachFromActor _humanpawn_3 "KeepWorld" "KeepWorld" "KeepRelative")))
          (Utilities|FlowControl|Sequence
            (:then_0
              ;;[0.15 0.1008 0.06004 0.5] Unbind delegates from human pawn.
              (EventDispatchers|UnbindEventfromOnDied (LocalVariables|GetHumanPawn) (EventDispatchers|CreateEvent :!event HandleSecondaryPilotDied))
              (Game|UnbindEventfromOnDestroyed (LocalVariables|GetHumanPawn) (EventDispatchers|CreateEvent :!event HandleSecondaryPilotDestroyed)))
            (:then_1
              ;;[0.06871 0.03344 0.15 0.5] Remove all the gameplay effects that have been added to this HumanPawn when they were a pilot.
              (bind _returnvalue_2 (Ability|GetFortAbilitySystemComponent (Variables|GetPlayerPawn)))
              (Utilities|IsValid _returnvalue_2
                (:Is%20Valid
                  (bind _returnvalue_3 (Utilities|LoadClassAssetBlocking (Variables|Settings|GetPrimaryPilotGameplayEffectClass)))
                  (GameplayEffects|RemoveActiveGameplayEffectbySourceEffect _returnvalue_2 (Utilities|Casting|CastToGameplayEffectClass _returnvalue_3) 0 -1)
                  (bind _returnvalue_4 (Utilities|LoadClassAssetBlocking (Variables|Settings|GetSecondaryPilotGameplayEffectClass)))
                  (GameplayEffects|RemoveActiveGameplayEffectbySourceEffect _returnvalue_2 (Utilities|Casting|CastToGameplayEffectClass _returnvalue_4) 0 -1)
                  (bind _returnvalue_5 (Utilities|LoadClassAssetBlocking (Variables|Settings|GetPilotStasisGameplayEffect)))
                  (GameplayEffects|RemoveActiveGameplayEffectbySourceEffect _returnvalue_2 (Utilities|Casting|CastToGameplayEffectClass _returnvalue_5) 0 -1)
                  (bind _returnvalue_6 (Utilities|LoadClassAssetBlocking (Variables|Settings|GetSecondaryPilotFinisherGameplayEffectClass))) ; Added through the giant finisher ability. Remove it here just in case.
                  (GameplayEffects|RemoveActiveGameplayEffectbySourceEffect _returnvalue_2 (Utilities|Casting|CastToGameplayEffectClass _returnvalue_6) 0 -1))))
            (:then_2
              ;;[0.1138 0.04025 0.15 0.5] Destroy their pilot pawn components.
              (Class|BToyFoxPilotPawnComponentParent|DestroyComponent (Actor|GetComponentsByClass (Variables|GetPlayerPawn) "/ToyFoxGameplay/Gameplay/PawnComp/B_ToyFoxPilotPawnComponent_Parent.B_ToyFoxPilotPawnComponent_Parent_C")))
            (:then_3
              ;;[0.02765 0.04911 0.15 0.5] Switch back from the giant to regular player pawn.
              (bind _returnvalue_7 (Fortnite|GetFortPlayerControllerfromActor (Variables|GetPlayerPawn)))
              (Utilities|IsValid _returnvalue_7
                (:Is%20Valid
                  (Class|FortGiantPawnFunctionLibrary|SwitchBackfromandDestroyGiant _returnvalue_7)
                  (Components|RemoveControllerComponent _returnvalue_7 "/ToyFoxGameplay/Gameplay/ControllerComp/B_ToyFoxControllerComp_Secondary.B_ToyFoxControllerComp_Secondary_C")
                  ;;[1 1 1] Force detach 2nd pilots from seeking missile
                  (Class|FortBikeKidControllingComponent|TrytoStopControllingBikeKid (Actor|GetComponentbyClass _returnvalue_7 "/Script/BikeKidRuntime.FortBikeKidControllingComponent") (Variables|BikeKid|GetDeferred%20Destroy%20Data)))))
            (:then_4
              ;;[0.15 0.04798 0.1306 0.5] Teleport player into the sky and enter skydive.
              (bind (_bvalid _location _rotation) (Pilots|GetPilotExitLocationandRotation _humanpawn_2))
              (if _bvalid
                (Transform|Actor|TeleportPlayerPawn _humanpawn_2 _location _rotation true) ; Put Player into Sky
                (else
                  _))
              (Parachute|BeginSkydiving _humanpawn_2 false)
              (Parachute|ForceOpenParachute _humanpawn_2 false)
              ;;[0.02476 0.0733 0.15 0.5] Max out the health and shields for the player pawns that exit being a pilot.
              (Fortnite|ApplyGameplayEffectsToActor _humanpawn_2 (Utilities|Array|MakeArray (Variables|Pilots|GetPilotExitHealthAndShieldGameplayEffectClass) (Variables|Pilots|GetExitRiftTeleportOutVisualsGameplayEffect)) (0.0 0.0 0.0) false 0 (Ability|GetFortAbilitySystemComponent _humanpawn_2) (GameplayEffects|MakeEffectContext (Ability|GetFortAbilitySystemComponent _humanpawn_2)))
              ;;[0.04124 0 0.15 0.5] Forces the camera to face the giant's location on exit.
              (if (> (ScalableFloat|GetCachedScalableFloatValue (Variables|Hotfix|GetHotfix_ExitPawnLookAtGiant)) 0)
                (Utilities|IsValid _returnvalue
                  (:Is%20Valid
                    (Game|AddComponentbyClass _returnvalue "/ToyFoxGameplay/Gameplay/ControllerComp/B_ToyFoxControllerComponent_ExitCamera.B_ToyFoxControllerComponent_ExitCamera_C" (Math|Rotator|FindLookatRotation (Transformation|GetActorLocation (LocalVariables|GetHumanPawn)) (Transformation|GetActorLocation (Giant|GetGiantPawn)))) ; Cleans self, no need to manually destory somewhere.
                  ))))))
        (:then_2
          (Utilities|FlowControl|Sequence
            (:then_0
              ;;[0.08774 0.03279 0.15 0.5] Replace the player with a stand in secondary pilot if the event is still active and the exiting player is a secondary pilot.
              (if (and (Utilities|Enum|Equal(Enum) (Event|GetEventState) "NewEnumerator3") (> (ScalableFloat|GetCachedScalableFloatValue (Variables|Hotfix|GetHotfix_ReplaceExitingSecondaryPilotWithStandIn)) 0))
                (if (Utilities|Array|ContainsItem (Variables|Transient|Pilots|GetSecondaryPilotPlayerPawns) _humanpawn_4)
                  (Pilots|ReplaceSecondaryPlayerPilotwithStandIn _humanpawn_4))))
            (:then_1
              ;;[0.03756 0.04109 0.15 0.5] Remove exiting players from array.
              (Utilities|Array|RemoveItem (Variables|Transient|Pilots|GetSecondaryPilotPlayerPawns) _humanpawn_5)
              (Utilities|Array|RemoveItem (Variables|Transient|Pilots|GetPilotActors) _humanpawn_5)
              (Variables|Transient|Pilots|SetPilotActors (Variables|Transient|Pilots|GetPilotActors)) ; Forces OnRep() call.
              ;;[1 1 1] Clear primary pilot when exited
              (if (== (LocalVariables|GetHumanPawn) (Variables|Transient|Pilots|GetPrimaryPilotPlayerPawn))
                (Variables|Transient|Pilots|SetPrimaryPilotPlayerPawn 0)))))))))

(fn HandlePlayerExitForIntro
  (params (ExitPawn :type Fort%20Player%20Pawn%20Object%20Reference))
  (bind _exitpawn (Variables|GetExitPawn))
  (bind _returnvalue (Fortnite|GetFortPlayerControllerfromActor _exitpawn))
  (bind (_returnvalue_x _returnvalue_y _returnvalue_z) (Math|Vector|BreakVector (Transformation|GetActorLocation (Variables|GetExitPawn))))
  (bind _exitpawn_2 (Variables|GetExitPawn))
  (bind _exitpawn_3 (Variables|GetExitPawn))
  (Fortnite|FortShippingLog self (Utilities|String|Append (Utilities|GetDisplayName (Variables|GetExitPawn)) " is exiting from the intro sequence now!"))
  (Utilities|IsValid ExitPawn
    (:Is%20Valid
      (if (Variables|Transient|GetbIntermittentPawnAttach)
        (Class|ToyFoxGiantPlayerPawn|DetachActorfromSelf (Variables|Transient|GetToyFoxGiantPawn) _exitpawn_2)
        (else
          (Transformation|DetachFromActor _exitpawn_2 "KeepWorld" "KeepWorld" "KeepRelative")))
      (Utilities|FlowControl|Sequence
        (:then_0
          ;[1 1 1] Remove controller component from exiting pawn
          (Utilities|IsValid _returnvalue
            (:Is%20Valid
              (Components|RemoveControllerComponent _returnvalue "/ToyFoxGameplay/Gameplay/ControllerComp/B_ToyFoxControllerComp_Secondary.B_ToyFoxControllerComp_Secondary_C"))))
        (:then_1
          ;[0.15 0.1008 0.06004 0.5] Unbind delegates from human pawn.
          (EventDispatchers|UnbindEventfromOnDied (Variables|GetExitPawn) (EventDispatchers|CreateEvent :!event HandleSecondaryPilotDied))
          (Game|UnbindEventfromOnDestroyed (Variables|GetExitPawn) (EventDispatchers|CreateEvent :!event HandleSecondaryPilotDestroyed)))
        (:then_2
          ;[0.06871 0.03344 0.15 0.5] Remove all the gameplay effects that have been added to this HumanPawn when they were a pilot.
          (bind _returnvalue_2 (Ability|GetFortAbilitySystemComponent (Variables|GetExitPawn)))
          (Utilities|IsValid _returnvalue_2
            (:Is%20Valid
              (bind _returnvalue_3 (Utilities|LoadClassAssetBlocking (Variables|Settings|GetPrimaryPilotGameplayEffectClass)))
              (bind _asgameplay_effect (Utilities|Casting|CastToGameplayEffectClass _returnvalue_3))
              (GameplayEffects|RemoveActiveGameplayEffectbySourceEffect _returnvalue_2 _asgameplay_effect 0 -1)
              (bind _returnvalue_4 (Utilities|LoadClassAssetBlocking (Variables|Settings|GetPilotStasisGameplayEffect)))
              (bind _asgameplay_effect_2 (Utilities|Casting|CastToGameplayEffectClass _returnvalue_4))
              (GameplayEffects|RemoveActiveGameplayEffectbySourceEffect _returnvalue_2 _asgameplay_effect_2 0 -1))))
        (:then_3
          ;[0.15 0.04798 0.1306 0.5] Teleport player into the sky and enter skydive.
          (Transform|Actor|TeleportPlayerPawn _exitpawn_3 (Math|Vector|MakeVector _returnvalue_x _returnvalue_y (+ _returnvalue_z (Class|ToyFoxGiantPlayerPawn|GetPawnExitTeleportExtraHeight (Variables|Transient|GetToyFoxGiantPawn)))) (Transformation|GetActorRotation _exitpawn_3) true false) ; Put Player into Sky
          (Parachute|BeginSkydiving _exitpawn_3 false)
          (Parachute|ForceOpenParachute _exitpawn_3 false))
        (:then_4
          ;[1 1 1] Spawn in the missing stand in pilot
          (if (> (ScalableFloat|GetCachedScalableFloatValue (Variables|Hotfix|GetHotfix_SpawnStandInPawnForPilotIntroExit)) 0)
            (bind _spawnedstandinpilot (Pilots|SpawnStandinPilotBGA))
            (Utilities|Array|Add (Variables|Transient|Pilots|GetPilotActors) _spawnedstandinpilot)
            (Utilities|Array|Add (Variables|Transient|Pilots|GetStandInPilots) _spawnedstandinpilot)))
        (:then_5
          (Utilities|Array|RemoveItem (Variables|Transient|Pilots|GetPilotActors) (Variables|GetExitPawn))
          (Variables|Transient|Pilots|SetPilotActors (Variables|Transient|Pilots|GetPilotActors))
          (if (== (Variables|Transient|Pilots|GetPrimaryPilotPlayerPawn) (Variables|GetExitPawn))
            (if (Utilities|Array|IsEmpty (Variables|Transient|Pilots|GetSecondaryPilotPlayerPawns))
              (Variables|Transient|Pilots|SetPrimaryPilotPlayerPawn 0)
              (else
                ;[1 1 1] Promote first 2nd pilot to be primary and apply missing GE
                (Variables|Transient|Pilots|SetPrimaryPilotPlayerPawn (Utilities|Array|Get(acopy) (Variables|Transient|Pilots|GetSecondaryPilotPlayerPawns)))
                (Utilities|Array|RemoveIndex (Variables|Transient|Pilots|GetSecondaryPilotPlayerPawns))
                (bind _returnvalue_5 (Utilities|LoadClassAssetBlocking (Variables|Settings|GetPrimaryPilotGameplayEffectClass))) ; In regular event flow, these will already be async loaded by this point.
                (bind _asgameplay_effect_3 (Utilities|Casting|CastToGameplayEffectClass _returnvalue_5))
                (Fortnite|ApplyGameplayEffectsToActor (Variables|Transient|Pilots|GetPrimaryPilotPlayerPawn) (Utilities|Array|MakeArray _asgameplay_effect_3) (0.0 0.0 0.0) false 0 (Ability|GetFortAbilitySystemComponent (Variables|Transient|Pilots|GetPrimaryPilotPlayerPawn)) (GameplayEffects|MakeFortEffectContext (Ability|GetFortAbilitySystemComponent (Variables|Transient|Pilots|GetPrimaryPilotPlayerPawn)) (Giant|GetGiantPawn) 0 0 1))))
            (else
              (Utilities|Array|RemoveItem (Variables|Transient|Pilots|GetSecondaryPilotPlayerPawns) (Variables|GetExitPawn)))))))))

(fn HandleSecondaryPilotDestroyed
  (params (DestroyedActor :type Actor%20Object%20Reference))
  ;[0.01545 0 0.15 0.5] HandleSecondaryPilotDestroyed: Secondary pilot has been destroyed. Replace them with a stand in pilot.
  (Networking|SwitchHasAuthority
    (:Authority
      (Fortnite|FortShippingLog self (Utilities|String|Append "HandleSecondaryPilotDestroyed: Invoked for " (Utilities|GetDisplayName DestroyedActor)))
      (Game|UnbindEventfromOnDestroyed DestroyedActor (EventDispatchers|CreateEvent :!event HandleSecondaryPilotDestroyed))
      (Utilities|FlowControl|Sequence
        (:then_0
          (if (> (ScalableFloat|GetCachedScalableFloatValue (Variables|Hotfix|GetHotfix_ReplaceDyingSecondaryPilotWithStandIn)) 0)
            (Pilots|ReplaceSecondaryPlayerPilotwithStandIn DestroyedActor)))
        (:then_1
          (Utilities|Array|RemoveItem (Variables|Transient|Pilots|GetPilotActors) DestroyedActor)
          (Variables|Transient|Pilots|SetPilotActors (Variables|Transient|Pilots|GetPilotActors)) ; Forces OnRep() call.
          (Utilities|Array|RemoveItem (Variables|Transient|Pilots|GetSecondaryPilotPlayerPawns) (Utilities|Casting|CastToFortPlayerPawn DestroyedActor)))))))

(fn HandleSecondaryPilotDied
  (params (DamagedActor :type Actor%20Object%20Reference) (Damage :type Float%20%28single-precision%29) (InstigatedBy :type Controller%20Object%20Reference) (DamageCauser :type Actor%20Object%20Reference) (HitLocation :type Vector) (FHitComponent :type Primitive%20Component%20Object%20Reference) (BoneName :type Name) (Momentum :type Vector))
  ;[0.01545 0 0.15 0.5] HandleSecondaryPilotDied: Secondary pilot has died. Replace them with a stand in pilot.
  (Networking|SwitchHasAuthority
    (:Authority
      (Fortnite|FortShippingLog self (Utilities|String|Append "HandleSecondaryPilotDied: Invoked for " (Utilities|GetDisplayName DamagedActor)))
      (bind _asfort_player_pawn (Utilities|Casting|CastToFortPlayerPawn DamagedActor))
      (EventDispatchers|UnbindEventfromOnDied _asfort_player_pawn (EventDispatchers|CreateEvent :!event HandleSecondaryPilotDied))
      (Utilities|FlowControl|Sequence
        (:then_0
          (if (> (ScalableFloat|GetCachedScalableFloatValue (Variables|Hotfix|GetHotfix_ReplaceDyingSecondaryPilotWithStandIn)) 0)
            (Pilots|ReplaceSecondaryPlayerPilotwithStandIn _asfort_player_pawn)))
        (:then_1
          (Utilities|Array|RemoveItem (Variables|Transient|Pilots|GetPilotActors) _asfort_player_pawn)
          (Variables|Transient|Pilots|SetPilotActors (Variables|Transient|Pilots|GetPilotActors)) ; Forces OnRep() call.
          (Utilities|Array|RemoveItem (Variables|Transient|Pilots|GetSecondaryPilotPlayerPawns) _asfort_player_pawn))))))

(fn HandleTrackedPlayerDamageDied
  (params (PawnDamageInfo :type Pawn%20Damage%20Data%20Structure%20%28by%20ref%29))
  (bind _playerdamagetrackinginfo (Variables|Event|GetPlayerDamageTrackingInfo))
  (Networking|SwitchHasAuthority
    (:Authority
      (bind _aspawn (Utilities|Casting|CastToPawn (Utilities|Struct|BreakPawnDamageData PawnDamageInfo)))
      (bind (_array_element _array_index) (Utilities|Array|ForEachLoopwithBreak _playerdamagetrackinginfo)
        (:LoopBody
          (if (== (Utilities|Struct|BreakToyFoxPlayerDamageStruct _array_element) (Variables|/Script/Engine.Pawn|GetPlayerState _aspawn))
            (LocalVariables|SetIndexToRemove _array_index)))
        (:Completed
          _))
      (Utilities|Array|RemoveIndex _playerdamagetrackinginfo (LocalVariables|GetIndexToRemove))
      (TrackedPlayerDamage|TrytoUpdateDamageTrackingInfo))))

(fn HasPilotCinematicCameraViewCached
  (return (Variables|Transient|GetbPilotCinematicCameraViewCached)))

(fn IncreasePowerLevel
  (params (Amouint :type Float%20%28double-precision%29))
  ;[0 0.15 0.1398 0.5] Increase the power level by the desired amount.
  (Networking|SwitchHasAuthority
    (:Authority
      (Utilities|IsValid (Giant|GetGiantPawn)
        (:Is%20Valid
          (if (Utilities|Enum|Equal(Enum) (Event|GetEventState) "NewEnumerator3")
            (Variables|PowerLevel|SetPowerLevel (Math|Float|Clamp(Float) (+ (PowerLevel|GetPowerLevel) Amouint)))))))))

(fn IsEnemyActive
  (bind _enemy (Variables|Transient|Enemy|GetEnemy))
  (if (Utilities|IsValid _enemy :!class /Script/Engine.KismetSystemLibrary)
    (return (and (GameplayTags|DoesContainerMatchTagQuery (GameplayTags|GetOwnedGameplayTags _enemy :!class /Script/GameplayTags.BlueprintGameplayTagLibrary) (Variables|Enemy|GetEnemySpawnInGameplayTagQuery)) (not (Variables|Transient|GetbFinisherExecuted))))
    (else
      (return false))))

(fn IsEnemyStunned
  (if (Enemy|IsEnemyActive)
    (return (GameplayTags|DoesContainerMatchTagQuery (GameplayTags|GetOwnedGameplayTags (Variables|Transient|Enemy|GetEnemy) :!class /Script/GameplayTags.BlueprintGameplayTagLibrary) (Variables|Enemy|GetEnemyStunnedGameplayTagQuery)))
    (else
      (return false))))

(fn IsPlayerActiveRewardGoalBoon
  (params (Player :type Fort%20Player%20State%20Athena%20Object%20Reference))
  (bind _damage (TrackedPlayerDamage|GetDamageDonetoGiantbyPlayer Player))
  (return (< _damage (Reward|GetRewardBoonThresholdDamage))))

(fn IsPlayerActiveRewardGoalMedallion
  (params (Player :type Fort%20Player%20State%20Athena%20Object%20Reference))
  (bind _returnvalue (Reward|IsPlayerActiveRewardGoalBoon Player))
  (return (not _returnvalue)))

(fn IsViewingPilotsOrGiant
  (params (ViewTarget :type Actor%20Object%20Reference))
  (if (== ViewTarget (Variables|Transient|GetToyFoxGiantPawn))
    _
    (elif (>= (Utilities|Array|FindItem (Variables|Transient|Pilots|GetPilotActors) ViewTarget) 0)
      _
      (else
        (return false))))
  (return true))

(fn LocalDisableSpectateTeleportCamera
  (for _array_element (Fortnite|GetLocalFortPlayerControllers true true)
    (bind _output_get (LocalVariables|SetThisLocalController _array_element))
    (EventDispatchers|UnbindEventfromOnViewTargetChanged _output_get (EventDispatchers|CreateEvent :!event HandleLocalViewTargetChanged))
    (Cinematics|Spectate|SetSpectateTeleportCamera _output_get)))

(fn MarkPilotCinematicCameraActorViewCacheUsed
  (Variables|Transient|SetbPilotCinematicCameraViewCached false))

(fn NotifyCinematicGameplayEvent_GiantRise
  (bind _pawn (Giant|GetGiantPawn))
  ;[0.009428 0.125 0 0.5] NotifyCinematicGameplayEvent_GiantRise: Called from the sequencer to let the giant pawn know that it is time to rise from the kneel animation.
  (Networking|SwitchHasAuthority
    (:Authority
      (Fortnite|FortShippingLog self (Utilities|String|Append "NotifyCinematicGameplayEvent_GiantRise: Invoked: Sending Gameplay Event to Actor " (Utilities|GetDisplayName _pawn)) (not (Utilities|IsValid _pawn :!class /Script/Engine.KismetSystemLibrary)))
      (Ability|SendGameplayEventtoActor _pawn "(TagName=\"ToyFox.Event.Ability.Intro.Kneel.Rise\")"))))

(fn NotifyCinematicGameplayEvent_SpawnEnemy
  ;[0.03524 0.09188 0.15 0.5] NotifyCinematicGameplayEvent_GiantRise: Called from the sequencer to spawn the enemy.
  (Networking|SwitchHasAuthority
    (:Authority
      (if (ScalableFloat|GetValueasBool (Variables|Hotfix|GetHotfix_EventBossShouldSpawn))
        (Fortnite|FortShippingLog self "NotifyCinematicGameplayEvent_SpawnEnemy: Invoked")
        (Giant|SpawnEnemyBoss)))))

(fn NotifyGiantFinisherExecuted
  ;[0.01782 0.1299 0.15 0.5] NotifyGiantFinisherExecuted: Update the analytics bool.
  ; Called from GA_ToyFox_Finisher.
  (Networking|SwitchHasAuthority
    (:Authority
      (Variables|Analytics|SetbGiantPawnBeatEnemyBoss true) ; Analytics
      (Variables|Transient|SetbFinisherExecuted true))))

(fn NotifyGiantWeakpointCreated
  ;[0.1065 0 0.15 0.5] NotifyGiantWeakpointCreated: Track the number of weakpoints created on the giant.
  ; Called directly from B_ToyFoxWeakpointComponent.
  (Networking|SwitchHasAuthority
    (:Authority
      (Math|Integer|IncrementInt (Variables|Analytics|GetGiantWeakpointsCreatedCount)) ; Analytics
    )))

(fn NotifySecondaryPilotAbilityExit
  ;[0.01997 0.04031 0.15 0.5] NotifySecondaryPilotAbilityExit: Track the number of secondary pilots that used their exit ability to leave the megazord experience.
  ; Called from GA_ToyFox_SecondaryPlayer_Exit.
  (Networking|SwitchHasAuthority
    (:Authority
      (Math|Integer|IncrementInt (Variables|Analytics|GetSecondaryPilotAbilityExitCount)) ; Analytics
    )))

(fn OnRep_bEnableSpectateTeleportCamera
  (bind _thislocalcontroller (LocalVariables|GetThisLocalController))
  (bind _returnvalue (Pawn|GetViewTarget _thislocalcontroller))
  (for _array_element (Fortnite|GetLocalFortPlayerControllers true true)
    (LocalVariables|SetThisLocalController _array_element)
    (if (Variables|PilotCinematic|Spectate|GetbEnableSpectateTeleportCamera)
      (EventDispatchers|BindEventtoOnViewTargetChanged (LocalVariables|GetThisLocalController) (EventDispatchers|CreateEvent :!event HandleLocalViewTargetChanged))
      (if (Utilities|Enum|Equal(Enum) (Networking|GetLocalRole _returnvalue) "ROLE_SimulatedProxy")
        (if (Cinematics|Spectate|IsViewingPilotsorGiant _returnvalue)
          (Cinematics|Spectate|SetSpectateTeleportCamera (LocalVariables|GetThisLocalController) true))))))

(fn OnRep_bFinisherExecuted
  (if (Variables|Transient|GetbFinisherExecuted)
    (Default|CallOnFinisherExecuted)))

(fn OnRep_Enemy
  (bind _enemy (Variables|Transient|Enemy|GetEnemy)
    (:then
      (Health|BindEventtoOnBuildingHealthChanged _enemy (EventDispatchers|CreateEvent :!event HandleEnemyHealthChanged))
      (Game|BindEventtoOnDestroyed _enemy (EventDispatchers|CreateEvent :!event HandleEnemyDestroyed))
      (Utilities|IsValid (Ability|GetFortAbilitySystemComponent _enemy)
        (:Is%20Valid
          (CallFunction|HandleEnemyAbilitySystemComponentInitialized))
        (:Is%20Not%20Valid
          (EventDispatchers|BindEventtoOnAbilitySystemComponentInitialized _enemy (EventDispatchers|CreateEvent :!event HandleEnemyAbilitySystemComponentInitialized)))))
    (:else)))

(fn OnRep_EventState
  (bind _state (Event|GetEventState))
  ;[0.08019 0 0.15 0.5] OnRep_EventState: Called on both server and client.
  ; Fire delegate and run the desired function based on the event state.
  (Fortnite|FortShippingLog self (Utilities|String|Append "OnRep_EventState: State set to " (Utilities|String|EnumtoString (Event|GetEventState))))
  (Default|CallOnEventStateChanged _state)
  (switch Utilities|FlowControl|Switch|SwitchonToyFoxEventStateEnum _state
    (:NewEnumerator5
      (State|EnterStateEventIntro))
    (:NewEnumerator1
      (State|EnterStateGiantIntro))
    (:NewEnumerator2
      (State|EnterStateGiantAvailable))
    (:NewEnumerator3
      (State|EnterStateGiantPossessed))
    (:NewEnumerator4
      (State|EnterStateEnd))))

(fn OnRep_GiantDamageMedallionRewardedPlayerState
  (Default|CallOnGiantDamagedMedallionRewardedPlayerSet (Variables|Transient|Reward|GetGiantDamageMedallionRewardedPlayerState)))

(fn OnRep_GiantPossessingPlayerState
  (Default|CallOnGiantPossessedbyPlayerState (Variables|Transient|GetGiantPossessingPlayerState)))

(fn OnRep_PedestalInteractionInfo
  (Default|CallOnPedestalInteractionInfoUpdated (Pedestal|GetPedestalInteractionInfo)))

(fn OnRep_PilotActors
  (Default|CallOnPilotActorsUpdated))

(fn OnRep_PowerLevel
  (bind _return (PowerLevel|GetPowerLevel))
  (Default|CallOnGiantPowerLevelChanged _return)
  (Utilities|FlowControl|Sequence
    (:then_0
      (bind _updated_variable (Utilities|FlowControl|DeltaGateWildcard (Variables|Transient|GetbGiantPowerLevelLow) (<= (PowerLevel|GetPowerLevel) (ScalableFloat|GetCachedScalableFloatValue (Variables|Hotfix|GetHotfix_LowPowerLevelThreshold))))
        (:Success
          (bind _output_get (Variables|Transient|SetbGiantPowerLevelLow _updated_variable))
          (Default|CallOnGiantPowerLevelLowStateChanged _output_get)
          (Fortnite|FortShippingLog self (Utilities|String|Append "OnRep_PowerLevel: Low State Changed to " (Utilities|String|ToString(Boolean) _output_get))))))
    (:then_1
      (Networking|SwitchHasAuthority
        (:Authority
          (if (<= _return 0)
            (Event|SetEventState "NewEnumerator4") ; Should have event state changed before Outro starts
            (bind _returnvalue (Utilities|LoadClassAssetBlocking (Variables|Settings|GetGiantOutroGameplayEffectClass))) ; Should already be async loaded by this point.
            (Fortnite|ApplyGameplayEffectsToActor (Giant|GetGiantPawn) (Utilities|Array|MakeArray (Utilities|Casting|CastToGameplayEffectClass _returnvalue)) (0.0 0.0 0.0) false 0 (Ability|GetAbilitySystemComponent (Giant|GetGiantPawn)) (GameplayEffects|MakeEffectContext (Ability|GetAbilitySystemComponent (Giant|GetGiantPawn))))))))))

(fn OnRep_ToyFoxGiantPawn
  (Networking|CanExecuteCosmeticEvents
    (:True
      (bind _toyfoxgiantpawn (Variables|Transient|GetToyFoxGiantPawn)
        (:then
          (CallFunction|SpawnMusicComponentforLocalPlayerControllers)
          (Default|CallOnGiantPawnSpawned _toyfoxgiantpawn))
        (:else)))))

(fn PlayPilotTeleport
  (Networking|SwitchHasAuthority
    (:Authority
      (for _array_element (Pilots|GetPilotPlayerPawns)
        (bind _returnvalue (Utilities|LoadClassAssetBlocking (Variables|Settings|GetPilotTeleportGameplayEffectClass))) ; Should be loaded by this time.
        (Fortnite|ApplyGameplayEffectsToActor _array_element (Utilities|Array|MakeArray (Utilities|Casting|CastToGameplayEffectClass _returnvalue)) (0.0 0.0 0.0) false 0 (Ability|GetAbilitySystemComponent _array_element) "()")))))

(fn PossessGiantAndEnablePilotControls
  (bind _pawn (Giant|GetGiantPawn))
  ;[0.15 0 0.04488 0.5] PossessGiantAndEnablePilotControls: Possess the giant player pawn and provides pilot controls for the pilot players.
  (Networking|SwitchHasAuthority
    (:Authority
      (Utilities|IsValid _pawn
        (:Is%20Valid
          (bind _primarypilotplayerpawn (Variables|Transient|Pilots|GetPrimaryPilotPlayerPawn)
            (:then
              (Utilities|FlowControl|Sequence
                (:then_0
                  (Utilities|IsValid (Fortnite|GetFortPlayerControllerfromActor _primarypilotplayerpawn)
                    (:Is%20Valid
                      (bind _returnvalue (Utilities|LoadClassAssetBlocking (Variables|Settings|GetPilotStasisGameplayEffect)))
                      (GameplayEffects|RemoveActiveGameplayEffectbySourceEffect (Ability|GetFortAbilitySystemComponent _primarypilotplayerpawn) (Utilities|Casting|CastToGameplayEffectClass _returnvalue) 0 -1) ; Remove stasis before possession.
                      (Class|ToyFoxGiantPlayerPawn|PrepareforPlayerPossession _pawn)
                      (bind _returnvalue_2 (Class|FortGiantPawnFunctionLibrary|SwitchtoGiant _pawn (Fortnite|GetFortPlayerControllerfromActor _primarypilotplayerpawn)))
                      (bind _returnvalue_3 (Utilities|LoadClassAssetBlocking (Variables|Settings|GetPrimaryPilotGameplayEffectClass)))
                      (Fortnite|ApplyGameplayEffectsToActor _returnvalue_2 (Utilities|Array|MakeArray (Utilities|Casting|CastToGameplayEffectClass _returnvalue_3)) (0.0 0.0 0.0) false 0 (Ability|GetFortAbilitySystemComponent _primarypilotplayerpawn) (GameplayEffects|MakeFortEffectContext (Ability|GetFortAbilitySystemComponent _primarypilotplayerpawn) (Giant|GetGiantPawn) 0 0 1)) ; Apply primary pilot GE after pawn switch\nSo that GE won't be cleared out on possession change
                    )))
                (:then_1
                  ;;[1 1 1] Add primary pilot pawn component.
                  ;; Must add component after the possession is finished
                  (Game|AddComponentbyClass (Giant|GetGiantPawn) (Variables|Settings|GetPrimaryPilotPawnComponent))
                  (for _array_element (Variables|Transient|Pilots|GetSecondaryPilotPlayerPawns)
                    ;;[1 1 1] Add secondary pilot pawn component.
                    (Utilities|IsValid _array_element
                      (:Is%20Valid
                        (Game|AddComponentbyClass _array_element (Variables|Settings|GetSecondaryPilotPawnComponent))))))))
            (:else)))))))

(fn PrintDebugStringsToScreen
  (bind _eventstate (Variables|Transient|GetEventState))
  ;[0.03521 0.01203 0.05729 0.5] PrintDebugStringsToScreen: Used in PIE to print debug strings on the screen.
  (if (Editor|InEditororPIE)
    (Networking|SwitchHasAuthority
      (:Authority
        (Development|PrintString (Utilities|String|Append "ToyFox:EventState:" (Utilities|String|EnumtoString _eventstate)) true false "(R=0.000000,G=1.000000,B=0.114623,A=1.000000)" 5.0 "ToyFox:EventState:")
        (if (Utilities|Enum|Equal(Enum) _eventstate "NewEnumerator3")
          (Development|PrintString (Utilities|String|Append "ToyFox:GiantPowerLevel:" (Utilities|String|ToString(Float) (PowerLevel|GetPowerLevel))) true false "(R=0.000000,G=1.000000,B=0.114623,A=1.000000)" 5.0 "ToyFox:GiantPowerLevel"))))))

(fn ReducePowerLevel
  (params (Reduction :type Float%20%28double-precision%29))
  (bind _pawn (Giant|GetGiantPawn))
  (Networking|SwitchHasAuthority
    (:Authority
      (Utilities|IsValid _pawn
        (:Is%20Valid
          (if (not (GameplayTags|HasMatchingGameplayTag _pawn "(TagName=\"ToyFox.PowerLevel.ReductionImmune\")"))
            _))
        (:Is%20Not%20Valid
          _))
      (if (Utilities|Enum|Equal(Enum) (Event|GetEventState) "NewEnumerator3")
        (Variables|PowerLevel|SetPowerLevel (Math|Float|Clamp(Float) (- (Variables|PowerLevel|GetPowerLevel) Reduction)))))))

(fn ReducePowerLevelByDamage
  (params (Damage :type Float%20%28double-precision%29) (DamageTags :type Gameplay%20Tag%20Container%20Structure%20%28by%20ref%29) (Instigator :type Controller%20Object%20Reference))
  ;[0.1193 0 0.15 0.5] ReducePowerLevelByDamage: Reduces the power level based on damage taken. Storm damage does not go through DDR and is directly applied.
  (Networking|SwitchHasAuthority
    (:Authority
      (Variables|Analytics|SetTotalDamageTakenByGiant (+ Damage (Variables|Analytics|GetTotalDamageTakenByGiant))) ; Analytics
      (if (GameplayTags|HasTag DamageTags "(TagName=\"Gameplay.Damage.OutsideSafeZone\")" true)
        (bind _output_get (LocalVariables|SetReduction (Utilities|Operators|Multiply Damage 0.0001 (ScalableFloat|GetCachedScalableFloatValue (Variables|Hotfix|GetHotfix_StormDamageMultiplier)))))
        (PowerLevel|ReducePowerLevel _output_get)
        (Variables|Analytics|SetGiantPowerLevelReductionFractionStormDamage (+ (Variables|Analytics|GetGiantPowerLevelReductionFractionStormDamage) (LocalVariables|GetReduction))) ; Analytics
        (else
          (bind _output_get_2 (LocalVariables|SetDiminishedDamage (* (* Damage 0.001) (Utilities|RemapFloat (Variables|Transient|GetDamageDiminishingReturns) (Math|Vector2D|MakeVector2D 0.0 (ScalableFloat|GetCachedScalableFloatValue (Variables|Hotfix|GetHotfix_DDRMax))) (Math|Vector2D|MakeVector2D (ScalableFloat|GetCachedScalableFloatValue (Variables|Hotfix|GetHotfix_DDRMultiplierForMinDamage)) (ScalableFloat|GetCachedScalableFloatValue (Variables|Hotfix|GetHotfix_DDRMultiplierForMaxDamage))) true))))
          (Variables|Transient|SetDamageDiminishingReturns (+ _output_get_2 (Variables|Transient|GetDamageDiminishingReturns)))
          (bind _output_get_3 (LocalVariables|SetReduction (* (LocalVariables|GetDiminishedDamage) (ScalableFloat|GetCachedScalableFloatValue (Variables|Hotfix|GetHotfix_PowerLevelReductionDamageMultiplier)))))
          (PowerLevel|ReducePowerLevel _output_get_3)
          (if (GameplayTags|HasTag (Variables|GetDamageTags) "(TagName=\"Gameplay.Damage.TidalCrane.ToyFox.Boss\")")
            (Variables|Analytics|SetGiantPowerLevelReductionFractionEnemyBossDamage (+ (Variables|Analytics|GetGiantPowerLevelReductionFractionEnemyBossDamage) (LocalVariables|GetReduction))) ; Analytics.
            (else
              (Utilities|Casting|CastToFortPlayerControllerAthena (Variables|GetInstigator))
              (Variables|Analytics|SetGiantPowerLevelReductionFractionPlayerDamage (+ (Variables|Analytics|GetGiantPowerLevelReductionFractionPlayerDamage) (LocalVariables|GetReduction))) ; Analytics
            )))))))

(fn ReducePowerLevelPassive
  ;[0 0.08321 0.15 0.5] ReducePowerLevelPassive: Passively reducing power level on a timer.
  (Networking|SwitchHasAuthority
    (:Authority
      (bind _returnvalue (ScalableFloat|GetValueatLevel (Variables|Hotfix|GetHotfix_PassivePowerLevelDrop) (Math|Conversions|ToFloat(Integer) (Math|Float|Floor (Math|Float|Max(Float) (Math|Float|SafeDivide (- (Fortnite|Sync|Time|GetServerWorldTimeSeconds) (Variables|Analytics|Transient|GetGiantPawnPossessStartServerGameTime)) (ScalableFloat|GetCachedScalableFloatValue (Variables|Hotfix|GetHotfix_PassivePowerLevelDropIncreaseIncrementTime))))))))
      (PowerLevel|ReducePowerLevel _returnvalue)
      (Variables|Analytics|SetGiantPowerLevelReductionFractionPassive (+ _returnvalue (Variables|Analytics|GetGiantPowerLevelReductionFractionPassive))) ; Analytics
      (Variables|Transient|SetDamageDiminishingReturns (Math|Float|Clamp(Float) (- (Variables|Transient|GetDamageDiminishingReturns) (ScalableFloat|GetCachedScalableFloatValue (Variables|Hotfix|GetHotfix_DDRRecoveryRate))) 0.0 (ScalableFloat|GetCachedScalableFloatValue (Variables|Hotfix|GetHotfix_DDRMax)))))))

(fn ReducePowerLevelWeakpointDestroy
  (params (WeakpointDamage :type Float%20%28double-precision%29))
  (Utilities|SwitchHasAuthority
    (:Authority
      (bind _output_get (LocalVariables|SetReduction (Utilities|Operators|Multiply WeakpointDamage 0.001 (ScalableFloat|GetCachedScalableFloatValue (Variables|Hotfix|GetHotfix_PowerLevelReductionWeakPointDestroyMulti)))))
      (PowerLevel|ReducePowerLevel _output_get)
      (Variables|Analytics|SetGiantPowerLevelReductionFractionWeakpointDestroy (+ (Variables|Analytics|GetGiantPowerLevelReductionFractionWeakpointDestroy) (LocalVariables|GetReduction))))))

(fn RegisterEventLocationActor
  (params (EventLocationActor :type B%20Toy%20Fox%20Event%20Location%20Object%20Reference))
  ;[0.02619 0.15 0.1147 0.5] RegisterEventLocationActor: Register the location actor. It will eventually be used with EQS to determine the spawn location of the event per match.
  (Networking|SwitchHasAuthority
    (:Authority
      (Utilities|IsValid EventLocationActor
        (:Is%20Valid
          (Utilities|Array|Add (Variables|Event|GetRegisteredEventLocations) EventLocationActor)
          (Fortnite|FortShippingLog self (Utilities|String|Append "RegisterEventLocationActor: Event location registered: " (Class|BToyFoxEventLocation|GetEventLocationName EventLocationActor))))))))

(fn RegisterHoagieVehicle
  (params (HoagieVehicle :type Fort%20Athena%20Vehicle%20Object%20Reference))
  ;[0 0.003374 0.15 0.5] RegisterHoagie: Track the hoagie vehicle and bind to know when it is destroyed.
  ; Called from the B_ToyFox_HoagieRift_Spawner.
  ; DO NOT DIRECTLY CAST TO THE VEHICLE HERE.
  (Networking|SwitchHasAuthority
    (:Authority
      (Utilities|IsValid HoagieVehicle
        (:Is%20Valid
          (Utilities|Array|Add (Variables|Transient|Hoagie|GetHoagieVehicles) HoagieVehicle)
          (Vehicle|Destruction|BindEventtoOnVehicleExplodeDelegate HoagieVehicle (EventDispatchers|CreateEvent :!event HandleHoagieDestroyed)))))))

(fn ReplaceSecondaryPlayerPilotWithStandIn
  (params (SecondaryPlayerPawn :type Actor%20Object%20Reference))
  (bind _state (Event|GetEventState))
  (bind _returnvalue (Utilities|GetDisplayName SecondaryPlayerPawn))
  (bind _pilotactors (Variables|Transient|Pilots|GetPilotActors))
  ;[0.03799 0.08375 0.15 0.5] ReplaceSecondaryPlayerPilotWithStandIn: Secondary player pawn has died, exited, or destroyed. Replace it with a stand in BGA.
  (Networking|SwitchHasAuthority
    (:Authority
      (Utilities|IsValid SecondaryPlayerPawn
        (:Is%20Valid
          (Fortnite|FortShippingLog self (Utilities|String|Append "ReplaceSecondaryPlayerPilotWithStandIn: Invoked with player pawn " _returnvalue))
          (if (Utilities|Enum|Equal(Enum) _state "NewEnumerator3")
            (bind _returnvalue_2 (Utilities|Array|FindItem _pilotactors SecondaryPlayerPawn))
            (if (>= _returnvalue_2 0)
              (bind _spawnedstandinpilot (Pilots|SpawnStandinPilotBGA))
              (Utilities|IsValid _spawnedstandinpilot
                (:Is%20Valid
                  (Utilities|Array|SetArrayElem _pilotactors _returnvalue_2 _spawnedstandinpilot true)
                  (Utilities|Array|Add (Variables|Transient|Pilots|GetStandInPilots) _spawnedstandinpilot)
                  (Variables|Transient|Pilots|SetPilotActors _pilotactors) ; Invoked to call OnRep() on the server.
                  (Fortnite|FortShippingLog self (Utilities|String|Append "ReplaceSecondaryPlayerPilotWithStandIn: Successfully replaced player " _returnvalue " with stand in " (Utilities|GetDisplayName _spawnedstandinpilot))))
                (:Is%20Not%20Valid
                  (Fortnite|FortShippingLog self "ReplaceSecondaryPlayerPilotWithStandIn: Failed to spawn StandIn." true)))
              (else
                (Fortnite|FortShippingLog self "ReplaceSecondaryPlayerPilotWithStandIn: Could not find input player as a pilot." true)))
            (else
              (Fortnite|FortShippingLog self (Utilities|String|Append "ReplaceSecondaryPlayerPilotWithStandIn: Ignoring request since event state is NOT GiantPossessed: " (Utilities|String|EnumtoString _state))))))))))

(fn ResetObjective
  (|CallParentFunction)
  (Variables|Transient|SetbObjectiveActive false)
  (Event|SetEventState "NewEnumerator0")
  (Variables|PowerLevel|SetPowerLevel 1.0)
  (Variables|Transient|SetbGiantPowerLevelLow false)
  (Utilities|Array|Clear (Variables|Transient|Pilots|GetSecondaryPilotPlayerPawns))
  (Utilities|Array|Clear (Variables|Transient|Pilots|GetPilotActors)))

(fn RewardGiantDamageBoonToPlayer
  (params (PlayerPawn :type Fort%20Player%20Pawn%20Object%20Reference))
  ;[0.1124 0.05176 0.15 0.5] RewardGiantDamageBoonToPlayer: Function to reward boon to the players that have dealt threshold damage to the giant.
  (Utilities|FlowControl|Sequence
    (:then_0
      (Networking|SwitchHasAuthority
        (:Authority
          (Utilities|IsValid PlayerPawn
            (:Is%20Valid
              (bind _asfort_player_state_athena (Utilities|Casting|CastToFortPlayerStateAthena (Variables|/Script/Engine.Pawn|GetPlayerState PlayerPawn)))
              (bind _returnvalue (Utilities|Array|AddUnique (Variables|Transient|Reward|GetGiantDamageBoonRewardedPlayers) _asfort_player_state_athena))
              (if (>= _returnvalue 0)
                (bind _returnvalue_2 (Inventory|SpawnPickupInWorld (Variables|Reward|GetGiantDamageRewardItemBoon) 1 (Math|Transform|TransformLocation (Transformation|GetActorTransform PlayerPawn) (Variables|Reward|GetRewardPickupSpawnRelativeOffset)) "0.000000,0.000000,1.000000" 0 true true false -1 "Tossed" "Unset" 0 false))
                (Pickup|BlueprintSetPickupTarget _returnvalue_2 PlayerPawn true)
                (Fortnite|FortShippingLog self (Utilities|String|Append "RewardBoonToPlayer: Rewarding boon " (Utilities|GetDisplayName (Variables|Reward|GetGiantDamageRewardItemBoon)) " to player " (Fortnite|GetPlayerNameSafe _asfort_player_state_athena)))
                (return true)))))))
    (:then_1
      (return false))))

(fn RewardGiantDamageMedallion
  ;[0.01746 0.15 0.09744 0.5] RewardGiantDamageMedallion: Providing reward medallion to the player that has done the highest damage to the giant..
  (Networking|SwitchHasAuthority
    (:Authority
      ;;[1 1 1] We aren't bypassing logic to get here (non-BR)
      (if (not (ScalableFloat|GetValueasBool (Variables|Hotfix|GetHotfix_EventBypassCheck)))
        (Utilities|FlowControl|DeltaGateWildcard (Variables|Transient|Reward|GetbGiantDamageMedallionRewarded) true
          (:Success
            (bind _damageinfo (TrackedPlayerDamage|GetHighestDamageDonetoGiantbyAnyPlayer (Pilots|GetPilotPlayerStates)))
            (bind _asfort_player_state_athena (Utilities|Casting|CastToFortPlayerStateAthena (Utilities|Struct|BreakToyFoxPlayerDamageStruct _damageinfo)))
            (bind _output_get (Variables|Transient|Reward|SetGiantDamageMedallionRewardedPlayerState _asfort_player_state_athena))
            (Utilities|IsValid (PlayerState|GetCurrentPawn _output_get)
              (:Is%20Valid
                (bind _returnvalue (Inventory|SpawnPickupInWorld (Variables|Reward|GetGiantDamageRewardItemMedallion) 1 (Math|Transform|TransformLocation (Transformation|GetActorTransform (PlayerState|GetCurrentPawn _output_get)) (Variables|Reward|GetRewardPickupSpawnRelativeOffset)) "0.000000,0.000000,1.000000" 0 true true false -1 "Tossed" "Unset" 0 false))
                (Pickup|BlueprintSetPickupTarget _returnvalue (PlayerState|GetCurrentPawn _output_get) true)
                (Fortnite|FortShippingLog self (Utilities|String|Append "GiveRewardMedallion: Rewarding medallion " (Utilities|GetDisplayName (Variables|Reward|GetGiantDamageRewardItemMedallion)) " to player " (Fortnite|GetPlayerNameSafe _asfort_player_state_athena)))
                (CustomVerb|BroadcastCustomVerbMessage (CustomVerb|MakeCustomVerbMessage (PlayerState|GetCurrentPawn _output_get) "(TagName=\"CustomVerb.ToyFox.HighestToyFoxDamager\")"))))))))))

(fn RewardMilitaryRankThresholdPoints
  (params (In%20Pawn :type Fort%20Player%20Pawn%20Object%20Reference))
  (bind _in_pawn (Variables|GetIn%20Pawn))
  (bind _returnvalue (Team|GetActorTeam _in_pawn))
  (bind _array_element (Utilities|Array|ForEachLoopwithBreak (Variables|Transient|Reward|GetGiantDamageBoonRewardedPlayers))
    (:LoopBody
      (if (== (Class|FortPlayerStateAthena|GetTeam _array_element) _returnvalue)
        (LocalVariables|SetHasTeamAlreadyEarnedPoints (== (Class|FortPlayerStateAthena|GetTeam _array_element) _returnvalue))))
    (:Completed
      _))
  (if (not (LocalVariables|GetHasTeamAlreadyEarnedPoints))
    (CustomVerb|BroadcastCustomVerbMessage (CustomVerb|MakeCustomVerbMessage (Variables|GetIn%20Pawn) "(TagName=\"CustomVerb.ToyFox.DamagedToyfox\")"))))

(fn SetEnableCinematicCameraHandlerComponents
  (params (bEnable :type Boolean))
  (Networking|SwitchHasAuthority
    (:Authority
      (if bEnable
        (for _array_element (Pilots|GetPilotPlayerStates)
          (Utilities|IsValid _array_element
            (:Is%20Valid
              (bind _returnvalue (Game|AddComponentbyClass (Fortnite|GetFortPlayerControllerfromActor _array_element) "/ToyFoxGameplay/Gameplay/PilotCinematic/CameraHandlers/B_ToyFoxPilotCinematicControllerComponent_CameraHandler.B_ToyFoxPilotCinematicControllerComponent_CameraHandler_C" (Transformation|GetActorRotation (Giant|GetGiantPawn)) (Transformation|GetActorLocation (Giant|GetGiantPawn))))
              (Utilities|Array|Add (Variables|Transient|PilotCinematic|GetPilotCinematicCameraHandlerControllerComponents) _returnvalue))))
        (else
          (Class|BToyFoxPilotCinematicControllerComponentCameraHandler|DestroyComponent (Variables|Transient|PilotCinematic|GetPilotCinematicCameraHandlerControllerComponents)))))))

(fn SetEnableCinematicHideHUDComponents
  (params (bEnable :type Boolean))
  ;[0 0.15 0.03744 0.5] SetEnableCinematicHideHUDComponents: Adding a controller component to hide the HUD elements for the pilot players during the cinematic. This component lives on the players from the time they interacted with the pedestal to the time they are given controller of the giant.
  (Networking|SwitchHasAuthority
    (:Authority
      (Fortnite|FortShippingLog self (Utilities|String|Append "SetEnableCinematicHideHUDComponents: " (Utilities|String|ToString(Boolean) bEnable)))
      (if bEnable
        (for _array_element (Pilots|GetPilotPlayerStates)
          (Utilities|IsValid _array_element
            (:Is%20Valid
              (bind _returnvalue (Game|AddComponentbyClass (Fortnite|GetFortPlayerControllerfromActor _array_element) "/ToyFoxGameplay/Gameplay/PilotCinematic/B_ToyFoxPilotCinematicControllerComponent_HideHUD.B_ToyFoxPilotCinematicControllerComponent_HideHUD_C"))
              (Utilities|Array|Add (Variables|Transient|PilotCinematic|GetPilotCinematicHideHUDControllerComponents) _returnvalue))))
        (else
          (Class|BToyFoxPilotCinematicControllerComponentHideHUD|DestroyComponent (Variables|Transient|PilotCinematic|GetPilotCinematicHideHUDControllerComponents)))))))

(fn SetEventState
  (params (InEventState :type ToyFoxEventStateEnum%20Enum :default "NewEnumerator0"))
  ;[0.01596 0.01421 0.15 0.5] SetEventState: Update the event state variable.
  (Networking|SwitchHasAuthority
    (:Authority
      (if (Utilities|Enum|NotEqual(Enum) InEventState (Variables|Transient|GetEventState))
        (Variables|Transient|SetEventState InEventState)))))

(fn SetPedestalInteractionInfo
  (params (Info :type Toy%20Fox%20Pedestal%20Interaction%20Info%20Struct%20Structure))
  (Networking|SwitchHasAuthority
    (:Authority
      (Variables|Transient|SetPedestalInteractionInfo Info))))

(fn SetPilotCinematicCameraActorView
  (params (ViewInfo :type Minimal%20View%20Info%20Structure))
  (Variables|Transient|SetPilotCinematicCameraView ViewInfo)
  (Variables|Transient|SetbPilotCinematicCameraViewCached true))

(fn SetSecondaryPilot3PView
  (params (bEnabled :type Boolean))
  (Networking|SwitchHasAuthority
    (:Authority
      (bind _returnvalue (Utilities|LoadClassAssetBlocking (Variables|Settings|GetSecondaryPilotFinisherGameplayEffectClass)))
      (bind _asgameplay_effect (Utilities|Casting|CastToGameplayEffectClass _returnvalue))
      (LocalVariables|SetGEClass _asgameplay_effect)
      (if (Variables|GetbEnabled)
        (for _array_element (Variables|Transient|Pilots|GetSecondaryPilotPlayerPawns)
          (Utilities|IsValid _array_element
            (:Is%20Valid
              (Class|FortBikeKidControllingComponent|TrytoStopControllingBikeKid (Actor|GetComponentbyClass (Pawn|GetController _array_element) "/Script/BikeKidRuntime.FortBikeKidControllingComponent") (Variables|BikeKid|GetDeferred%20Destroy%20Data))
              (bind _returnvalue_2 (Ability|GetFortAbilitySystemComponent _array_element))
              (Utilities|IsValid _returnvalue_2
                (:Is%20Valid
                  (GameplayEffects|ApplyGameplayEffectToSelf :self _returnvalue_2 :GameplayEffectClass (LocalVariables|GetGEClass) :Level 0.0))))))
        (else
          (for _array_element_2 (Variables|Transient|Pilots|GetSecondaryPilotPlayerPawns)
            (bind _returnvalue_3 (Ability|GetFortAbilitySystemComponent _array_element_2))
            (Utilities|IsValid _returnvalue_3
              (:Is%20Valid
                (GameplayEffects|RemoveActiveGameplayEffectbySourceEffect _returnvalue_3 (LocalVariables|GetGEClass) 0 -1)))))))))

(fn SetSpectateTeleportCamera
  (params (PlayerController :type Player%20Controller%20Object%20Reference) (bEnabled :type Boolean))
  (bind _returnvalue (Game|Player|FindCameraModifierbyClass (Variables|/Script/Engine.PlayerController|GetPlayerCameraManager PlayerController) (Variables|PilotCinematic|Spectate|GetSpectateTeleportCameraModifier)))
  (Utilities|IsValid _returnvalue
    (:Is%20Valid
      (bind _output_get (LocalVariables|SetCameraWasEnabled (not (or (CameraModifier|IsDisabled _returnvalue) (CameraModifier|IsPendingDisable _returnvalue)))))
      (if (!= _output_get (Variables|GetbEnabled))
        (Utilities|FlowControl|Sequence
          (:then_0
            (if (Variables|GetbEnabled)
              (CameraModifier|EnableModifier _returnvalue)
              (else
                (CameraModifier|DisableModifier _returnvalue true))))
          (:then_1
            _))))
    (:Is%20Not%20Valid
      (if (Variables|GetbEnabled)
        (Game|Player|AddNewCameraModifier (Variables|/Script/Engine.PlayerController|GetPlayerCameraManager (Variables|GetPlayerController)) (Variables|PilotCinematic|Spectate|GetSpectateTeleportCameraModifier)))))
  (bind _asfort_player_camera_base (Utilities|Casting|CastToFortPlayerCameraBase (Variables|/Script/Engine.PlayerController|GetPlayerCameraManager (Variables|GetPlayerController))))
  (if (Variables|GetbEnabled)
    (Camera|AddCameraModeOverride _asfort_player_camera_base (Variables|PilotCinematic|Spectate|GetSpectateTeleportCameraMode) (Variables|PilotCinematic|Spectate|GetSpectateTeleportCameraTag) 0 0 0)
    (else
      (Camera|RemoveCameraModeOverride _asfort_player_camera_base (Variables|PilotCinematic|Spectate|GetSpectateTeleportCameraTag) 0 false))))

(fn SetupStandInPilots
  (bind _secondarypilotplayerpawns (Variables|Transient|Pilots|GetSecondaryPilotPlayerPawns))
  ;[0.04799 0 0.15 0.5] SetupStandInPilots: Primary and secondary pilots are expected to be decided at this point. Based on that information, spawn the desired number of stand in pilots.
  (Utilities|FlowControl|Sequence
    (:then_0
      (Networking|SwitchHasAuthority
        (:Authority
          (Variables|Transient|Pilots|GetPrimaryPilotPlayerPawn
            (:then
              (for _index (range (Math|Float|Truncate (ScalableFloat|GetCachedScalableFloatValue (Variables|Hotfix|GetHotfix_SecondaryPilotMaxCountUsedForStandInPilotSpawn))))
                (if (Utilities|Array|IsValidIndex _secondarypilotplayerpawns _index)
                  (Utilities|IsValid (Utilities|Array|Get(acopy) _secondarypilotplayerpawns _index)
                    (:Is%20Valid
                      (EventDispatchers|BindEventtoOnDied (Utilities|Array|Get(acopy) _secondarypilotplayerpawns _index) (EventDispatchers|CreateEvent :!event HandleSecondaryPilotDied))
                      (Game|BindEventtoOnDestroyed (Utilities|Array|Get(acopy) _secondarypilotplayerpawns _index) (EventDispatchers|CreateEvent :!event HandleSecondaryPilotDestroyed)))
                    (:Is%20Not%20Valid
                      _))
                  (else
                    _))
                (bind _spawnedstandinpilot (Pilots|SpawnStandinPilotBGA))
                (Utilities|IsValid _spawnedstandinpilot
                  (:Is%20Valid
                    (Utilities|Array|Add (Variables|Transient|Pilots|GetStandInPilots) _spawnedstandinpilot)))))
            (:else)))))
    (:then_1
      (return))))

(fn SpawnEnemyBoss
  ;[0 0.15 0.006933 0.5] SpawnEnemyBoss: Spawn the enemy boss through selected the event location actor.
  (Networking|SwitchHasAuthority
    (:Authority
      (if (Variables|Transient|Enemy|GetbSpawnEnemyInvoked)
        (bind _eventlocationactor (Variables|Event|GetEventLocationActor)
          (:then
            (Default|BindEventtoOnEnemySpawned _eventlocationactor (EventDispatchers|CreateEvent :!event HandleEnemySpawned))
            (Class|BToyFoxEventLocation|SpawnEnemy _eventlocationactor)
            (Variables|Transient|Enemy|SetbSpawnEnemyInvoked true)
            (Fortnite|FortShippingLog self "SpawnEnemyBoss: Successfully invoked."))
          (:else
            ;;[1 0.0434 0 0.5] Warning: EventLocationActor is null.
            (Fortnite|FortShippingLog self "SpawnEnemyBoss: EventLocationActor is nullptr." true)))))))

(fn SpawnStandInPilotBGA
  (bind (_bannericonid _bannercolorid _skydivecontrail _glider _pickaxe _character) (Utilities|Struct|BreakFortAthenaLoadout (Cosmetics|GetPawnCosmeticLoadout (Variables|Transient|Pilots|GetPrimaryPilotPlayerPawn) false true)))
  (bind _pawn (Giant|GetGiantPawn))
  (bind _pawn_2 (Giant|GetGiantPawn))
  (bind _availablestandinpilotscharacteritemdefinitions (Variables|Pilots|GetAvailableStandInPilotsCharacterItemDefinitions))
  ;[0.121 0 0.15 0.5] Spawn secondary stand in pilot with the desired character item definition.
  (Utilities|FlowControl|Sequence
    (:then_0
      (Networking|SwitchHasAuthority
        (:Authority
          (Utilities|IsValid _pawn_2
            (:Is%20Valid
              (bind _primarypilotplayerpawn (Variables|Transient|Pilots|GetPrimaryPilotPlayerPawn)
                (:then
                  (Utilities|FlowControl|Sequence
                    (:then_0
                      ;;[0 0.15 0.1082 0.5] First, cache all the cosmetic character item definitions that have already been used by the existing pilots.
                      (Utilities|FlowControl|Sequence
                        (:then_0
                          (Utilities|Array|Add (LocalVariables|GetUsedCharacterItemDefinitions) _character)
                          (for _array_element (Variables|Transient|Pilots|GetSecondaryPilotPlayerPawns)
                            (bind _asfort_player_pawn (Utilities|Casting|CastToFortPlayerPawn _array_element))
                            (Utilities|Array|AddUnique (LocalVariables|GetUsedCharacterItemDefinitions) (.Character (Utilities|Struct|BreakFortAthenaLoadout (Cosmetics|GetPawnCosmeticLoadout _asfort_player_pawn false true)))))
                          (for _array_element_2 (Variables|Transient|Pilots|GetStandInPilots)
                            (Utilities|IsValid _array_element_2
                              (:Is%20Valid
                                (bind _characteritemdef (Class|BPISecondaryPilotStandIn|GetCharacterItemDef(Message) _array_element_2))
                                (Utilities|Array|AddUnique (LocalVariables|GetUsedCharacterItemDefinitions) _characteritemdef))))
                          (for _array_element_3 (LocalVariables|GetUsedCharacterItemDefinitions)
                            (Utilities|IsValid _array_element_3
                              (:Is%20Valid
                                (Fortnite|FortShippingLog self (Utilities|String|Append "SpawnStandInPilotBGA: Used character item: " (Utilities|GetDisplayName _array_element_3)))))))
                        (:then_1
                          ;;[0.1187 0.01581 0.15 0.5] Select a character item definition that hasn't been used yet.
                          (bind _output_get (LocalVariables|SetAvailableRandomizedCharacterItemDefinitions _availablestandinpilotscharacteritemdefinitions))
                          (Utilities|Array|Shuffle _output_get)
                          (bind _array_element_4 (Utilities|Array|ForEachLoopwithBreak _output_get)
                            (:LoopBody
                              (Utilities|IsValid _array_element_4
                                (:Is%20Valid
                                  (if (not (Utilities|Array|ContainsItem (LocalVariables|GetUsedCharacterItemDefinitions) _array_element_4))
                                    (LocalVariables|SetSelectedCharacterItemDefinition _array_element_4)))))
                            (:Completed
                              _))
                          (LocalVariables|GetSelectedCharacterItemDefinition
                            (:then
                              _)
                            (:else
                              (bind _output_get_2 (LocalVariables|SetSelectedCharacterItemDefinition (Utilities|Array|RandomArrayItem _availablestandinpilotscharacteritemdefinitions)))))
                          (Fortnite|FortShippingLog self (Utilities|String|Append "SpawnStandInPilotBGA: Selected character item def: " (Utilities|GetDisplayName _output_get_2))))))
                    (:then_1
                      ;;[0.006332 0.15 0.1276 0.5] Spawn secondary pilot and set it up.
                      (bind _returnvalue (Utilities|LoadClassAssetBlocking (Variables|Settings|GetStandInPilotClass))) ; Should already be async loaded by this point.
                      (bind _asbuilding_gameplay_actor (Utilities|Casting|CastToBuildingGameplayActorClass _returnvalue))
                      (bind _returnvalue_2 (Building|SpawnBuildingGameplayActorfromWorldContext _asbuilding_gameplay_actor (Math|Transform|MakeTransform (Transformation|GetActorLocation _pawn_2)) _primarypilotplayerpawn 0 true))
                      (Utilities|IsValid _returnvalue_2
                        (:Is%20Valid
                          (Fortnite|FortShippingLog self (Utilities|String|Append "SpawnStandInPilotBGA: Successfully spawned: " (Utilities|GetDisplayName _returnvalue_2)))
                          (bind _asbpi_secondary_pilot_stand_in (Utilities|Casting|CastToBPI_SecondaryPilotStandIn _returnvalue_2)
                            (:then
                              (Class|BPISecondaryPilotStandIn|SetCharacterItemDef _asbpi_secondary_pilot_stand_in (LocalVariables|GetSelectedCharacterItemDefinition)))
                            (:CastFailed
                              _))
                          (Building|InitializeKismetSpawnedBuildingActor _returnvalue_2 0 0 false 0 false)
                          (if (Variables|Transient|GetbIntermittentPawnAttach)
                            (Class|ToyFoxGiantPlayerPawn|AttachActortoSelf _pawn _returnvalue_2)
                            (else
                              (Transformation|AttachActorToActor _returnvalue_2 _pawn "None" "SnapToTarget" "KeepRelative" "KeepRelative" true)))
                          (return _returnvalue_2))))))
                (:else)))))))
    (:then_1
      (Fortnite|FortShippingLog self "SpawnStandInPilotBGA: Failed to spawn BGA." true)
      (return 0))))

(fn SubmitAnalytics
  ;[0.03754 0.005307 0.04167] SubmitAnalytics: Fire analytics if it hasn't been done yet.
  (Networking|SwitchHasAuthority
    (:Authority
      (if (Math|Boolean|ANDBoolean (not (Replay|GetIsPlayingReplay)) (not (Variables|Analytics|Transient|GetbAnalyticsSubmitted)) (Utilities|Enum|NotEqual(Enum) (Event|GetEventState) "NewEnumerator0"))
        (Variables|Analytics|Transient|SetbAnalyticsSubmitted true)
        (Utilities|FlowControl|Sequence
          (:then_0
            (Analytics|AddAnalyticsAttribute "EventLocation" (Class|BToyFoxEventLocation|GetEventLocationName (Variables|Event|GetEventLocationActor)))
            (Analytics|AddAnalyticsAttribute :AttributeName "GiantPawnLifeTime" :IntValue -1 :floatValue (Variables|Analytics|GetGiantPawnLifeTime) :BoolValue false)
            (Analytics|AddAnalyticsAttribute :AttributeName "TotalDamageTakenByGiant" :IntValue -1 :floatValue (Variables|Analytics|GetTotalDamageTakenByGiant) :BoolValue false)
            (Analytics|AddAnalyticsAttribute :AttributeName "EnemyBossLifeTime" :IntValue -1 :floatValue (Variables|Analytics|GetEnemyBossLifeTime) :BoolValue false)
            (Analytics|AddAnalyticsAttribute :AttributeName "GiantPawnBeatEnemyBoss" :IntValue -1 :floatValue -1.0 :BoolValue (Variables|Analytics|GetbGiantPawnBeatEnemyBoss))
            (Analytics|AddAnalyticsAttribute :AttributeName "SecondaryPilotAbilityExitCount" :IntValue (Math|Integer|Max(Integer) (Variables|Analytics|GetSecondaryPilotAbilityExitCount)) :floatValue -1.0 :BoolValue false)
            (Analytics|AddAnalyticsAttribute :AttributeName "GiantAvailableTime" :IntValue -1 :floatValue (Variables|Analytics|GetGiantAvailableTime) :BoolValue false)
            (Analytics|AddAnalyticsAttribute :AttributeName "GiantWeakpointsCreatedCount" :IntValue (Variables|Analytics|GetGiantWeakpointsCreatedCount) :floatValue -1.0 :BoolValue false)
            (Analytics|AddAnalyticsAttribute :AttributeName "GiantPowerLevelReductionFractionStormDamage" :IntValue -1 :floatValue (Variables|Analytics|GetGiantPowerLevelReductionFractionStormDamage) :BoolValue false)
            (Analytics|AddAnalyticsAttribute :AttributeName "GiantPowerLevelReductionFractionPlayerDamage" :IntValue -1 :floatValue (Variables|Analytics|GetGiantPowerLevelReductionFractionPlayerDamage) :BoolValue false)
            (Analytics|AddAnalyticsAttribute :AttributeName "GiantPowerLevelReductionFractionPassive" :IntValue -1 :floatValue (Variables|Analytics|GetGiantPowerLevelReductionFractionPassive) :BoolValue false)
            (Analytics|AddAnalyticsAttribute :AttributeName "GiantPowerLevelReductionFractionEnemyBossDamage" :IntValue -1 :floatValue (Variables|Analytics|GetGiantPowerLevelReductionFractionEnemyBossDamage) :BoolValue false)
            (Analytics|AddAnalyticsAttribute :AttributeName "GiantPowerLevelReductionFractionWeakpointDestroy" :IntValue -1 :floatValue (Variables|Analytics|GetGiantPowerLevelReductionFractionWeakpointDestroy) :BoolValue false))
          (:then_1
            (Analytics|FireAnalyticsEventBlueprintArray "Core.ToyFox.Event" (Variables|Analytics|Transient|GetAnalyticsEventAttributes))))))))

(fn TryToUpdateDamageTrackingInfo
  (if (not (Utilities|Time|IsTimerActivebyHandle (Variables|Transient|GetDamageTrackingUpdateTimer)))
    (bind _returnvalue (Utilities|Time|SetTimerbyEvent (EventDispatchers|CreateEvent :!event UpdatePlayerDamageTrackingInfo) (ScalableFloat|GetCachedScalableFloatValue (Variables|Hotfix|GetHotfix_UpdatePlayerDamageTrackingInterval))))
    (Variables|Transient|SetDamageTrackingUpdateTimer _returnvalue)))

(fn UpdateDamageTakenByGiantTracking
  (params (DamagedActor :type Actor%20Object%20Reference) (Instigator :type Controller%20Object%20Reference) (Damage :type Float%20%28double-precision%29))
  (bind _outputdelegate (EventDispatchers|CreateEvent :!event HandleTrackedPlayerDamageDied))
  (bind _cachedtotaldamage (LocalVariables|GetCachedTotalDamage))
  (bind _returnvalue (+ _cachedtotaldamage Damage))
  (Networking|SwitchHasAuthority
    (:Authority
      (if (!= (Fortnite|GetFortPlayerControllerfromActor DamagedActor) Instigator)
        (Utilities|IsValid Instigator
          (:Is%20Valid
            (bind _asfort_player_state_athena (Utilities|Casting|CastToFortPlayerStateAthena (Variables|/Script/Engine.Controller|GetPlayerState Instigator)))
            (if (and (not (Class|FortPlayerStateZone|IsPlayerDead _asfort_player_state_athena)) (Class|FortPlayerStateAthena|IsAnAthenaGameParticipant _asfort_player_state_athena))
              (Utilities|FlowControl|Sequence
                (:then_0
                  (bind (_array_element _array_index) (Utilities|Array|ForEachLoopwithBreak (Variables|Event|GetPlayerDamageTrackingInfo))
                    (:LoopBody
                      (if (== (Utilities|Struct|BreakToyFoxPlayerDamageStruct _array_element) _asfort_player_state_athena)
                        (LocalVariables|SetFoundIndex _array_index)
                        (LocalVariables|SetCachedTotalDamage (.TotalDamage_5_F91677274558D42AB1D4FFBE0B7EFC33 (Utilities|Struct|BreakToyFoxPlayerDamageStruct _array_element)))))
                    (:Completed
                      _))
                  (Utilities|Array|RemoveIndex (Variables|Event|GetPlayerDamageTrackingInfo) (LocalVariables|GetFoundIndex))
                  (EventDispatchers|UnbindEventfromOnPawnDiedDelegate _asfort_player_state_athena _outputdelegate)
                  (EventDispatchers|BindEventtoOnPawnDiedDelegate _asfort_player_state_athena _outputdelegate)
                  (Utilities|Array|Add (Variables|Event|GetPlayerDamageTrackingInfo) (Utilities|Struct|MakeToyFoxPlayerDamageStruct _asfort_player_state_athena _returnvalue))
                  (if (>= _returnvalue (Reward|GetRewardBoonThresholdDamage))
                    (bind _brewarded (Reward|RewardGiantDamageBoontoPlayer (PlayerState|GetCurrentPawn _asfort_player_state_athena)))
                    (if _brewarded
                      (CallFunction|BroadcastAlertNotification 0 (Utilities|Struct|MakeAlertNotificationStruct "(TagName=\"ToyFox.HUD.Alert.Rita.Reward.Boon\")" (Utilities|Array|MakeArray (Utilities|Casting|CastToFortPlayerStateAthena (Variables|/Script/Engine.Pawn|GetPlayerState (PlayerState|GetCurrentPawn _asfort_player_state_athena))))))
                      (Reward|RewardMilitaryRankThresholdPoints (PlayerState|GetCurrentPawn _asfort_player_state_athena)))))
                (:then_1
                  (TrackedPlayerDamage|TrytoUpdateDamageTrackingInfo))))))))))
