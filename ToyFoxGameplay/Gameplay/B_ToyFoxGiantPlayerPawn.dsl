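; Player pawn Blueprint for the giant ToyFox boss, handling its combat, movement and cosmetic state.
; Damage drains game-state power level and drives weakpoints and stagger; power changes update low-power effects and mesh materials.
; Anim-notify gameplay events trigger local footstep cues and damage-trace abilities, while server-side overlap timers knock back actors.
; Workaround timers re-equip the default weapon and retry the client-activated movement mode extension if it stops.
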(event EventBeginPlay
  (bind _returnvalue (Footsteps|GetFootstepComponent))
  (bind _self self)
  (bind _returnvalue_2 (Ability|GetFortAbilitySystemComponent _self))
  (bind _returnvalue_3 (Actor|GetComponentbyClass "/Game/Sounds/SoundLibrary/Contexts/HitNotify/BP_HitNotifySoundLibraryComponent_FortPawn.BP_HitNotifySoundLibraryComponent_FortPawn_C"))
  (bind _returnvalue_4 (WorldSubsystems|GetFortGlobalActorRegistrySubsystem))
  (bind _enemyactorregistrygameplaytag (Variables|Enemy|GetEnemyActorRegistryGameplayTag))
  (bind _mesh (Variables|Character|GetMesh))
  ;[0 0.15 0.03082 0.5] BeginPlay: General setup of the actor after spawning.
  (|CallParentFunction)
  (Class|FortAsyncActionCurrentPlaylistReady|PlaylistDataReadyAsync
    (:then
      ;;[1 1 1] Initialize weakpoint components array for code
      (Variables|ToyFoxGiantPlayerPawn|SetWeakpointComponents (Utilities|Array|MakeArray (Variables|Default|GetWeakpointComp_Back) (Variables|Default|GetWeakpointComp_Arm_R) (Variables|Default|GetWeakpointComp_LowerLeg_R) (Variables|Default|GetWeakpointComp_LowerLeg_L) (Variables|Default|GetWeakpointComp_Shoulder_R) (Variables|Default|GetWeakpointComp_Shoulder_L) (Variables|Default|GetWeakpointComp_Arm_L) (Variables|Default|GetWeakpointComp_Chest))))
    (:OnReady
      (Utilities|FlowControl|Sequence
        (:then_0
          ;;[1 1 1] When first spawned, update the collision based on the latest variables that control it.
          (Intro|RefreshIntroCollisionState)
          ;;[0.08098 0.1013 0.15 0.5] Bind to know when ability system component is initialized.
          (EventDispatchers|AssignOnAbilitySystemComponentInitialized (EventDispatchers|CreateEvent :!event Custom|HandleAbilitySystemComponentInitialized))
          (Utilities|IsValid _returnvalue_2
            (:Is%20Valid
              (CallFunction|HandleAbilitySystemComponentInitialized _returnvalue_2 _self))
            (:Is%20Not%20Valid
              _))
          (Networking|SwitchHasAuthority
            (:Authority
              ;;[0.15 0.108 0.05969 0.5] Bind to know when the event state is changing.
              (bind _component (Default|GetGameStateComponent)
                (:Valid
                  (Default|BindEventtoOnEventStateChanged _component (EventDispatchers|CreateEvent :!event HandleEventStateChanged))
                  (Event|HandleEventStateChanged (Event|GetEventState _component))
                  ;;[0.124 0.008355 0.15 0.5] Logic to detect the queen bugs (enemy) in the world. The queens are building actors so the perception system does not work.
                  (bind _outactors (Class|FortGlobalActorRegistrySubsystem|GetRegisteredActors _returnvalue_4 _enemyactorregistrygameplaytag))
                  (for _array_element _outactors
                    (Utilities|IsValid _array_element
                      (:Is%20Valid
                        (Enemy|HandleEnemyActorRegistered _enemyactorregistrygameplaytag _array_element))))
                  (bind _returnvalue_5 (Class|FortGlobalActorRegistrySubsystem|GetActorRegistryDataforGameplayTag _returnvalue_4 _enemyactorregistrygameplaytag))
                  (Utilities|IsValid _returnvalue_5
                    (:Is%20Valid
                      (EventDispatchers|BindEventtoOnActorRegistered _returnvalue_5 (EventDispatchers|CreateEvent :!event HandleEnemyActorRegistered))
                      (EventDispatchers|BindEventtoOnActorUnregistered _returnvalue_5 (EventDispatchers|CreateEvent :!event HandleEnemyActorUnregistered)))))))))
        (:then_1
          (Networking|CanExecuteCosmeticEvents
            (:True
              (CallFunction|RequestForceAnimTickDuration 7.0) ; Request force anim tick in case the gameplay cue \nresponsible for this during the intro sequence isn't replicated yet.\nWe can only request this for 7 seconds at a time\nHardcoded value.
              ;;[0.09395 0.08169 0.15 0.5] Unfortunately, there doesn't seem to be a way to override the hit notify sound component since it is being added through the GameFeatureData for all FortPawns. So for now, we're manually destroying the component and adding the toyfox specific version.
              (Utilities|IsValid _returnvalue_3
                (:Is%20Valid
                  _)
                (:Is%20Not%20Valid
                  (Utilities|FlowControl|DelayUntilNextTick)))
              (Components|DestroyComponent _returnvalue_3)
              (Components|AddDynamicComponentbyClass "/ToyFoxGameplay/Assets/Sound/BP_HitNotifySoundLibraryComponent_ToyFox.BP_HitNotifySoundLibraryComponent_ToyFox_C"))))
        (:then_2
          ;;[0.04241 0.1472 0.15 0.5] General settings for the giant sized player pawn.
          (CallFunction|SetEnableRenderCustomDepth false)
          (CallFunction|OverrideMovementNoiseGeneration (Utilities|Struct|MakeFortPlayerPawnAthenaMovementNoiseOverride (ScalableFloat|GetValueatLevel (Variables|Hotfix|GetHotfix_AIMinFootstepNoise)) (ScalableFloat|GetValueatLevel (Variables|Hotfix|GetHotfix_AIMaxFootstepNoise))))
          (Anim|SetNeverSkipAnimTick true)
          ;;[0.1185 0.01179 0.15 0.5] Workaround to get the giant pawn to equip the right weapon.
          (Utilities|Time|SetTimerbyEvent (EventDispatchers|CreateEvent :!event Custom|TryEquipDefaultWeapon) 3.0 true true)
          (Utilities|FlowControl|Sequence
            (:then_0
              ;;[0.08433 0.02196 0.15 0.5] Actor starts hidden so that it isn't visible for a brief duration before the intro GA replicates.
              (Utilities|FlowControl|Delay (Variables|Settings|GetSetActorVisibleWaitDelay))
              (Rendering|SetActorHiddenInGame false))
            (:then_1
              ;;[0.01263 0.02412 0.15 0.5] Increase the max footstep distance to account for the giant size.
              (Utilities|IsValid _returnvalue
                (:Is%20Valid
                  (Variables|/Script/FortniteGame.FortFootstepComponent|SetMaxFootstepDistance (ScalableFloat|GetValueatLevel (Variables|Hotfix|GetHotfix_TickAnimOffscreenDistanceForFootsteps)) _returnvalue))))
            (:then_2
              ;;[0.04632 0 0.15 0.5] Bind to know when the power level of the giant changes.
              (bind _component_2 (Default|GetGameStateComponent)
                (:Valid
                  (Default|BindEventtoOnGiantPowerLevelChanged _component_2 (EventDispatchers|CreateEvent :!event Custom|HandleGiantPowerLevelChanged))
                  (CallFunction|HandleGiantPowerLevelChanged (PowerLevel|GetPowerLevel _component_2))
                  ;;[0.01225 0.02429 0.15 0.5] Bind to know when the power level low state changes.
                  ;; Update if the giant power level is currently low or high.
                  (Default|BindEventtoOnGiantPowerLevelLowStateChanged _component_2 (EventDispatchers|CreateEvent :!event Custom|HandleGiantPowerLevelLowStateChanged))
                  (CallFunction|HandleGiantPowerLevelLowStateChanged (PowerLevel|GetPowerLevelLow _component_2)))))))
        (:then_3
          ;;[1 1 1] Check if we should reset collision to follow root mesh component
          (if (<= (ScalableFloat|GetCachedScalableFloatValue (Variables|Hotfix|GetHotfix_CollideWithActualMesh)) 0)
            (Collision|SetCollisionEnabled (Variables|Default|GetToyFoxSkeletalMeshComponentBudgeted) "NoCollision")
            (Collision|SetCollisionEnabled _mesh "QueryAndPhysics")
            (Components|SkinnedMesh|SetPhysicsAsset _mesh "/ToyFoxGameplay/Assets/Meshes/ToyFox/M_LRG_ToyFox_PhysicsAsset.M_LRG_ToyFox_PhysicsAsset" true) ; Set the appropriate physics asset for collision.
            (Variables|Default|SetbCollideWithActualMesh false)))))))

(event EventReceiveControllerChanged
  (params (OldController :type Controller%20Object%20Reference) (NewController :type Controller%20Object%20Reference))
  (bind _returnvalue (Utilities|GetDisplayName NewController))
  (bind _self self)
  ;[0.02925 0.15 0.04148 0.5] ReceiveControllerChanged: When the controller is updated, activate the giant movement mode extension.
  (|CallParentFunction OldController NewController)
  (Networking|CanExecuteCosmeticEvents
    (:True
      (Utilities|IsValid NewController
        (:Is%20Valid
          (if (Pawn|IsLocalController NewController)
            (bind _returnvalue_2 (Class|FortMovementModeExtBlueprintLibrary|ActivateRegisteredMovementModeExtension (Variables|Settings|GetToyFoxMovementModeExt) self))
            (Utilities|IsValid _returnvalue_2
              (:Is%20Valid
                _)
              (:Is%20Not%20Valid
                (Fortnite|FortShippingLog _self (Utilities|String|Append "ReceiveControllerChanged: Failed to set MME for controller: " _returnvalue) true)
                (Utilities|FlowControl|DelayUntilNextTick) ; Try again soon since this player pawn's \nnet role could not have changed yet.
              ))
            (Fortnite|FortShippingLog _self (Utilities|String|Append "ReceiveControllerChanged: Successfully set MME for controller: " _returnvalue))))))))

(event Custom|HandleAbilitySystemComponentInitialized
  (params (AbilitySystemComponent :type Fort%20Ability%20System%20Component%20Object%20Reference) (AffectedPawn :type Fort%20Player%20Pawn%20Object%20Reference))
  (bind _asyncaction_footstepevents (Variables|Transient|GetAsyncAction_FootstepEvents))
  (bind _hotfix_retrymmeapplicationonremovalinterval (Variables|Hotfix|GetHotfix_RetryMMEApplicationOnRemovalInterval))
  (bind _leftfootstepanimnotifytag (Variables|Footsteps|GetLeftFootstepAnimNotifyTag))
  (bind _rightfootstepanimnotifytag (Variables|Footsteps|GetRightFootstepAnimNotifyTag))
  ;[0.1035 0.01953 0.15 0.5] HandleAbilitySystemComponentInitialized: On the client, retry application of MME in case it ends at unexpected time. MME ending breaks the event totally. There's native logging to track down the reason if this ever happens but this implementation here is to help handle a failcase.
  ; On the server and client, handle the footstep behavior.
  (Utilities|IsValid AbilitySystemComponent
    (:Is%20Valid
      (Utilities|IsValid AffectedPawn
        (:Is%20Valid
          (Utilities|FlowControl|Sequence
            (:then_0
              (Networking|CanExecuteCosmeticEvents
                (:True
                  (if (Pawn|IsLocallyControlled AffectedPawn)
                    ;;[0 0.09744 0.15 0.5] Unbind actions tracking any previous ability system component.
                    (Utilities|Time|ClearandInvalidateTimerbyHandle (Variables|Transient|GetRetryMMEActivationTimerHandle))
                    (Ability|Async|EndAction (Variables|Transient|GetAsyncAction_WaitMovementModeStart))
                    (if (ScalableFloat|GetValueasBool _hotfix_retrymmeapplicationonremovalinterval)
                      (bind _asynctaskproxy (Ability|Async|WaitGameplayEventToActor 0 "(TagName=\"ToyFox.MME.Event.Start\")")
                        (:then
                          (Variables|Transient|SetAsyncAction_WaitMovementModeStart _asynctaskproxy)
                          (bind _asynctaskproxy_2 (Ability|Async|WaitGameplayEventToActor 0 "(TagName=\"ToyFox.MME.Event.Stop\")")
                            (:then
                              (Variables|Transient|SetAsyncAction_WaitMovementModeEnd _asynctaskproxy_2))
                            (:EventReceived
                              (bind _returnvalue (Utilities|Time|SetTimerbyEvent (delegate Custom|RetryMMEApplication) (ScalableFloat|GetValueatLevel _hotfix_retrymmeapplicationonremovalinterval) true))
                              (Variables|Transient|SetRetryMMEActivationTimerHandle _returnvalue))))
                        (:EventReceived
                          (Utilities|Time|ClearandInvalidateTimerbyHandle (Variables|Transient|GetRetryMMEActivationTimerHandle)))))))))
            (:then_1
              ;;[0 0.15 0.01845 0.5] Handling footsteps.
              (Ability|Async|EndAction (Variables|Transient|GetAsyncAction_FootstepEvents))
              (bind _asynctaskproxy_3 (Ability|Async|WaitGameplayEventToActor 0 _leftfootstepanimnotifytag)
                (:then
                  (Utilities|Array|Add _asyncaction_footstepevents _asynctaskproxy_3)
                  (bind _asynctaskproxy_4 (Ability|Async|WaitGameplayEventToActor 0 _rightfootstepanimnotifytag)
                    (:then
                      (Utilities|Array|Add _asyncaction_footstepevents _asynctaskproxy_4)
                      ;;[0.000409 0.1358 0.15 0.5] Apply gameplay effects on the giant that exists on the giant when unpossessed and possessed.
                      (Networking|SwitchHasAuthority
                        (:Authority
                          (Fortnite|ApplyGameplayEffectsToActor self (Variables|Default|GetGiantGameplayEffects) (0.0 0.0 0.0) false 0 (Ability|GetFortAbilitySystemComponent self) (GameplayEffects|MakeEffectContext (Ability|GetFortAbilitySystemComponent self))))))
                    (:EventReceived
                      (bind (_bvalidfootstep _hitresult) (Footsteps|CacheFootstepHitResult _rightfootstepanimnotifytag))
                      (if _bvalidfootstep
                        ;;;[1 1 1] Cosmetic is executed for each player on their own client for best footstep visual representation.
                        (GameplayCue|ExecuteGameplayCue%28Burst%29Local self (Utilities|Struct|BreakGameplayCueTag (Variables|GameplayCue|GetRightFootstepGameplayCueTag)) (Footsteps|MakeGameplayCueParamforFootstepImpact _hitresult))
                        (if (or (Utilities|Enum|Equal(Enum) (Networking|GetLocalRole) "ROLE_AutonomousProxy") (Utilities|Enum|Equal(Enum) (Networking|GetLocalRole) "ROLE_Authority"))
                          (Utilities|FlowControl|Delay 0.0333) ; Delay for perf
                          (Ability|SendGameplayEventtoActor self "(TagName=\"ToyFox.Event.Ability.Footstep\")") ; Invokes the footstep ability for damage tracing.
                        )))))
                (:EventReceived
                  (bind (_bvalidfootstep_2 _hitresult_2) (Footsteps|CacheFootstepHitResult _leftfootstepanimnotifytag))
                  (if _bvalidfootstep_2
                    ;;;[1 1 1] Cosmetic is executed for each player on their own client for best footstep visual representation.
                    (GameplayCue|ExecuteGameplayCue%28Burst%29Local self (Utilities|Struct|BreakGameplayCueTag (Variables|GameplayCue|GetLeftFootstepGameplayCueTag)) (Footsteps|MakeGameplayCueParamforFootstepImpact _hitresult_2))
                    (if (or (Utilities|Enum|Equal(Enum) (Networking|GetLocalRole) "ROLE_AutonomousProxy") (Utilities|Enum|Equal(Enum) (Networking|GetLocalRole) "ROLE_Authority"))
                      (Utilities|FlowControl|Delay 0.0333) ; Delay for perf
                      (Ability|SendGameplayEventtoActor self "(TagName=\"ToyFox.Event.Ability.Footstep\")") ; Invokes the footstep ability for damage tracing.
                    )))))))))))

(event Custom|RetryMMEApplication
  ;[0.1035 0.01953 0.15 0.5] HandleAbilitySystemComponentInitialized: On the client, retry application of MME in case it ends at unexpected time. MME ending breaks the event totally. There's native logging to track down the reason if this ever happens but this implementation here is to help handle a failcase.
  ; On the server and client, handle the footstep behavior.
  (if (not (CallFunction|IsDying))
    (if (Pawn|IsLocallyControlled)
      (Class|FortMovementModeExtBlueprintLibrary|ActivateRegisteredMovementModeExtension (Variables|Settings|GetToyFoxMovementModeExt) self))))

(event EventOnDisplayHitNotify
  (params (DamageDealt :type Float%20%28single-precision%29) (bCriticalHit :type Boolean) (bFatalHit :type Boolean) (bShieldHit :type Boolean) (HitActor :type Actor%20Object%20Reference) (HitTags :type Gameplay%20Tag%20Container%20Structure) (HitLocation :type Vector :default "0, 0, 0"))
  ;[0.01324 0.15 0.1094 0.5] Shows damage numbers for the damage dealt by the giant.
  (EventDispatchers|CallDisplayHitNotify DamageDealt bCriticalHit bFatalHit bShieldHit HitActor HitLocation HitTags))

(event EventOnDamageServer
  (params (Damage :type Float%20%28single-precision%29) (DamageTags :type Gameplay%20Tag%20Container%20Structure%20%28by%20ref%29) (Momentum :type Vector :default "0, 0, 0") (HitInfo :type Hit%20Result%20Structure%20%28by%20ref%29) (InstigatedBy :type Controller%20Object%20Reference) (DamageCauser :type Actor%20Object%20Reference) (EffectContext :type Gameplay%20Effect%20Context%20Handle%20Structure))
  (bind (_bblockinghit _binitialoverlap _time _distance _location _impactpoint _normal _impactnormal _physmat _hitactor _hitcomponent) (Collision|BreakHitResult HitInfo))
  ;[0.15 0.01994 0.09637 0.5] OnDamageServer: When damage is received by the giant, run the desired behavior.
  (|CallParentFunction Damage DamageTags Momentum HitInfo InstigatedBy DamageCauser EffectContext)
  ;;[1 1 1] Power level reduction in the game state component.
  (bind _component (Default|GetGameStateComponent)
    (:Valid
      (PowerLevel|ReducePowerLevelbyDamage _component Damage DamageTags InstigatedBy))
    (:Invalid
      _))
  ;;[1 1 1] Generate weakpoints.
  (if (GameplayTags|HasTag DamageTags "(TagName=\"Gameplay.Damage.TidalCrane.ToyFox.Boss.WeakpointGenerator\")")
    (Weakspot|EnableWeakpointwithHitResult HitInfo)
    (else
      _))
  (if (GameplayTags|HasTag DamageTags "(TagName=\"Gameplay.Damage.TidalCrane.ToyFox.Stagger\")")
    (Ability|StaggerGiantPawn (Ability|GetAbilitySystemComponent DamageCauser) EffectContext)
    (else
      _))
  ;;[0.01674 0.08074 0.15 0.5] Weakpoint should take damage when hit. After enough damage, the weakpoint is destroyed.
  (if (Weakspot|IsValidWeakpointDamage InstigatedBy)
    (bind _bsuccess (Weakspot|ApplyDamagetoWeakpoint HitInfo Damage))
    (if _bsuccess
      (CallFunction|MulticastPlayWeakpointAnimation (Utilities|Casting|CastToB_ToyFoxWeakpointComponent _hitcomponent)))))

(event Custom|MulticastPlayWeakpointAnimation
  (params (WeakpointComponent :type B%20Toy%20Fox%20Weakpoint%20Component%20Object%20Reference))
  ;[0 0.15 0.1337 0.5] MulticastPlayWeakpointAnimation: OnPlayDamageEffects does not receive replicated hit component for perf reasons. Because of this we need to manually multicast notify everyone that a weakpoint is hit so that the players that are 'viewing' the giant as their view target see the weakpoint hit animations.
  (Networking|CanExecuteCosmeticEvents
    (:True
      (Utilities|IsValid WeakpointComponent
        (:Is%20Valid
          (if (State|IsWeakpointActive WeakpointComponent)
            (Utilities|IsValid (Pawn|GetLocalViewingPlayerController)
              (:Is%20Valid
                (Weakspot|PlayWeakpointMaterialAnimation WeakpointComponent)))))))))

(event EventOnWeaponEquipped
  (params (NewWeapon :type Fort%20Weapon%20Object%20Reference) (PrevWeapon :type Fort%20Weapon%20Object%20Reference))
  ;[0.09375 0.08819 0.01548] Work Around Try Equip Default Weapon
  (|CallParentFunction NewWeapon PrevWeapon)
  (CallFunction|TryEquipDefaultWeapon))

(event Custom|TryEquipDefaultWeapon
  (bind _returnvalue (Weapon|GetCurrentWeapon))
  ;[0.09375 0.08819 0.01548] Work Around Try Equip Default Weapon
  (Utilities|IsValid _returnvalue
    (:Is%20Valid
      (if (!= (Data|GetWeaponData _returnvalue) (Variables|FortGiantPlayerPawn|GetDefaultWeapon))
        (Utilities|Casting|CastToFortPlayerController (Pawn|GetController))
        (CallFunction|TryEquipDefaultItem)))))

(event EventOnDeathPlayEffects
  (params (Damage :type Float%20%28single-precision%29) (DamageTags :type Gameplay%20Tag%20Container%20Structure%20%28by%20ref%29) (Momentum :type Vector :default "0, 0, 0") (HitInfo :type Hit%20Result%20Structure%20%28by%20ref%29) (InstigatedBy :type Fort%20Pawn%20Object%20Reference) (DamageCauser :type Actor%20Object%20Reference) (EffectContext :type Gameplay%20Effect%20Context%20Handle%20Structure))
  ;[0.005094 0.00146 0.15 0.5] OnDeathPlayEffects: Unbind retry mme actions on death.
  ; Note: MME is activated from the client, hence placing this logic here.
  (|CallParentFunction Damage DamageTags Momentum HitInfo InstigatedBy DamageCauser EffectContext)
  (Utilities|Time|ClearandInvalidateTimerbyHandle (Variables|Transient|GetRetryMMEActivationTimerHandle))
  (Ability|Async|EndAction (Variables|Transient|GetAsyncAction_WaitMovementModeStart)))

(event Custom|PlayBeforeOutroMaterialRestore
  (bind _toyfoxskeletalmeshcomponentbudgeted (Variables|Default|GetToyFoxSkeletalMeshComponentBudgeted))
  ;[1 1 1] FadeOut Weakpoint and LowLife Visual Before Outro
  (Rendering|Material|SetScalarParameterValueonMaterials _toyfoxskeletalmeshcomponentbudgeted "Outro_Animation_Duration" 2.0) ; Set Duration. 1s
  (Rendering|Material|SetScalarParameterValueonMaterials _toyfoxskeletalmeshcomponentbudgeted "Outro_Stage" 1.0) ; Stage1
  (Rendering|Material|SetScalarParameterValueonMaterials _toyfoxskeletalmeshcomponentbudgeted "Outro_GameTimeFromBP" (Utilities|Time|GetGameTimeinSeconds)) ; Passthough Time
)

(event EventEndPlay
  (params (EndPlayReason :type EEndPlayReason%20Enum :default "Destroyed"))
  ;[0.125 0.06824 0 0.5] EndPlay: Clean up, unbind delegates.
  (Utilities|FlowControl|Sequence
    (:then_0
      (Utilities|Time|ClearandInvalidateTimerbyHandle (Variables|Transient|GetRetryMMEActivationTimerHandle))
      (Ability|Async|EndAction (Variables|Transient|GetAsyncAction_WaitMovementModeStart))
      (Ability|Async|EndAction (Variables|Transient|GetAsyncAction_FootstepEvents))
      (EventDispatchers|UnbindEventfromOnAbilitySystemComponentInitialized (EventDispatchers|CreateEvent :!event Custom|HandleAbilitySystemComponentInitialized))
      (Utilities|FlowControl|Sequence
        (:then_0
          (bind _component (Default|GetGameStateComponent)
            (:Valid
              (Default|UnbindEventfromOnEventStateChanged _component (EventDispatchers|CreateEvent :!event HandleEventStateChanged))
              (Default|UnbindEventfromOnGiantPowerLevelChanged _component (EventDispatchers|CreateEvent :!event Custom|HandleGiantPowerLevelChanged))
              (Default|UnbindEventfromOnGiantPowerLevelLowStateChanged _component (EventDispatchers|CreateEvent :!event Custom|HandleGiantPowerLevelLowStateChanged)))))
        (:then_1
          (Networking|SwitchHasAuthority
            (:Authority
              (bind _returnvalue (Class|FortGlobalActorRegistrySubsystem|GetActorRegistryDataforGameplayTag (WorldSubsystems|GetFortGlobalActorRegistrySubsystem) (Variables|Enemy|GetEnemyActorRegistryGameplayTag)))
              (Utilities|IsValid _returnvalue
                (:Is%20Valid
                  (EventDispatchers|UnbindEventfromOnActorRegistered _returnvalue (EventDispatchers|CreateEvent :!event HandleEnemyActorRegistered))
                  (EventDispatchers|UnbindEventfromOnActorUnregistered _returnvalue (EventDispatchers|CreateEvent :!event HandleEnemyActorUnregistered)))))))))
    (:then_1
      (|CallParentFunction EndPlayReason))))

(event EventOnDamagePlayEffects
  (params (Damage :type Float%20%28single-precision%29) (DamageTags :type Gameplay%20Tag%20Container%20Structure%20%28by%20ref%29) (Momentum :type Vector :default "0, 0, 0") (HitInfo :type Hit%20Result%20Structure%20%28by%20ref%29) (InstigatedBy :type Fort%20Pawn%20Object%20Reference) (DamageCauser :type Actor%20Object%20Reference) (EffectContext :type Gameplay%20Effect%20Context%20Handle%20Structure))
  (bind (_bblockinghit _binitialoverlap _time _distance _location _impactpoint _normal _impactnormal _physmat _hitactor _hitcomponent) (Collision|BreakHitResult HitInfo))
  ;[0 0.15 0.1159 0.5] OnDamagePlayEffects: When the giant is damaged, play the desired effects.
  (|CallParentFunction Damage DamageTags Momentum HitInfo InstigatedBy DamageCauser EffectContext)
  (Local|ExecuteDamageReceivedfromEnemyLocalPlayerFeedback DamageTags HitInfo)
  (Utilities|IsValid InstigatedBy
    (:Is%20Valid
      (if (Weakspot|IsValidWeakpointDamage (Pawn|GetController InstigatedBy))
        (bind _asb_toy_fox_weakpoint_component (Utilities|Casting|CastToB_ToyFoxWeakpointComponent _hitcomponent))
        (Weakspot|PlayWeakpointMaterialAnimation _asb_toy_fox_weakpoint_component)))))

(event Custom|SetEyeEmissiveEnabled
  (params (bEnable :type Boolean))
  ;[0.01004 0 0.15 0.5] SetEyeEmissiveEnabled: Called on the server to replicate eye emissive state to the clients.
  ; This is called directly through the cinematic..
  (Networking|SwitchHasAuthority
    (:Authority
      (Variables|Eye|SetbEyeEmissiveEnabled bEnable))))

(event Custom|OnRep_bEyeEmissiveEnabled
  (bind _toyfoxskeletalmeshcomponentbudgeted (Variables|Default|GetToyFoxSkeletalMeshComponentBudgeted))
  ;[0.01004 0 0.15 0.5] SetEyeEmissiveEnabled: Called on the server to replicate eye emissive state to the clients.
  ; This is called directly through the cinematic..
  ;;[0.101 0.04341 0.15 0.5] OnRep_bEyeEmissiveEnabled: Update the eye emissive for non pilot players only. Pilot players will update their emissive through the cinematic.
  (Networking|CanExecuteCosmeticEvents
    (:True
      (bind _breturnvalue (General|IsLocalPlayerPilot))
      (if _breturnvalue
        _
        (elif (Variables|Eye|GetbEyeEmissiveEnabled)
          (Rendering|Material|SetScalarParameterValueonMaterials _toyfoxskeletalmeshcomponentbudgeted "Eye_AnimDuration" 1.0) ; Set Duration. 3s
          (Rendering|Material|SetScalarParameterValueonMaterials _toyfoxskeletalmeshcomponentbudgeted "Eye_CurrentStage" 2.0) ; Stage2 is for active Blink
          (Rendering|Material|SetScalarParameterValueonMaterials _toyfoxskeletalmeshcomponentbudgeted "Eye_GameTimeFromBP" (Utilities|Time|GetGameTimeinSeconds)) ; Passthough Time
          (Rendering|Material|SetScalarParameterValueonMaterials _toyfoxskeletalmeshcomponentbudgeted (Variables|Eye|GetEyeIntensityParamName) (Variables|Eye|GetEyeIntensityMax)) ; Set Duration. 3s
          (else
            (Rendering|Material|SetScalarParameterValueonMaterials _toyfoxskeletalmeshcomponentbudgeted "Eye_CurrentStage" 0.0) ; Stage0 reset
            (Rendering|Material|SetScalarParameterValueonMaterials _toyfoxskeletalmeshcomponentbudgeted (Variables|Eye|GetEyeIntensityParamName) (Variables|Eye|GetEyeIntensityMin)) ; Set Duration. 3s
          ))))))

(event Custom|HandleGiantPowerLevelChanged
  (params (PowerLevel :type Float%20%28double-precision%29))
  (bind _toyfoxskeletalmeshcomponentbudgeted (Variables|Default|GetToyFoxSkeletalMeshComponentBudgeted))
  ;[0.138 0.03722 0.15 0.5] HandleGiantPowerLevelChanged: Run desired behavior based on the giant power level (0 to 1). This event is invoked on both server and client.
  ; NOTE: THE POWER LEVEL CAN GO BOTH UP AND DOWN. HANDLE BOTH CASES. POWER LEVEL GOES UP DURING THE FINISHER BUT CAN ALSO BE CHEATED TO GO UP for easy testing.
  (Networking|SwitchHasAuthority
    (:Remote
      ;[1 1 1] PowerRegenVisualTreatment
      (Utilities|FlowControl|Sequence
        (:then_0
          (if (< (Variables|Default|GetPrevious%20Power%20Level%20Regen) PowerLevel)
            (Rendering|Material|SetScalarParameterValueonMaterials _toyfoxskeletalmeshcomponentbudgeted "IsPowerRegen" 1.0) ; Colored Elecric Wrap
            (Rendering|Material|SetScalarParameterValueonMaterials _toyfoxskeletalmeshcomponentbudgeted "PowerRegen_Stage" 2.0)
            (Rendering|Material|SetScalarParameterValueonMaterials _toyfoxskeletalmeshcomponentbudgeted "PowerRegen_Animation_Duration" 3.0) ; Stage2 is for active Blink
            (Rendering|Material|SetScalarParameterValueonMaterials _toyfoxskeletalmeshcomponentbudgeted "PowerRegen_GameTimeFromBP" (Utilities|Time|GetGameTimeinSeconds)) ; Passthough Time
            (else
              (Rendering|Material|SetScalarParameterValueonMaterials _toyfoxskeletalmeshcomponentbudgeted "IsPowerRegen" 0.0) ; Normal Elecric Wrap
            )))
        (:then_1
          (Variables|Default|SetPrevious%20Power%20Level%20Regen PowerLevel))))))

(event Custom|HandleGiantPowerLevelLowStateChanged
  (params (bLow :type Boolean))
  (bind _returnvalue (Utilities|Time|GetGameTimeinSeconds))
  (bind _self self)
  (bind _returnvalue_2 (Ability|GetFortAbilitySystemComponent _self))
  (bind _self_2 self)
  (bind _returnvalue_3 (Ability|GetFortAbilitySystemComponent _self_2))
  (bind _toyfoxskeletalmeshcomponentbudgeted (Variables|Default|GetToyFoxSkeletalMeshComponentBudgeted))
  ;[0 0.1256 0.15 0.5] HandleGiantPowerLevelLow: Apply or remove the low power level warning gameplay effect based on the latest low power level state.
  (Networking|SwitchHasAuthority
    (:Authority
      ;;[0.15 0.04002 0.003809 0.5] Server, Apply or remove the low power level warning gameplay effect based on the latest power level.
      (if bLow
        (Utilities|IsValid _returnvalue_3
          (:Is%20Valid
            (GameplayEffects|ApplyGameplayEffectToSelf _returnvalue_3 (Variables|Settings|GetLowPowerLevelGameplayEffectClass) 0.0 (GameplayEffects|MakeEffectContext _returnvalue_3)) ; Sending in Level 0 to indicate primary player.
          ))
        (else
          (Utilities|IsValid _returnvalue_2
            (:Is%20Valid
              (GameplayEffects|RemoveActiveGameplayEffectbySourceEffect _returnvalue_2 (Variables|Settings|GetLowPowerLevelGameplayEffectClass) 0 -1))))))
    (:Remote
      ;[0.004686 0.1313 0.15 0.5] Client, apply cosmetic material updates.
      (if bLow
        (Rendering|Material|SetScalarParameterValueonMaterials _toyfoxskeletalmeshcomponentbudgeted "LF_Animation_Duration" 3.0) ; Set Duration. 10s
        (Rendering|Material|SetScalarParameterValueonMaterials _toyfoxskeletalmeshcomponentbudgeted "LF_Stage" 1.0) ; Stage1 to active eletric wrap
        (Rendering|Material|SetScalarParameterValueonMaterials _toyfoxskeletalmeshcomponentbudgeted "LF_GameTimeFromBP" _returnvalue) ; Passthough Time
        (Rendering|Material|SetScalarParameterValueonMaterials _toyfoxskeletalmeshcomponentbudgeted "IsPowerRegen" 0.0) ; Normal Elecric Wrap
        (Rendering|Material|SetScalarParameterValueonMaterials _toyfoxskeletalmeshcomponentbudgeted "PowerRegen_Stage" 1.0)
        (Rendering|Material|SetScalarParameterValueonMaterials _toyfoxskeletalmeshcomponentbudgeted "PowerRegen_Animation_Duration" 3.0) ; Stage2 is for active Blink
        (Rendering|Material|SetScalarParameterValueonMaterials _toyfoxskeletalmeshcomponentbudgeted "PowerRegen_GameTimeFromBP" _returnvalue) ; Passthough Time
        (else
          (Rendering|Material|SetScalarParameterValueonMaterials _toyfoxskeletalmeshcomponentbudgeted "LF_Animation_Duration" 3.0) ; Set Duration. 6s
          (Rendering|Material|SetScalarParameterValueonMaterials _toyfoxskeletalmeshcomponentbudgeted "LF_Stage" 2.0) ; Stage2 to deactive eletric wrap
          (Rendering|Material|SetScalarParameterValueonMaterials _toyfoxskeletalmeshcomponentbudgeted "LF_GameTimeFromBP" _returnvalue) ; Passthough Time
          (Rendering|Material|SetScalarParameterValueonMaterials _toyfoxskeletalmeshcomponentbudgeted "IsPowerRegen" 1.0) ; Colored Elecric Wrap
          (Rendering|Material|SetScalarParameterValueonMaterials _toyfoxskeletalmeshcomponentbudgeted "PowerRegen_Stage" 2.0)
          (Rendering|Material|SetScalarParameterValueonMaterials _toyfoxskeletalmeshcomponentbudgeted "PowerRegen_Animation_Duration" 3.0) ; Stage2 is for active Blink
          (Rendering|Material|SetScalarParameterValueonMaterials _toyfoxskeletalmeshcomponentbudgeted "PowerRegen_GameTimeFromBP" _returnvalue) ; Passthough Time
        )))))

(event OnComponentBeginOverlap(Mesh)
  (params (OverlappedComponent :type Primitive%20Component%20Object%20Reference) (OtherActor :type Actor%20Object%20Reference) (OtherComp :type Primitive%20Component%20Object%20Reference) (OtherBodyIndex :type Integer) (bFromSweep :type Boolean) (SweepResult :type Hit%20Result%20Structure%20%28by%20ref%29))
  ;[0.03291 0.1457 0.15 0.5] Handle overlapping player or vehicles
  (Utilities|HandlePawnMeshStartOverlap OtherActor
    (:Handled
      (if (not (Utilities|Time|IsTimerActivebyHandle (Variables|Transient|GetOverlapKnockbackTimerHandle)))
        (bind _returnvalue (Utilities|Time|SetTimerbyEvent (EventDispatchers|CreateEvent :!event Custom|KnocbackOverlappedActors) (ScalableFloat|GetCachedScalableFloatValue (Variables|Hotfix|GetHotfix_OverlapKnockbackCheckFreq)) true))
        (Variables|Transient|SetOverlapKnockbackTimerHandle _returnvalue)))))

(event OnComponentBeginOverlap(ToyFoxSkeletalMeshComponentBudgeted)
  (params (OverlappedComponent :type Primitive%20Component%20Object%20Reference) (OtherActor :type Actor%20Object%20Reference) (OtherComp :type Primitive%20Component%20Object%20Reference) (OtherBodyIndex :type Integer) (bFromSweep :type Boolean) (SweepResult :type Hit%20Result%20Structure%20%28by%20ref%29))
  ;[0.03291 0.1457 0.15 0.5] Handle overlapping player or vehicles
  (Utilities|HandlePawnMeshStartOverlap OtherActor
    (:Handled
      (if (not (Utilities|Time|IsTimerActivebyHandle (Variables|Transient|GetOverlapKnockbackTimerHandle)))
        (bind _returnvalue (Utilities|Time|SetTimerbyEvent (EventDispatchers|CreateEvent :!event Custom|KnocbackOverlappedActors) (ScalableFloat|GetCachedScalableFloatValue (Variables|Hotfix|GetHotfix_OverlapKnockbackCheckFreq)) true))
        (Variables|Transient|SetOverlapKnockbackTimerHandle _returnvalue)))))

(event OnComponentBeginOverlap(WeaponSkeletalMeshComponentBudgeted)
  (params (OverlappedComponent :type Primitive%20Component%20Object%20Reference) (OtherActor :type Actor%20Object%20Reference) (OtherComp :type Primitive%20Component%20Object%20Reference) (OtherBodyIndex :type Integer) (bFromSweep :type Boolean) (SweepResult :type Hit%20Result%20Structure%20%28by%20ref%29))
  ;[0.03291 0.1457 0.15 0.5] Handle overlapping player or vehicles
  (Networking|SwitchHasAuthority
    (:Authority
      (Utilities|IsValid OtherActor
        (:Is%20Valid
          (if (!= OtherActor self)
            (Utilities|Array|AddUnique (Variables|Transient|GetOverlappedActors) OtherActor)
            (CallFunction|KnocbackOverlappedActors)
            (if (not (Utilities|Time|IsTimerActivebyHandle (Variables|Transient|GetOverlapKnockbackTimerHandle)))
              (bind _returnvalue (Utilities|Time|SetTimerbyEvent (EventDispatchers|CreateEvent :!event Custom|KnocbackOverlappedActors) (ScalableFloat|GetCachedScalableFloatValue (Variables|Hotfix|GetHotfix_OverlapKnockbackCheckFreq)) true))
              (Variables|Transient|SetOverlapKnockbackTimerHandle _returnvalue))))))))

(event OnComponentEndOverlap(Mesh)
  (params (OverlappedComponent :type Primitive%20Component%20Object%20Reference) (OtherActor :type Actor%20Object%20Reference) (OtherComp :type Primitive%20Component%20Object%20Reference) (OtherBodyIndex :type Integer))
  ;[0.03291 0.1457 0.15 0.5] Handle overlapping player or vehicles
  (Utilities|HandlePawnMeshStopOverlap OtherActor
    (:Handled
      (if (Utilities|Array|IsEmpty (Variables|Transient|GetOverlappedActors))
        (Utilities|Time|ClearandInvalidateTimerbyHandle (Variables|Transient|GetOverlapKnockbackTimerHandle))))))

(event OnComponentEndOverlap(ToyFoxSkeletalMeshComponentBudgeted)
  (params (OverlappedComponent :type Primitive%20Component%20Object%20Reference) (OtherActor :type Actor%20Object%20Reference) (OtherComp :type Primitive%20Component%20Object%20Reference) (OtherBodyIndex :type Integer))
  ;[0.03291 0.1457 0.15 0.5] Handle overlapping player or vehicles
  (Utilities|HandlePawnMeshStopOverlap OtherActor
    (:Handled
      (if (Utilities|Array|IsEmpty (Variables|Transient|GetOverlappedActors))
        (Utilities|Time|ClearandInvalidateTimerbyHandle (Variables|Transient|GetOverlapKnockbackTimerHandle))))))

(event OnComponentEndOverlap(WeaponSkeletalMeshComponentBudgeted)
  (params (OverlappedComponent :type Primitive%20Component%20Object%20Reference) (OtherActor :type Actor%20Object%20Reference) (OtherComp :type Primitive%20Component%20Object%20Reference) (OtherBodyIndex :type Integer))
  ;[0.03291 0.1457 0.15 0.5] Handle overlapping player or vehicles
  (Networking|SwitchHasAuthority
    (:Authority
      (Utilities|IsValid OtherActor
        (:Is%20Valid
          (Utilities|Array|RemoveItem (Variables|Transient|GetOverlappedActors) OtherActor)
          (if (Utilities|Array|IsEmpty (Variables|Transient|GetOverlappedActors))
            (Utilities|Time|ClearandInvalidateTimerbyHandle (Variables|Transient|GetOverlapKnockbackTimerHandle))))))))

(event Custom|KnocbackOverlappedActors
  ;[0.03291 0.1457 0.15 0.5] Handle overlapping player or vehicles
  (Networking|SwitchHasAuthority
    (:Authority
      (for _array_element (Variables|Transient|GetOverlappedActors)
        (Utilities|IsValid _array_element
          (:Is%20Valid
            (Ability|KnockbackOverlappedActor _array_element)))))))

(event OnComponentHit(Mesh)
  (params (HitComponent :type Primitive%20Component%20Object%20Reference) (OtherActor :type Actor%20Object%20Reference) (OtherComp :type Primitive%20Component%20Object%20Reference) (NormalImpulse :type Vector :default "0, 0, 0") (Hit :type Hit%20Result%20Structure%20%28by%20ref%29))
  ;[1 1 1] Handle hit players during kneeling stage, to avoid player stuck
  (Intro|HandlePawnMeshHit OtherActor Hit))

(event OnComponentHit(ToyFoxSkeletalMeshComponentBudgeted)
  (params (HitComponent :type Primitive%20Component%20Object%20Reference) (OtherActor :type Actor%20Object%20Reference) (OtherComp :type Primitive%20Component%20Object%20Reference) (NormalImpulse :type Vector :default "0, 0, 0") (Hit :type Hit%20Result%20Structure%20%28by%20ref%29))
  ;[1 1 1] Handle hit players during kneeling stage, to avoid player stuck
  (Intro|HandlePawnMeshHit OtherActor Hit))

(event EventOutroMaterialRestore
  (CallFunction|PlayBeforeOutroMaterialRestore))

(event EventDestroyWeakPoints
  (bind _components (Weakspot|GetWeakpointComponents))
  (for _array_element _components
    (Components|DestroyComponent _array_element)))

(fn AnimFacingDebug
  (bind _astoy_fox_giant_pawn_anim_bp (Utilities|Casting|CastToToyFoxGiantPawnAnimBP (Components|SkeletalMesh|GetAnimInstance (Variables|Character|GetMesh))))
  (if (Variables|/ToyFoxGameplay/Gameplay/ToyFoxGiantPawnAnimBP.ToyFoxGiantPawnAnimBP_C|GetPrintLocoDebug _astoy_fox_giant_pawn_anim_bp)
    (Rendering|Debug|DrawDebugArrow (+ (Transformation|GetActorLocation) "0.000000,0.000000,-1000.000000") (+ (+ (Transformation|GetActorLocation) "0.000000,0.000000,-1000.000000") (* (Transformation|GetActorForwardVector) 2e+03)) 3e+03 "(R=1.000000,G=0.000000,B=0.995681,A=1.000000)" 0.1 3e+01)
    (Rendering|Debug|DrawDebugArrow (Transformation|GetSocketLocation (Variables|Character|GetMesh) "root") (+ (Transformation|GetSocketLocation (Variables|Character|GetMesh) "root") (* (Math|Vector|GetForwardVector (Math|Rotator|CombineRotators (Transformation|GetSocketRotation (Variables|Character|GetMesh) "root") "0.000000,90.000000,0.000000")) 2e+03)) 3e+03 "(R=1.000000,G=0.738503,B=0.000000,A=1.000000)" 0.1 3e+01)))

(fn ApplyDamageToWeakpoint
  (params (Hit :type Hit%20Result%20Structure) (Damage :type Float%20%28double-precision%29))
  (bind (_bblockinghit _binitialoverlap _time _distance _location _impactpoint _normal _impactnormal _physmat _hitactor _hitcomponent) (Collision|BreakHitResult Hit))
  ;[0.09367 0 0.15 0.5] ApplyDamageToWeakpoint: If an active weakpoint was hit, pass in the damage value to the weakpoint.
  (Utilities|FlowControl|Sequence
    (:then_0
      (Networking|SwitchHasAuthority
        (:Authority
          (bind _asb_toy_fox_weakpoint_component (Utilities|Casting|CastToB_ToyFoxWeakpointComponent _hitcomponent))
          (if (State|IsWeakpointActive _asb_toy_fox_weakpoint_component)
            (Utilities|FlowControl|Sequence
              (:then_0
                (Damage|NotifyDamage _asb_toy_fox_weakpoint_component Damage)
                ;;[1 1 1] Weakpoint destroyed, apply the bonus damage separately here
                (if (not (State|IsWeakpointActive _asb_toy_fox_weakpoint_component))
                  (bind _component (Default|GetGameStateComponent)
                    (:Valid
                      (PowerLevel|ReducePowerLevelWeakpointDestroy _component (Damage|GetWeakpointDestroyBonusDamage _asb_toy_fox_weakpoint_component))))))
              (:then_1
                (return true)))))))
    (:then_1
      (return false))))

(fn ApproximateHitWeakpointComponent
  (params (HitResult :type Hit%20Result%20Structure))
  (bind (_bblockinghit _binitialoverlap _time _distance _location _impactpoint) (Collision|BreakHitResult HitResult))
  (bind _impactpoint_2 (LocalVariables|GetImpactPoint))
  ;[0 0.09156 0.15 0.5] ApproximateHitWeakpointComponent: Based on the hit result of incoming damage, figure out the weakpoint to activate.
  (Utilities|FlowControl|Sequence
    (:then_0
      ;;[0.15 0.04092 0.1473 0.5] Early return if there's already a max number of weakpoints.
      (bind _components (Weakspot|GetWeakpointComponents))
      (for _array_element _components
        (if (State|IsWeakpointActive _array_element)
          (bind _result (Math|Integer|IncrementInt (LocalVariables|GetCurrentNumofWeakpoints))
            (:%20%20
              (if (>= _result (ScalableFloat|GetCachedScalableFloatValue (Variables|Hotfix|GetHotfix_MaxNumOfWeakpoints)))
                (return 0)))))))
    (:then_1
      (LocalVariables|SetImpactPoint _impactpoint)
      ;;[0.05585 0.1019 0.15 0.5] Loop through all the weakpoints and find the closest weakpoint that should register the hit result as an impact.
      ;; Note: The weakpoints do not actually have their collision enabled when they are disabled, so this needs to approximate the impact on that weakpoint.
      (bind _components_2 (Weakspot|GetWeakpointComponents))
      (for _array_element_2 _components_2
        (Utilities|IsValid _array_element_2
          (:Is%20Valid
            (bind (_outradius _outhalfheight) (Components|Capsule|GetScaledCapsuleSize _array_element_2))
            (if (Math|Geometry|IsPointinBoxwithTransform (LocalVariables|GetImpactPoint) (Math|Transform|MakeTransform (.location (Transformation|GetWorldTransform _array_element_2)) (.rotation (Transformation|GetWorldTransform _array_element_2))) (Math|Vector|MakeVector (* _outradius 1.3) (* _outradius 1.3) (* _outhalfheight 1.3)))
              (if (< (Math|Vector|Distance2DSquared(Vector) _impactpoint_2 (.location (Transformation|GetWorldTransform _array_element_2))) (LocalVariables|GetSmallestDistanceSq))
                (LocalVariables|SetSmallestDistanceSq (Math|Vector|Distance2DSquared(Vector) _impactpoint_2 (.location (Transformation|GetWorldTransform _array_element_2))))
                (LocalVariables|SetRetComponent _array_element_2)))))))
    (:then_2
      (return (LocalVariables|GetRetComponent)))))

(fn BlueprintModifyIncomingDamage
  (params (Damage :type Float%20%28single-precision%29) (DamageTags :type Gameplay%20Tag%20Container%20Structure%20%28by%20ref%29) (EffectContext :type Gameplay%20Effect%20Context%20Handle%20Structure%20%28by%20ref%29) (EventInstigator :type Controller%20Object%20Reference) (DamageCauser :type Actor%20Object%20Reference) (HitInfo :type Hit%20Result%20Structure))
  ;[0.02273 0.15 0.1062 0.5] Do not take fall damage.
  (if (GameplayTags|HasTag DamageTags "(TagName=\"Gameplay.Damage.Environment.Falling\")")
    (return 0.0)
    (else
      (bind _returnvalue (|CallParentFunction Damage DamageTags EffectContext EventInstigator DamageCauser HitInfo))
      (return _returnvalue))))

(fn CacheFootstepHitResult
  (params (EventTag :type Gameplay%20Tag%20Structure))
  (bind _eventtagtofootsocketname (Variables|Footsteps|GetEventTagToFootSocketName))
  (bind (_value _returnvalue) (Utilities|Map|Find _eventtagtofootsocketname EventTag))
  (bind _tracedfootsteplocation (LocalVariables|GetTracedFootstepLocation))
  (Utilities|FlowControl|Sequence
    (:then_0
      (if _returnvalue
        (bind _returnvalue_2 (Transformation|GetSocketLocation (Variables|Character|GetMesh) _value))
        (bind _outhits (Collision|MultiLineTraceForObjects (Utilities|Operators|Add _returnvalue_2 (Math|Vector|MakeVector 0.0 0.0 (Variables|Footsteps|GetGroundDetectionTraceStartFootSocketLocationZOffset))) (Utilities|Operators|Subtract _returnvalue_2 (Math|Vector|MakeVector 0.0 0.0 (Variables|Footsteps|GetGroundDetectionTraceEndFootSocketLocationZOffset))) (LocalVariables|GetTraceObjectTypeQuery) false (Utilities|Array|MakeArray self) "ForDuration" true "(R=1.000000,G=0.000000,B=0.000000,A=1.000000)" "(R=0.000000,G=1.000000,B=0.000000,A=1.000000)" 8.0))
        (if (Utilities|Array|IsValidIndex _outhits)
          (bind _output_get (LocalVariables|SetTracedFootstepLocation (Utilities|Operators|Add (.ImpactPoint (Collision|BreakHitResult (Utilities|Array|Get(aref) _outhits))) (Math|Vector|MakeVector 0.0 0.0 (Variables|Footsteps|GetFootstepZOffset)))))
          (if (> (Math|Vector|DistanceSquared(Vector) _output_get (Utilities|Map|Find (Variables|Transient|GetEventTagToLatestFootstepLocationMap) (Variables|GetEventTag))) (Math|Float|Square (ScalableFloat|GetCachedScalableFloatValue (Variables|Hotfix|GetHotfix_FootstepExecuteMinDistanceDeltaThreshold))))
            (Footsteps|SetLatestFootstepHitLocation _tracedfootsteplocation)
            (Utilities|Map|Add (Variables|Transient|GetEventTagToLatestFootstepLocationMap) (Variables|GetEventTag) _tracedfootsteplocation) ; Track the latest footstep location per foot.
            (return true (Utilities|Array|Get(aref) _outhits))))))
    (:then_1
      (return false 0))))

(fn EnableWeakpointWithHitResult
  (params (HitResult :type Hit%20Result%20Structure))
  ;[0.1079 0.02275 0.1094 0.5] EnableWeakpointWithHitResult: Use the hit result to enable the expected weakpoint.
  (Networking|SwitchHasAuthority
    (:Authority
      (bind _component (Weakspot|ApproximateHitWeakpointComponent HitResult))
      (Utilities|IsValid _component
        (:Is%20Valid
          (if (Utilities|Enum|Equal(Enum) (State|GetWeakpointState _component) "Inactive")
            (State|SetWeakpointState _component "Active")))))))

(fn ExecuteDamageReceivedFromEnemyLocalPlayerFeedback
  (params (DamageTags :type Gameplay%20Tag%20Container%20Structure) (HitInfo :type Hit%20Result%20Structure))
  ;[0.1143 0.000506 0.15 0.5] ExecuteDamageReceivedFromEnemyLocalPlayerFeedback: Function is responsible for playing the appropriate feedback for the received damage.
  ; Currently only executes feedback gameplay cues for attacks from the tidal crane enemy.
  (Networking|CanExecuteCosmeticEvents
    (:True
      (bind _localcontrollers (CallFunction|GetLocalControllerstoViewFeedback))
      (if (Utilities|Array|IsEmpty _localcontrollers)
        _
        (elif (GameplayTags|HasTag DamageTags "(TagName=\"Gameplay.Damage.TidalCrane.ToyFox.Boss.Melee\")" true)
          ;;[1 1 1] Feedback for melee attacks.
          (LocalVariables|SetSelectedGameplayCueTag (Variables|GameplayCue|GetTidalCraneBossMeleeDamageReceivedLocalFeedbackGameplayCueTag))
          (for _array_element _localcontrollers
            (Utilities|IsValid _array_element
              (:Is%20Valid
                (GameplayCue|ExecuteGameplayCue%28Burst%29Local (Variables|/Script/Engine.Controller|GetPlayerState _array_element) (Utilities|Struct|BreakGameplayCueTag (LocalVariables|GetSelectedGameplayCueTag)) (GameplayCue|MakeGameplayCueParametersfromHitResult HitInfo)))))
          (elif (GameplayTags|HasTag DamageTags "(TagName=\"Gameplay.Damage.TidalCrane.ToyFox.Boss.Projectile\")" true)
            ;;[1 1 1] Feedback for range attacks.
            (LocalVariables|SetSelectedGameplayCueTag (Variables|GameplayCue|GetTidalCraneBossRangeDamageReceivedLocalFeedbackGameplayCueTag))))))))

(fn GetCannonMuzzleLocations
  (return (Transformation|GetSocketLocation (Variables|Default|GetToyFoxSkeletalMeshComponentBudgeted) "CannonMuzzle_L") (Transformation|GetSocketLocation (Variables|Default|GetToyFoxSkeletalMeshComponentBudgeted) "CannonMuzzle_R")))

(fn GetCannonMuzzleLocOld
  (return (Transformation|GetSocketLocation (Variables|Default|GetToyFoxSkeletalMeshComponentBudgeted) "CannonMuzzle_L") (Transformation|GetSocketLocation (Variables|Default|GetToyFoxSkeletalMeshComponentBudgeted) "CannonMuzzle_R")))

(fn GetComponentDamageZone
  (params (HitComponent :type Primitive%20Component%20Object%20Reference))
  ;[0.0171 0.1137 0.15 0.5] GetComponentDamageZone: Return if the damaged component should take critical, normal, or no damage.
  ; Note: weak points are handled in code directly
  (if (== HitComponent (Variables|Default|GetWeaponSkeletalMeshComponentBudgeted))
    (return "Invulnerability")
    (else
      (return "Normal"))))

(fn GetHeadSocketTrans
  (return (Transformation|GetSocketTransform (Variables|Character|GetMesh) "neck_02" "RTS_World")))

(fn GetHeadSocketTransform
  ;[1 1 1] we must use neck bone, since head bone is hidden and isn't updating socket loc correctly
  (return (Transformation|GetSocketTransform (Variables|Character|GetMesh) "neck_02" "RTS_World")))

(fn GetLaserBeamHitResult
  (return (Variables|Transient|GetLaserBeamHitResult)))

(fn GetLatestFootstepHitLocation
  (return (Variables|Transient|GetLatestFootstepHitLocation)))

(fn GetLatestLandscapeHitResults
  (bind _component (Components|GetWeaponDestructionComponent))
  (Utilities|IsValid _component
    (:Is%20Valid
      (bind _hitresults (Class|BToyFoxWeaponDestructionComponent|GetLatestLandscapeHitResults _component))
      (return _hitresults))))

(fn GetLocalControllersToViewFeedback
  (Utilities|Array|Clear (LocalVariables|GetOutPlayerControllers))
  (Utilities|FlowControl|Sequence
    (:then_0
      (bind _component (Default|GetGameStateComponent)
        (:Valid
          (LocalVariables|SetCurrentPilotActors (Pilots|GetPilotActors _component))
          (for _array_element (Fortnite|GetLocalFortPlayerControllers true true)
            (if (== self (Pawn|GetViewTarget _array_element))
              _
              (elif (>= (Utilities|Array|FindItem (LocalVariables|GetCurrentPilotActors) (Pawn|GetViewTarget _array_element)) 0)
                _))
            (Utilities|Array|Add (LocalVariables|GetOutPlayerControllers) _array_element)))))
    (:then_1
      (return (LocalVariables|GetOutPlayerControllers)))))

(fn GetMarkerDetailsId
  (bind _component (Default|GetGameStateComponent)
    (:Valid
      (if (Utilities|Enum|Equal(Enum) (Event|GetEventState _component) "NewEnumerator3")
        (return "MarkerDetails.ToyFox.Giant.Possessed")
        (else
          _)))
    (:Invalid
      _))
  (return "MarkerDetails.ToyFox.Giant.Unpossessed"))

(fn GetNearbyPerceptionActors
  (params (bWithinAttackAzimuth :type Boolean :default true))
  ;[0 0.1193 0.15 0.5] Returns the nearby actors based on the input condition.
  (bind _output_get (LocalVariables|SetAllPerceptionActors (Variables|Transient|GetAllNearbyPerceptionFoundActors)))
  (Utilities|Array|AppendArray _output_get (Variables|Transient|GetEnemyActors))
  (if bWithinAttackAzimuth
    (for _array_element (LocalVariables|GetAllPerceptionActors)
      (Utilities|IsValid _array_element
        (:Is%20Valid
          (bind _azimuth (Math|Vector|GetAzimuthandElevation (Math|Vector|GetUnitDirection(Vector) (Transformation|GetActorLocation) (Transformation|GetActorLocation _array_element)) (Transformation|GetActorTransform)))
          (if (and (> _azimuth (* (ScalableFloat|GetCachedScalableFloatValue (Variables|Hotfix|GetHotfix_StandInPilotPerceptionAzimuthAngle)) -1.0)) (< _azimuth (ScalableFloat|GetCachedScalableFloatValue (Variables|Hotfix|GetHotfix_StandInPilotPerceptionAzimuthAngle))))
            (Utilities|Array|Add (LocalVariables|GetAzimuthPerceptionFoundActors) _array_element)))))
    (return (LocalVariables|GetAzimuthPerceptionFoundActors))
    (else
      (return (LocalVariables|GetAllPerceptionActors)))))

(fn GetPawnExitTeleportExtraHeight
  (return (Math|Random|RandomFloatinRange (ScalableFloat|GetCachedScalableFloatValue (Variables|Hotfix|GetHotfix_ExitPawnTeleportHeightMin)) (ScalableFloat|GetCachedScalableFloatValue (Variables|Hotfix|GetHotfix_ExitPawnTeleportHeightMax)))))

(fn GetRocketMuzzleLocations
  (return (Transformation|GetSocketLocation (Variables|Default|GetToyFoxSkeletalMeshComponentBudgeted) "RocketSocket_L") (Transformation|GetSocketLocation (Variables|Default|GetToyFoxSkeletalMeshComponentBudgeted) "RocketSocket_R")))

(fn GetSkeletalMeshBudgeted
  (return (Variables|Default|GetToyFoxSkeletalMeshComponentBudgeted)))

(fn GetToyFoxSkeletalMeshComponent
  (return (Variables|Default|GetToyFoxSkeletalMeshComponentBudgeted)))

(fn GetWeakpointComponents
  ;[0.03246 0.09931 0.15 0.5] Return all the weakpoint components.
  (return (Utilities|Array|MakeArray (Variables|Default|GetWeakpointComp_LowerLeg_R) (Variables|Default|GetWeakpointComp_LowerLeg_L) (Variables|Default|GetWeakpointComp_Back) (Variables|Default|GetWeakpointComp_Chest) (Variables|Default|GetWeakpointComp_Arm_R) (Variables|Default|GetWeakpointComp_Arm_L) (Variables|Default|GetWeakpointComp_Shoulder_L) (Variables|Default|GetWeakpointComp_Shoulder_R))))

(fn GetWeakspotComponents
  (bind _components (Weakspot|GetWeakpointComponents))
  (for _array_element _components
    (Utilities|IsValid _array_element
      (:Is%20Valid
        (if (State|IsWeakpointActive _array_element)
          (Utilities|Array|Add (LocalVariables|GetRetComponents) _array_element)))))
  (return (LocalVariables|GetRetComponents)))

(fn GetWeaponDestructionComponent
  (return (Variables|Default|GetToyFoxWeaponDestructionComponent)))

(fn GetWeaponMeshComponent
  (return (Variables|Default|GetWeaponSkeletalMeshComponentBudgeted)))

(fn HandleEnemyActorRegistered
  (params (GameplayTag :type Gameplay%20Tag%20Structure%20%28by%20ref%29) (RegisteredActor :type Actor%20Object%20Reference))
  ;[0.15 0 0.1272 0.5] HandleEnemyActorRegistered: Track the enemy acotrs in the world.
  (Utilities|FlowControl|Sequence
    (:then_0
      (Networking|SwitchHasAuthority
        (:Authority
          (if (== GameplayTag (Variables|Enemy|GetEnemyActorRegistryGameplayTag))
            (Utilities|IsValid RegisteredActor
              (:Is%20Valid
                (Utilities|Array|Add (Variables|Transient|GetEnemyActors) RegisteredActor)))))))
    (:then_1
      (return))))

(fn HandleEnemyActorUnregistered
  (params (GameplayTag :type Gameplay%20Tag%20Structure%20%28by%20ref%29) (UnregisteredActor :type Actor%20Object%20Reference))
  ;[0.006462 0.01688 0.15 0.5] HandleEnemyActorUnregistered: Unregister enemy actor.
  (Utilities|FlowControl|Sequence
    (:then_0
      (Networking|SwitchHasAuthority
        (:Authority
          (Utilities|IsValid UnregisteredActor
            (:Is%20Valid
              (if (== GameplayTag (Variables|Enemy|GetEnemyActorRegistryGameplayTag))
                (bind _returnvalue (Utilities|Array|RemoveItem (Variables|Transient|GetEnemyActors) UnregisteredActor))
                (if _returnvalue
                  _)))))))
    (:then_1
      (return))))

(fn HandleEventStateChanged
  (params (EventState :type ToyFoxEventStateEnum%20Enum :default "NewEnumerator0"))
  (bind _returnvalue (Utilities|Enum|Equal(Enum) EventState "NewEnumerator3"))
  (bind _athenaspecialactor (Variables|Default|GetAthenaSpecialActor))
  ;[0.02478 0.1372 0.15 0.5] HandleEventStateChanged: Enable the stand in pilots perception only in the possessed state.
  (Networking|SwitchHasAuthority
    (:Authority
      (StandinPilots|SetSecondaryStandinPilotsPerceptionEnabled _returnvalue)
      (if _returnvalue
        (Components|Activation|Activate _athenaspecialactor false)
        (else
          (Components|Activation|Deactivate _athenaspecialactor))))))

(fn HandleNearbyActorsPerceptionUpdated
  (params (AllFoundActors :type Array%20of%20Actor%20Object%20References) (EnteringActors :type Array%20of%20Actor%20Object%20References) (ExitingActors :type Array%20of%20Actor%20Object%20References))
  (bind _allnearbyperceptionfoundactors (Variables|Transient|GetAllNearbyPerceptionFoundActors))
  ;[0.008894 0.15 0.1448 0.5] HandleNearbyActorsPerceptionUpdated: Update the array that tracks the nearby actors found.
  (for _array_element EnteringActors
    (Utilities|Array|AddUnique _allnearbyperceptionfoundactors _array_element))
  (for _array_element_2 ExitingActors
    (Utilities|Array|RemoveItem _allnearbyperceptionfoundactors _array_element_2)))

(fn HandlePawnMeshHit
  (params (HitActor :type Actor%20Object%20Reference) (Hit :type Hit%20Result%20Structure%20%28by%20ref%29))
  (bind (_bblockinghit _binitialoverlap _time _distance _location _impactpoint) (Collision|BreakHitResult Hit))
  (Networking|SwitchHasAuthority
    (:Authority
      (Utilities|Casting|CastToFortPlayerPawn HitActor)
      (bind _component (Default|GetGameStateComponent)
        (:Valid
          (if (Utilities|Enum|Equal(Enum) (Event|GetEventState _component) "NewEnumerator2")
            (if (> (ScalableFloat|GetCachedScalableFloatValue (Variables|Hotfix|GetHotfix_KneelKnockback_Enabled)) 0)
              (if (> (.z _impactpoint) (.z (Transformation|GetSocketLocation (Variables|Default|GetToyFoxSkeletalMeshComponentBudgeted) "pelvis")))
                (Fortnite|ApplyGameplayEffectsToActor HitActor (Utilities|Array|MakeArray "/ToyFoxGameplay/Gameplay/Ability/Knockback/KneelKnockback/GE_ToyFox_KneelKnockback.GE_ToyFox_KneelKnockback_C") (Transformation|GetActorLocation) false 0 (Ability|GetFortAbilitySystemComponent self) (GameplayEffects|MakeFortEffectContext (Ability|GetFortAbilitySystemComponent self) (select (Variables|Default|GetbCollideWithActualMesh) (Variables|Default|GetToyFoxSkeletalMeshComponentBudgeted) (Variables|Character|GetMesh)) self 0 0))))))))))

(fn IsLocalPlayerPilot
  ;[0.1168 0 0.15 0.5] Returns if a local player is the pilot.
  (Utilities|FlowControl|Sequence
    (:then_0
      (Networking|CanExecuteCosmeticEvents
        (:True
          (bind _component (Default|GetGameStateComponent)
            (:Valid
              (for _array_element (Pilots|GetPilotPlayerStates _component)
                (if (Class|FortPlayerState|HasLocalPlayerController _array_element)
                  (return true))))))))
    (:then_1
      (return false))))

(fn IsValidWeakpointDamage
  (params (Instigator :type Controller%20Object%20Reference))
  ;[0 0.1339 0.15 0.5] Weakpoint damage should only be registered for game participants.
  (Utilities|FlowControl|Sequence
    (:then_0
      (Utilities|IsValid Instigator
        (:Is%20Valid
          (if (!= Instigator (Pawn|GetController))
            (bind _asfort_player_state_athena (Utilities|Casting|CastToFortPlayerStateAthena (Variables|/Script/Engine.Controller|GetPlayerState Instigator)))
            (return (Class|FortPlayerStateAthena|IsAnAthenaGameParticipant _asfort_player_state_athena))))))
    (:then_1
      (return false))))

(fn KnockbackOverlappedActor
  (params (OverlapActor :type Actor%20Object%20Reference))
  ;[0.05974 0.15 0.1418 0.5] KnockbackOverlappedActor: Apply knockback to valid overlapping actors.
  (Networking|SwitchHasAuthority
    (:Authority
      (Utilities|IsValid OverlapActor
        (:Is%20Valid
          (if (or (Utilities|IsA OverlapActor "/Script/FortniteGame.FortPawn") (Utilities|IsA OverlapActor "/Script/FortniteGame.FortPhysicsPawn"))
            (Fortnite|ApplyGameplayEffectsToActor OverlapActor (Utilities|Array|MakeArray (Variables|OverlapKnockback|GetOverlapKnockbackGameplayEffectClass)) (Transformation|GetActorLocation) false 0 (Ability|GetFortAbilitySystemComponent self) (GameplayEffects|MakeFortEffectContext (Ability|GetFortAbilitySystemComponent self) (select (Variables|Default|GetbCollideWithActualMesh) (Variables|Default|GetToyFoxSkeletalMeshComponentBudgeted) (Variables|Character|GetMesh)) self 0 0))))))))

(fn MakeGameplayCueParamForFootstepImpact
  (params (HitResult :type Hit%20Result%20Structure%20%28by%20ref%29))
  (bind (_normalizedmagnitude _rawmagnitude _effectcontext _matchedtagname _originaltag _aggregatedsourcetags _aggregatedtargettags _location _normal _instigator _effectcauser _sourceobject _physicalmaterial _gameplayeffectlevel _abilitylevel _targetattachcomponent _breplicatelocationwhenusingminimalrepproxy) (Ability|GameplayCue|BreakGameplayCueParameters (GameplayCue|MakeGameplayCueParametersfromHitResult HitResult)))
  (bind _output_get (LocalVariables|SetEffectContext (Class|AthenaFunctionLibrary|MakeSimpleFortEffectContext self self)))
  (Ability|EffectContext|AddHitResult _output_get HitResult)
  (return (Ability|GameplayCue|MakeGameplayCueParameters _normalizedmagnitude _rawmagnitude _output_get _matchedtagname _originaltag _aggregatedsourcetags _aggregatedtargettags _location _normal _instigator _effectcauser _sourceobject _physicalmaterial _gameplayeffectlevel _abilitylevel _targetattachcomponent _breplicatelocationwhenusingminimalrepproxy)))

(fn OnRep_bIntroCollisionState
  ;[4.8e-05 0.04651 0.15 0.5] OnRep_bIntroCollisionState: Update the collision response channel based on the latest intro collision state.
  (Fortnite|FortShippingLog self (Utilities|String|Append "OnRep_bIntroCollisionState: Updated to " (Utilities|String|ToString(Boolean) (Variables|Transient|GetbIntroCollisionState))))
  (Intro|RefreshIntroCollisionState))

(fn PlayWeakpointMaterialAnimation
  (params (WeakpointComponent :type B%20Toy%20Fox%20Weakpoint%20Component%20Object%20Reference))
  (bind _toyfoxskeletalmeshcomponentbudgeted (Variables|Default|GetToyFoxSkeletalMeshComponentBudgeted))
  ;[0.02057 0 0.15 0.5] Weakpoint material animation.
  (Networking|CanExecuteCosmeticEvents
    (:True
      (Utilities|IsValid WeakpointComponent
        (:Is%20Valid
          (if (State|IsWeakpointActive WeakpointComponent)
            (Rendering|Material|SetScalarParameterValueonMaterials _toyfoxskeletalmeshcomponentbudgeted "Animation_Duration" 0.2) ; Set Duration
            (Rendering|Material|SetScalarParameterValueonMaterials _toyfoxskeletalmeshcomponentbudgeted (Utilities|Select :Index (/ (Variables|/ToyFoxGameplay/Gameplay/Weakpoints/B_ToyFoxWeakpointComponent.B_ToyFoxWeakpointComponent_C|GetWeakpointMaterialDataIndex WeakpointComponent) 4) :Option%200 "Hit01CurveTime" :Option%201 "Hit02CurveTime" :Option%202 "Hit03CurveTime" :Option%203 "Hit04CurveTime" :Option%204 "Hit05CurveTime" :Option%205 "Hit06CurveTime" :Option%206 "Hit07CurveTime" :Option%207 "Hit08CurveTime") (Utilities|Time|GetGameTimeinSeconds)) ; Passthough Time Individrually
          ))))))

(fn RefreshIntroCollisionState
  (bind _bintrocollisionstate (Variables|Transient|GetbIntroCollisionState))
  (bind _bshouldimpulsepawnsonoverlap (LocalVariables|GetbShouldImpulsePawnsOnOverlap))
  (bind _returnvalue (select _bshouldimpulsepawnsonoverlap "ECR_Overlap" "ECR_Ignore"))
  (bind _returnvalue_2 (select _bintrocollisionstate "ECR_Block" _returnvalue))
  ;[0.08385 0.001048 0.15 0.5] RefreshIntroCollisionState: Update the collision state based on the latest variables that control it.
  (LocalVariables|SetbShouldImpulsePawnsOnOverlap (and (> (ScalableFloat|GetCachedScalableFloatValue (Variables|Hotfix|GetHotfix_ImpulsePawnsOnOverlap)) 0) (not (Variables|Transient|GetbIntroCollisionState))))
  (for _array_element (Utilities|Array|MakeArray (select (Variables|Default|GetbCollideWithActualMesh) (Variables|Default|GetToyFoxSkeletalMeshComponentBudgeted) (Variables|Character|GetMesh)) (Variables|Default|GetWeaponSkeletalMeshComponentBudgeted))
    (Collision|SetCollisionResponsetoChannel _array_element "ECC_Pawn" _returnvalue_2)
    (Collision|SetCollisionResponsetoChannel _array_element "ECC_Vehicle" _returnvalue_2)
    (Collision|SetCollisionResponsetoChannel _array_element "ECC_PhysicsBody" (select (Variables|Transient|GetbIntroCollisionState) "ECR_Block" "ECR_Ignore"))
    (Collision|SetCollisionResponsetoChannel _array_element "ECC_Camera" (select (Variables|Transient|GetbIntroCollisionState) "ECR_Block" "ECR_Ignore")) ; Solves weird camera clipping issue when another player overlaps the pawn.
    (Networking|SwitchHasAuthority
      (:Authority
        (Variables|/Script/Engine.PrimitiveComponent|SetbGenerateOverlapEvents (LocalVariables|GetbShouldImpulsePawnsOnOverlap) _array_element) ; Overlap doesn't matter on the client.\nNo special behavior.
      ))))

(fn SetIntroCollisionStateEnabled
  (params (bIntroCollisionEnabled :type Boolean))
  ;[0 0.09375 0.07477 0.5] SetIntroCollisionStateEnabled: Intro collision state enabled/disabled. Mostly useful for the kneeling pose.
  ; Managed from GA_ToyFox_Intro.
  (Networking|SwitchHasAuthority
    (:Authority
      (bind _updated_variable (Utilities|FlowControl|DeltaGateWildcard (Variables|Transient|GetbIntroCollisionState) bIntroCollisionEnabled)
        (:Success
          (Variables|Transient|SetbIntroCollisionState _updated_variable))))))

(fn SetLaserBeamHitResult
  (params (HitResult :type Hit%20Result%20Structure))
  (Variables|Transient|SetLaserBeamHitResult HitResult) ; Skips replicating to owner.
)

(fn SetLatestFootstepHitLocation
  (params (Location :type Vector))
  (Variables|Transient|SetLatestFootstepHitLocation Location))

(fn SetPilotCinematicHeadMaterial
  (params (bEnable :type Boolean))
  (bind _component (Components|GetToyFoxSkeletalMeshComponent))
  ;[0 0.08922 0.15 0.5] SetPilotCinematicHeadMaterial: Updates the toyfox head material for the cinematic event.
  ; This is called directly from the sequencer.
  (Utilities|IsValid _component
    (:Is%20Valid
      (Rendering|Material|SetMaterial _component 4 (select bEnable (Variables|Cinematic|GetCinematicMaterialIndex4) (Variables|Cinematic|Transient|GetDefaultMaterialIndex4))) ; Cinematic HD head
      (Rendering|Material|SetMaterial _component 3 (select bEnable (Variables|Cinematic|GetCinematicMaterialIndex3) (Variables|Cinematic|Transient|GetDefaultMaterialIndex3))) ; LOD0 Head
      (Rendering|Material|SetMaterial _component 8 (select bEnable (Variables|Cinematic|GetCinematicMaterialIndex8) (Variables|Cinematic|Transient|GetDefaultMaterialIndex8))) ; LOD1Head
    )))

(fn SetSecondaryStandInPilotsPerceptionEnabled
  (params (bEnable :type Boolean))
  ;[0 0.15 0.145 0.5] SetSecondaryStandInPilotsPerceptionEnabled: Enable perception for the stand in pilots.
  (Networking|SwitchHasAuthority
    (:Authority
      (Query|BPUnregisterListener (Variables|Transient|GetNearbyActorsHandle))
      (Utilities|Array|Clear (Variables|Transient|GetAllNearbyPerceptionFoundActors))
      (if bEnable
        (bind _returnvalue (Query|BPRegisterListener (Variables|Default|GetFortNearbyActorsPerception) (Variables|SecondaryPilots|StandInPilots|GetNearbyActorsPerceptionForStandInPilots) (EventDispatchers|CreateEvent :!event HandleNearbyActorsPerceptionUpdated)))
        (Variables|Transient|SetNearbyActorsHandle _returnvalue)))))

(fn StaggerGiantPawn
  (params (InstigatorAbilitySystemComp :type Ability%20System%20Component%20Object%20Reference) (EffectContextHandle :type Gameplay%20Effect%20Context%20Handle%20Structure))
  ;[0.07051 0 0.15 0.5] Apply the stagger gameplay effect to the giant player.
  (Networking|SwitchHasAuthority
    (:Authority
      (Fortnite|ApplyGameplayEffectsToActor self (Utilities|Array|MakeArray (Variables|Settings|GetStaggerGameplayEffect)) (0.0 0.0 0.0) false 0 InstigatorAbilitySystemComp EffectContextHandle))))

(fn ConstructionScript
  (|CallParentFunction)
  (Components|SkinnedMesh|SetPhysicsAsset (select (Variables|Default|GetbCollideWithActualMesh) (Variables|Default|GetToyFoxSkeletalMeshComponentBudgeted) (Variables|Character|GetMesh)) "/ToyFoxGameplay/Assets/Meshes/ToyFox/M_LRG_ToyFox_PhysicsAsset.M_LRG_ToyFox_PhysicsAsset" true) ; Set the appropriate physics asset for collision.
  (Variables|Cinematic|Transient|SetDefaultMaterialIndex3 (Rendering|Material|GetMaterial (Components|GetToyFoxSkeletalMeshComponent) 3))
  (Variables|Cinematic|Transient|SetDefaultMaterialIndex4 (Rendering|Material|GetMaterial (Components|GetToyFoxSkeletalMeshComponent) 4)) ; Cache the default material.
  (Variables|Cinematic|Transient|SetDefaultMaterialIndex8 (Rendering|Material|GetMaterial (Components|GetToyFoxSkeletalMeshComponent) 8))
  (Rendering|Material|SetScalarParameterValueonMaterials (Variables|Default|GetToyFoxSkeletalMeshComponentBudgeted) (Variables|Eye|GetEyeIntensityParamName) (Variables|Eye|GetEyeIntensityMin)))
