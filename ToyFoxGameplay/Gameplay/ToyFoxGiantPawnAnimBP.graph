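# Animation Blueprint for the giant ToyFox pawn, handling locomotion, a cannon ability and aiming.
# A Locomotion state machine covers turn-in-place, walk starts, stops and pivots, choosing clips by foot and direction.
# Head aim offset, Cannon Ability and UpperBody slot layer over it, then an upper-body aim offset follows.
# The FullBody slot feeds root and pelvis twist adjustments, giant-pawn slope warping with foot planting, leg IK and inertialization.
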AnimBlueprint /ToyFoxGameplay/Gameplay/ToyFoxGiantPawnAnimBP
target_skeleton=/ToyFoxGameplay/Assets/Meshes/ToyFox/M_LRG_ToyFox_Skeleton
graphs [
 AnimGraph {nodes [
   n1 = Use cached pose 'PostAO'()
   n2 = "Transform (Modify) Bone - Bone: root"(
    ComponentPose=Local To Component(LocalPose=Slot 'FullBody'(Source=@n1) {bAlwaysUpdateSourcePose=true, slotName=FullBody})
    Translation="0,0,0"
    Rotation="0,0,0"
   ) {rotationMode=BMM_Replace, rotationSpace=BCS_WorldSpace, alphaBoolBlend {blendOption=Linear}, boneToModify {boneName=root}}
   n3 = "Transform (Modify) Bone - Bone: pelvis"!(
    ComponentPose=@n2
    Alpha=1
    Translation="0,0,-1"
    Rotation_Roll:float=0
    Rotation_Pitch:float=0
    Rotation_Yaw:float=Get CustomMeleeTwistYaw() -> (CustomMeleeTwistYaw:float)
    Scale="1,1,1"
   ) -> (Pose) {
    rotationMode=BMM_Additive
    translationMode=BMM_Additive
    alphaBoolBlend {blendOption=Linear}
    boneToModify {boneName=pelvis}
    translation {z=-1}
   }
   n4 = Giant Pawn Slope Warping!(ComponentPose:Component Space Pose Link Structure=@n3, Alpha:float=1) -> (Pose:Component Space Pose Link Structure) {
    is=title FortAnimGraphNode_GiantPawnSlopeWarping
    bApplyFootPlantingConstraint=true
    bApplyLimitsOnFootPlacementRotation=true
    bApplySpecificLimitsOnFootRaiseAndLower=true
    bCorrectForMeshOffset=false
    bFootPlantingConstraintAppliesToRotation=true
    bKeepMeshInsideOfCapsule=false
    bUseFloorOffsetInterpolator=false
    bUseFootHeightAsAlpha=true
    footHeightForFullAlpha=20
    footHeightForZeroAlpha=30
    footPlanting2DAlphaOutEndRadius=600
    footPlanting2DAlphaOutStartRadius=455
    footPlantingFootHeightAlphaOutEnd=14
    footPlantingFootHeightAlphaOutStart=11
    footPlantingRotationOutEndAngle=50
    leftFootPlantingAlphaOutCurve=FootSpeed_L
    leftFootPlantingDisableCurve=DisableToyFoxSlopeWarpFootPinningLeft
    maxFootHeightToStartFootPlant=11
    maxFootLowerHeight=700
    maxFootPlacementPitchAngleDegs=50
    maxFootPlacementRollAngleDegs=35
    maxFootRaiseHeight=280
    maxMovementDistBeforeNewTrace=3
    maxStepHeight=700
    pullPelvisDownMaxDistance=10
    rightFootPlantingAlphaOutCurve=FootSpeed_R
    rightFootPlantingDisableCurve=DisableToyFoxSlopeWarpFootPinningRight
    alphaBoolBlend {blendOption=Linear}
    feetDefinitions [
     {numBonesInLimb=2, fKFootBone {boneName=foot_l}, iKFootBone {boneName=ik_foot_l}}
     {numBonesInLimb=2, fKFootBone {boneName=foot_r}, iKFootBone {boneName=ik_foot_r}}
    ]
    iKFootRootBone {boneName=ik_foot_root}
    lFootPrediction {boneName=ik_foot_l}
    pelvisBone {boneName=pelvis}
    rFootPrediction {boneName=ik_foot_r}
   }
   n5 = Leg IK!(ComponentPose:Component Space Pose Link Structure=@n4, Alpha:float=1) -> (Pose:Component Space Pose Link Structure) {
    is=title AnimGraphNode_LegIK
    alphaBoolBlend {blendOption=Linear}
    legsDefinition [
     {
      bEnableKneeTwistCorrection=true
      bEnableRotationLimit=false
      footBoneForwardAxis=Y
      hingeRotationAxis=None
      minRotationAngle=15
      numBonesInLimb=2
      twistOffsetCurveName=None
      fKFootBone {boneName=foot_l}
      iKFootBone {boneName=ik_foot_l}
     }
     {
      bEnableKneeTwistCorrection=true
      bEnableRotationLimit=false
      footBoneForwardAxis=Y
      hingeRotationAxis=None
      minRotationAngle=15
      numBonesInLimb=2
      twistOffsetCurveName=None
      fKFootBone {boneName=foot_r}
      iKFootBone {boneName=ik_foot_r}
     }
    ]
   }
   Output Pose(Result=Inertialization!(Source:Pose Link Structure=Component To Local(ComponentPose=@n5)) -> (Pose:Pose Link Structure) {is=title AnimGraphNode_Inertialization})
   AnimGraphNode_StateMachine_1 = Locomotion() {is=title AnimGraphNode_StateMachine, bAllowConduitEntryStates=true}
   Save cached pose 'Locomotion'(Pose=Inertialization(Source=@AnimGraphNode_StateMachine_1)) {cacheName=Locomotion}
   Save cached pose 'PostUpperBodySlot'(Pose=Slot 'UpperBody'(Source=Use cached pose 'PreUpperBodySlot'()) {bAlwaysUpdateSourcePose=true, slotName=UpperBody}) {cacheName=PostUpperBodySlot}
   Save cached pose 'LocoPlusIdleAdditive'(Pose=Apply Additive(
     Base=Use cached pose 'Locomotion'()
     Additive="Sequence Player (additive) 'NoWep_Stand_Idle_Loop_Add'"() {sequence=/ToyFoxGameplay/Assets/Animations/Locomotion/NoWep_Stand_Idle_Loop_Add}
    ) {alphaBoolBlend {blendOption=Linear}}) {cacheName=LocoPlusIdleAdditive}
   AnimGraphNode_StateMachine_2 = Cannon Ability() {is=title AnimGraphNode_StateMachine}
   Save cached pose 'PreUpperBodySlot'(Pose=Layered blend per bone(BasePose=Use cached pose 'PostHeadAO'(), BlendPoses_0=@AnimGraphNode_StateMachine_2) {bMeshSpaceRotationBlend=true, layerSetup [{branchFilters [{boneName=Spine_01}]}]}) {cacheName=PreUpperBodySlot}
   n6 = Use cached pose 'PostUpperBodySlot'()
   n7 = Use cached pose 'PreUpperBodySlot'()
   n8 = Layered blend per bone(
    BasePose=Blend Poses by bool(
     BlendPose_0=Layered blend per bone(BasePose=@n7, BlendPoses_0=@n6) {layerSetup [{branchFilters [{boneName=spine_05}]}]}
     BlendPose_1=Use cached pose 'PostUpperBodySlot'()
     bActiveValue=Get bIsFirstPersonView() -> (bIsFirstPersonView:bool)
    ) {blendType=Linear}
    BlendPoses_0=Use cached pose 'Locomotion'()
   ) {
    bMeshSpaceRotationBlend=true
    curveBlendOption=BlendByWeight
    layerSetup [{branchFilters [{boneName=Root}, {blendDepth=-1, boneName=Ik_hand_root}, {blendDepth=-2, boneName=Spine_01}]}]
   }
   Save cached pose 'PostLocoLayering'(Pose=@n8) {cacheName=PostLocoLayering}
   Save cached pose 'PostHeadAO'(Pose=AimOffset Player 'M_NoWep_Relaxed_AO'!(
     X:float=Get AimYaw_AO() -> (AimYaw_AO:double)
     Y:float=Get AimPitch_AO() -> (AimPitch_AO:float)
     BasePose:Pose Link Structure=Use cached pose 'LocoPlusIdleAdditive'()
     Alpha:float=1
    ) -> (Pose:Pose Link Structure) {blendSpace=/ToyFoxGameplay/Assets/Animations/AimOffset/M_NoWep_Relaxed_AO, alphaBoolBlend {blendOption=Linear}}) {cacheName=PostHeadAO}
   Save cached pose 'PostAO'(Pose=AimOffset Player 'ToyFox_UpperBody_AO'!(X:float=Get AimYaw_AO(), Y:float=Get AimPitch_AO(), BasePose:Pose Link Structure=Use cached pose 'PostLocoLayering'(), Alpha:float=1) -> (Pose:Pose Link Structure) {blendSpace=/ToyFoxGameplay/Assets/Animations/ToyFox_UpperBody_AO, alphaBoolBlend {blendOption=Linear}}) {cacheName=PostAO}
  ]}
 Locomotion {
  entry=Idle
  path=AnimGraph.AnimGraphNode_StateMachine_1.Locomotion
  ids=id
  nodes [
   DoWalkStopAndTurn() {is=title AnimStateAliasNode, also {aliases [Pivot,Walk,WalkStart,WalkStop]}}
   Pivot Alias() {is=title AnimStateAliasNode, also {aliases [Walk,WalkStart,WalkStop]}}
   AnimStateNode_0 = Idle() {
    stateEntered {guid=#1}
    stateFullyBlended {guid=#2}
    stateLeft {guid=#3}
   }
   AnimStateNode_1 = Walk() {
    is=title AnimStateNode
    stateEntered {guid=#4}
    stateFullyBlended {guid=#5}
    stateLeft {guid=#6}
   }
   AnimStateNode_2 = WalkStart() {
    is=title AnimStateNode
    stateEntered {guid=#7}
    stateFullyBlended {guid=#8}
    stateLeft {guid=#9}
   }
   AnimStateNode_3 = WalkStop() {is=title AnimStateNode, bAlwaysResetOnEntry=true}
   AnimStateNode_4 = Pivot() {is=title AnimStateNode, bAlwaysResetOnEntry=true}
   AnimStateTransitionNode_0 = AnimStateTransitionNode() {bDisabled=true}
   AnimStateTransitionNode_1 = AnimStateTransitionNode() {bDisabled=true}
   AnimStateTransitionNode_10 = AnimStateTransitionNode() {blendMode=CubicInOut, crossfadeDuration=0.8}
   AnimStateTransitionNode_11 = AnimStateTransitionNode() {
    blendMode=CubicInOut
    crossfadeDuration=0.6
    blendProfileWrapper {
     bIsSkeletonBlendProfile=false
     blendProfile=&self:AnimGraph.AnimGraphNode_StateMachine_1.Locomotion.AnimStateTransitionNode_11.BlendProfile_0
     blendProfileProvider=BlendProfileStandaloneProvider'&self:AnimGraph.AnimGraphNode_StateMachine_1.Locomotion.AnimStateTransitionNode_11.BlendProfileStandaloneProvider_0'
    }
   }
   AnimStateTransitionNode_12 = AnimStateTransitionNode() {
    blendMode=CubicInOut
    crossfadeDuration=0.8
    blendProfileWrapper {
     bIsSkeletonBlendProfile=false
     blendProfile=&self:AnimGraph.AnimGraphNode_StateMachine_1.Locomotion.AnimStateTransitionNode_12.BlendProfile_0
     blendProfileProvider=BlendProfileStandaloneProvider'&self:AnimGraph.AnimGraphNode_StateMachine_1.Locomotion.AnimStateTransitionNode_12.BlendProfileStandaloneProvider_0'
    }
   }
   AnimStateTransitionNode_13 = AnimStateTransitionNode() {
    blendMode=CubicInOut
    crossfadeDuration=1.4
    blendProfileWrapper {
     bIsSkeletonBlendProfile=false
     blendProfile=&self:AnimGraph.AnimGraphNode_StateMachine_1.Locomotion.AnimStateTransitionNode_13.BlendProfile_0
     blendProfileProvider=BlendProfileStandaloneProvider'&self:AnimGraph.AnimGraphNode_StateMachine_1.Locomotion.AnimStateTransitionNode_13.BlendProfileStandaloneProvider_0'
    }
   }
   AnimStateTransitionNode_15 = AnimStateTransitionNode() {bDisabled=true}
   AnimStateTransitionNode_2 = AnimStateTransitionNode() {
    blendMode=CubicInOut
    crossfadeDuration=0.3
    transitionEnd {guid=#10}
    transitionInterrupt {guid=#11}
    transitionStart {guid=#12}
   }
   AnimStateTransitionNode_3 = AnimStateTransitionNode() {crossfadeDuration=0.3, logicType=TLT_Inertialization}
   AnimStateTransitionNode_4 = AnimStateTransitionNode() {
    blendMode=CubicInOut
    crossfadeDuration=0.8
    blendProfileWrapper {
     bIsSkeletonBlendProfile=false
     blendProfile=&self:AnimGraph.AnimGraphNode_StateMachine_1.Locomotion.AnimStateTransitionNode_4.BlendProfile_0
     blendProfileProvider=BlendProfileStandaloneProvider'&self:AnimGraph.AnimGraphNode_StateMachine_1.Locomotion.AnimStateTransitionNode_4.BlendProfileStandaloneProvider_0'
    }
    transitionEnd {guid=#13}
    transitionInterrupt {guid=#14}
    transitionStart {guid=#15}
   }
   AnimStateTransitionNode_5 = AnimStateTransitionNode() {
    blendMode=CubicInOut
    crossfadeDuration=0.8
    blendProfileWrapper {
     bIsSkeletonBlendProfile=false
     blendProfile=&self:AnimGraph.AnimGraphNode_StateMachine_1.Locomotion.AnimStateTransitionNode_5.BlendProfile_0
     blendProfileProvider=BlendProfileStandaloneProvider'&self:AnimGraph.AnimGraphNode_StateMachine_1.Locomotion.AnimStateTransitionNode_5.BlendProfileStandaloneProvider_0'
    }
    transitionEnd {guid=#16}
    transitionInterrupt {guid=#17}
    transitionStart {guid=#18}
   }
   AnimStateTransitionNode_6 = AnimStateTransitionNode() {
    blendMode=Linear
    crossfadeDuration=0.6
    blendProfileWrapper {
     bIsSkeletonBlendProfile=false
     blendProfile=&self:AnimGraph.AnimGraphNode_StateMachine_1.Locomotion.AnimStateTransitionNode_6.BlendProfile_0
     blendProfileProvider=BlendProfileStandaloneProvider'&self:AnimGraph.AnimGraphNode_StateMachine_1.Locomotion.AnimStateTransitionNode_6.BlendProfileStandaloneProvider_0'
    }
   }
   AnimStateTransitionNode_7 = AnimStateTransitionNode() {bAutomaticRuleBasedOnSequencePlayerInState=true, bDisabled=true, crossfadeDuration=0.5}
   AnimStateTransitionNode_8 = AnimStateTransitionNode() {bAutomaticRuleBasedOnSequencePlayerInState=true, crossfadeDuration=0.3}
   AnimStateTransitionNode_9 = AnimStateTransitionNode()
  ]
  links [
   {
    dst=Idle
    rule=AnimGraph.AnimGraphNode_StateMachine_1.Locomotion.AnimStateTransitionNode_10.Transition
    src=DoWalkStopAndTurn
    via=AnimStateTransitionNode_10
   }
   {
    dst=Walk
    rule=AnimGraph.AnimGraphNode_StateMachine_1.Locomotion.AnimStateTransitionNode_0.Transition
    src=Idle
    via=AnimStateTransitionNode_0
   }
   {
    dst=WalkStart
    rule=AnimGraph.AnimGraphNode_StateMachine_1.Locomotion.AnimStateTransitionNode_2.Transition
    src=Idle
    via=AnimStateTransitionNode_2
   }
   {
    dst=Walk
    rule=AnimGraph.AnimGraphNode_StateMachine_1.Locomotion.AnimStateTransitionNode_13.Transition
    src=Pivot
    via=AnimStateTransitionNode_13
   }
   {
    dst=WalkStop
    rule=AnimGraph.AnimGraphNode_StateMachine_1.Locomotion.AnimStateTransitionNode_12.Transition
    src=Pivot
    via=AnimStateTransitionNode_12
   }
   {
    dst=Pivot
    rule=AnimGraph.AnimGraphNode_StateMachine_1.Locomotion.AnimStateTransitionNode_11.Transition
    src=Pivot Alias
    via=AnimStateTransitionNode_11
   }
   {
    dst=WalkStop
    rule=AnimGraph.AnimGraphNode_StateMachine_1.Locomotion.AnimStateTransitionNode_1.Transition
    src=Walk
    via=AnimStateTransitionNode_1
   }
   {
    dst=WalkStop
    rule=AnimGraph.AnimGraphNode_StateMachine_1.Locomotion.AnimStateTransitionNode_3.Transition
    src=Walk
    via=AnimStateTransitionNode_3
   }
   {
    dst=Walk
    rule=AnimGraph.AnimGraphNode_StateMachine_1.Locomotion.AnimStateTransitionNode_6.Transition
    src=WalkStart
    via=AnimStateTransitionNode_6
   }
   {
    dst=WalkStop
    rule=AnimGraph.AnimGraphNode_StateMachine_1.Locomotion.AnimStateTransitionNode_4.Transition
    src=WalkStart
    via=AnimStateTransitionNode_4
   }
   {
    dst=Idle
    rule=AnimGraph.AnimGraphNode_StateMachine_1.Locomotion.AnimStateTransitionNode_7.Transition
    src=WalkStop
    via=AnimStateTransitionNode_7
   }
   {
    dst=Idle
    rule=AnimGraph.AnimGraphNode_StateMachine_1.Locomotion.AnimStateTransitionNode_8.Transition
    src=WalkStop
    via=AnimStateTransitionNode_8
   }
   {
    dst=Idle
    rule=AnimGraph.AnimGraphNode_StateMachine_1.Locomotion.AnimStateTransitionNode_9.Transition
    src=WalkStop
    via=AnimStateTransitionNode_9
   }
   {
    dst=Walk
    rule=AnimGraph.AnimGraphNode_StateMachine_1.Locomotion.AnimStateTransitionNode_15.Transition
    src=WalkStop
    via=AnimStateTransitionNode_15
   }
   {
    dst=WalkStart
    rule=AnimGraph.AnimGraphNode_StateMachine_1.Locomotion.AnimStateTransitionNode_5.Transition
    src=WalkStop
    via=AnimStateTransitionNode_5
   }
  ]
 }
 Idle {
  path=AnimGraph.AnimGraphNode_StateMachine_1.Locomotion.AnimStateNode_0.Idle
  nodes [
   AnimGraphNode_StateMachine_0 = Idle_TurnInPlace() {is=title AnimGraphNode_StateMachine}
   Output Animation Pose(Result=@AnimGraphNode_StateMachine_0)
  ]
 }
 Idle_TurnInPlace {
  entry=Idle
  path=AnimGraph.AnimGraphNode_StateMachine_1.Locomotion.AnimStateNode_0.Idle.AnimGraphNode_StateMachine_0.Idle_TurnInPlace
  ids=id
  nodes [
   AnimStateNode_0 = Idle()
   AnimStateNode_1 = TurnInPlaceLeft() {is=title AnimStateNode}
   AnimStateNode_3 = TurnInPlaceRight() {is=title AnimStateNode}
   AnimStateTransitionNode_0 = AnimStateTransitionNode() {crossfadeDuration=0.4}
   AnimStateTransitionNode_1 = AnimStateTransitionNode() {crossfadeDuration=0.4}
   AnimStateTransitionNode_2 = AnimStateTransitionNode() {bAutomaticRuleBasedOnSequencePlayerInState=true, crossfadeDuration=0.4}
   AnimStateTransitionNode_3 = AnimStateTransitionNode() {bAutomaticRuleBasedOnSequencePlayerInState=true, crossfadeDuration=0.4}
  ]
  links [
   {
    dst=TurnInPlaceLeft
    rule=AnimGraph.AnimGraphNode_StateMachine_1.Locomotion.AnimStateNode_0.Idle.AnimGraphNode_StateMachine_0.Idle_TurnInPlace.AnimStateTransitionNode_0.Transition
    src=Idle
    via=AnimStateTransitionNode_0
   }
   {
    dst=TurnInPlaceRight
    rule=AnimGraph.AnimGraphNode_StateMachine_1.Locomotion.AnimStateNode_0.Idle.AnimGraphNode_StateMachine_0.Idle_TurnInPlace.AnimStateTransitionNode_1.Transition
    src=Idle
    via=AnimStateTransitionNode_1
   }
   {
    dst=Idle
    rule=AnimGraph.AnimGraphNode_StateMachine_1.Locomotion.AnimStateNode_0.Idle.AnimGraphNode_StateMachine_0.Idle_TurnInPlace.AnimStateTransitionNode_2.Transition
    src=TurnInPlaceLeft
    via=AnimStateTransitionNode_2
   }
   {
    dst=Idle
    rule=AnimGraph.AnimGraphNode_StateMachine_1.Locomotion.AnimStateNode_0.Idle.AnimGraphNode_StateMachine_0.Idle_TurnInPlace.AnimStateTransitionNode_3.Transition
    src=TurnInPlaceRight
    via=AnimStateTransitionNode_3
   }
  ]
 }
 Idle {
  path=AnimGraph.AnimGraphNode_StateMachine_1.Locomotion.AnimStateNode_0.Idle.AnimGraphNode_StateMachine_0.Idle_TurnInPlace.AnimStateNode_0.Idle
  nodes [Output Animation Pose(Result=Sequence Player 'ToyFox_LRG_Idle_01'() {sequence=/ToyFoxGameplay/Assets/Animations/ToyFox_LRG_Idle_01})]
 }
 TurnInPlaceLeft {
  path=AnimGraph.AnimGraphNode_StateMachine_1.Locomotion.AnimStateNode_0.Idle.AnimGraphNode_StateMachine_0.Idle_TurnInPlace.AnimStateNode_1.TurnInPlaceLeft
  nodes [Output Animation Pose(Result=Blendspace Player 'ToyFox_TurnInPlace_L_BS'!(X=Get TurnInPlaceAmount() -> (TurnInPlaceAmount:double), Y=0, PlayRate:float=Get SpeedScaleConstant() -> (SpeedScaleConstant:double)) -> (Pose) {bLoop=false, blendSpace=/ToyFoxGameplay/Assets/Animations/Locomotion/ToyFox_TurnInPlace_L_BS}) {
    node {stateExitFunction {functionName=OnTurnInPlaceExit}}
    stateExitFunction {bSelfContext=true, memberGuid=#19, memberName=OnTurnInPlaceExit}
   }]
 }
 TurnInPlaceRight {
  path=AnimGraph.AnimGraphNode_StateMachine_1.Locomotion.AnimStateNode_0.Idle.AnimGraphNode_StateMachine_0.Idle_TurnInPlace.AnimStateNode_3.TurnInPlaceRight
  nodes [Output Animation Pose(Result=Blendspace Player 'ToyFox_TurnInPlace_R_BS'!(X=Get TurnInPlaceAmount(), Y=0, PlayRate:float=Get SpeedScaleConstant()) -> (Pose) {bLoop=false, blendSpace=/ToyFoxGameplay/Assets/Animations/Locomotion/ToyFox_TurnInPlace_R_BS}) {
    node {stateExitFunction {functionName=OnTurnInPlaceExit}}
    stateExitFunction {bSelfContext=true, memberGuid=#19, memberName=OnTurnInPlaceExit}
   }]
 }
 Transition {
  path=AnimGraph.AnimGraphNode_StateMachine_1.Locomotion.AnimStateNode_0.Idle.AnimGraphNode_StateMachine_0.Idle_TurnInPlace.AnimStateTransitionNode_0.Transition
  nodes [Result(bCanEnterTransition=Get bIsTurnInPlaceLeft() -> (bIsTurnInPlaceLeft:bool))]
 }
 Transition {
  path=AnimGraph.AnimGraphNode_StateMachine_1.Locomotion.AnimStateNode_0.Idle.AnimGraphNode_StateMachine_0.Idle_TurnInPlace.AnimStateTransitionNode_1.Transition
  nodes [Result(bCanEnterTransition=Get bIsTurnInPlaceRight() -> (bIsTurnInPlaceRight:bool))]
 }
 Transition {
  path=AnimGraph.AnimGraphNode_StateMachine_1.Locomotion.AnimStateNode_0.Idle.AnimGraphNode_StateMachine_0.Idle_TurnInPlace.AnimStateTransitionNode_2.Transition
  nodes [Result()]
 }
 Transition {
  path=AnimGraph.AnimGraphNode_StateMachine_1.Locomotion.AnimStateNode_0.Idle.AnimGraphNode_StateMachine_0.Idle_TurnInPlace.AnimStateTransitionNode_3.Transition
  nodes [Result()]
 }
 Walk {
  path=AnimGraph.AnimGraphNode_StateMachine_1.Locomotion.AnimStateNode_1.Walk
  nodes [
   n1 = Blendspace Player 'ToyFox_NoWep_WalkLoop_Fwd_BS'!(X=0, Y=0, PlayRate:float=1) -> (Pose) {blendSpace=/ToyFoxGameplay/Assets/Animations/Locomotion/ToyFox_NoWep_WalkLoop_Fwd_BS}
   n2 = Blendspace Player 'ToyFox_NoWep_WalkLoop_Bwd_BS'!(X=0, Y=0, PlayRate:float=1) -> (Pose) {blendSpace=/ToyFoxGameplay/Assets/Animations/Locomotion/ToyFox_NoWep_WalkLoop_Bwd_BS}
   n3 = Get bWalk_UseFwdBlendspace() -> (bWalk_UseFwdBlendspace:bool)
   Output Animation Pose(Result=Apply Additive(
     Base=Blend Poses by bool(BlendPose_0=@n1, BlendPose_1=@n2, BlendTime_0=0.8, BlendTime_1=0.8, bActiveValue=@n3) {blendType=Linear}
     Additive="Sequence Player (additive) 'ToyFox_Idle_Pose_Add_NoWepRelative'"() {bLoopAnimation=false, playRate=0, sequence=/ToyFoxGameplay/Assets/Animations/ToyFox_Idle_Pose_Add_NoWepRelative}
    ) {alphaBoolBlend {blendOption=Linear}}) {
    node {stateEntryFunction {functionName=OnStateEntry}}
    stateEntryFunction {bSelfContext=true, memberGuid=#20, memberName=OnStateEntry}
   }
  ]
 }
 WalkStart {
  path=AnimGraph.AnimGraphNode_StateMachine_1.Locomotion.AnimStateNode_2.WalkStart
  nodes [
   AnimGraphNode_StateMachine_0 = WalkStart_SM() {is=title AnimGraphNode_StateMachine, bAllowConduitEntryStates=true}
   Output Animation Pose(Result=Apply Additive(
     Base=@AnimGraphNode_StateMachine_0
     Additive="Sequence Player (additive) 'ToyFox_Idle_Pose_Add_NoWepRelative'"() {bLoopAnimation=false, playRate=0, sequence=&ToyFox_Idle_Pose_Add_NoWepRelative}
    ) {alphaBoolBlend {blendOption=Linear}}) {
    node {stateEntryFunction {functionName=OnStateEntry_0}}
    stateEntryFunction {bSelfContext=true, memberGuid=#21, memberName=WalkStart_OnEntry}
   }
  ]
 }
 WalkStart_SM {
  entry=Conduit
  path=AnimGraph.AnimGraphNode_StateMachine_1.Locomotion.AnimStateNode_2.WalkStart.AnimGraphNode_StateMachine_0.WalkStart_SM
  ids=id
  nodes [
   AnimStateConduitNode_0 = Conduit()
   AnimStateNode_0 = LeftFootStart() {is=title AnimStateNode}
   AnimStateNode_1 = RightFootStart() {is=title AnimStateNode}
   AnimStateTransitionNode_0 = AnimStateTransitionNode()
   AnimStateTransitionNode_1 = AnimStateTransitionNode() {priorityOrder=100}
  ]
  links [
   {
    dst=LeftFootStart
    rule=AnimGraph.AnimGraphNode_StateMachine_1.Locomotion.AnimStateNode_2.WalkStart.AnimGraphNode_StateMachine_0.WalkStart_SM.AnimStateTransitionNode_0.Transition
    src=Conduit
    via=AnimStateTransitionNode_0
   }
   {
    dst=RightFootStart
    rule=AnimGraph.AnimGraphNode_StateMachine_1.Locomotion.AnimStateNode_2.WalkStart.AnimGraphNode_StateMachine_0.WalkStart_SM.AnimStateTransitionNode_1.Transition
    src=Conduit
    via=AnimStateTransitionNode_1
   }
  ]
 }
 Conduit {
  path=AnimGraph.AnimGraphNode_StateMachine_1.Locomotion.AnimStateNode_2.WalkStart.AnimGraphNode_StateMachine_0.WalkStart_SM.AnimStateConduitNode_0.Conduit
  nodes [Result(bCanEnterTransition=true)]
 }
 LeftFootStart {
  path=AnimGraph.AnimGraphNode_StateMachine_1.Locomotion.AnimStateNode_2.WalkStart.AnimGraphNode_StateMachine_0.WalkStart_SM.AnimStateNode_0.LeftFootStart
  nodes [
   AnimGraphNode_StateMachine_0 = LeftFoot Start SM() {is=title AnimGraphNode_StateMachine, bAllowConduitEntryStates=true}
   Output Animation Pose(Result=@AnimGraphNode_StateMachine_0)
  ]
 }
 LeftFoot Start SM {
  entry=Conduit
  path=AnimGraph.AnimGraphNode_StateMachine_1.Locomotion.AnimStateNode_2.WalkStart.AnimGraphNode_StateMachine_0.WalkStart_SM.AnimStateNode_0.LeftFootStart.AnimGraphNode_StateMachine_0.LeftFoot Start SM
  ids=id
  nodes [
   AnimStateConduitNode_0 = Conduit()
   AnimStateNode_0 = Fwd Start BS() {is=title AnimStateNode}
   AnimStateNode_1 = Bwd Start BS() {is=title AnimStateNode}
   AnimStateTransitionNode_0 = AnimStateTransitionNode()
   AnimStateTransitionNode_1 = AnimStateTransitionNode() {priorityOrder=100}
  ]
  links [
   {
    dst=Bwd Start BS
    rule=AnimGraph.AnimGraphNode_StateMachine_1.Locomotion.AnimStateNode_2.WalkStart.AnimGraphNode_StateMachine_0.WalkStart_SM.AnimStateNode_0.LeftFootStart.AnimGraphNode_StateMachine_0.LeftFoot Start SM.AnimStateTransitionNode_1.Transition
    src=Conduit
    via=AnimStateTransitionNode_1
   }
   {
    dst=Fwd Start BS
    rule=AnimGraph.AnimGraphNode_StateMachine_1.Locomotion.AnimStateNode_2.WalkStart.AnimGraphNode_StateMachine_0.WalkStart_SM.AnimStateNode_0.LeftFootStart.AnimGraphNode_StateMachine_0.LeftFoot Start SM.AnimStateTransitionNode_0.Transition
    src=Conduit
    via=AnimStateTransitionNode_0
   }
  ]
 }
 Conduit {
  path=AnimGraph.AnimGraphNode_StateMachine_1.Locomotion.AnimStateNode_2.WalkStart.AnimGraphNode_StateMachine_0.WalkStart_SM.AnimStateNode_0.LeftFootStart.AnimGraphNode_StateMachine_0.LeftFoot Start SM.AnimStateConduitNode_0.Conduit
  nodes [Result(bCanEnterTransition=true)]
 }
 Fwd Start BS {
  path=AnimGraph.AnimGraphNode_StateMachine_1.Locomotion.AnimStateNode_2.WalkStart.AnimGraphNode_StateMachine_0.WalkStart_SM.AnimStateNode_0.LeftFootStart.AnimGraphNode_StateMachine_0.LeftFoot Start SM.AnimStateNode_0.Fwd Start BS
  nodes [
   n1 = Blendspace Player 'ToyFox_NoWep_WalkStart_Fwd_LFF_LFoot_BS'!(X=0, Y=0, PlayRate:float=1) -> (Pose) {blendSpace=/ToyFoxGameplay/Assets/Animations/Locomotion/ToyFox_NoWep_WalkStart_Fwd_LFF_LFoot_BS, groupName=WalkSync, method=SyncGroup}
   Output Animation Pose(Result=@n1)
  ]
 }
 Bwd Start BS {
  path=AnimGraph.AnimGraphNode_StateMachine_1.Locomotion.AnimStateNode_2.WalkStart.AnimGraphNode_StateMachine_0.WalkStart_SM.AnimStateNode_0.LeftFootStart.AnimGraphNode_StateMachine_0.LeftFoot Start SM.AnimStateNode_1.Bwd Start BS
  nodes [Output Animation Pose(Result=Blendspace Player 'ToyFox_NoWep_WalkStart_Bwd_LFF_LFoot_BS'!(X=0, Y=0, PlayRate:float=1) -> (Pose) {blendSpace=/ToyFoxGameplay/Assets/Animations/Locomotion/ToyFox_NoWep_WalkStart_Bwd_LFF_LFoot_BS})]
 }
 Transition {
  path=AnimGraph.AnimGraphNode_StateMachine_1.Locomotion.AnimStateNode_2.WalkStart.AnimGraphNode_StateMachine_0.WalkStart_SM.AnimStateNode_0.LeftFootStart.AnimGraphNode_StateMachine_0.LeftFoot Start SM.AnimStateTransitionNode_0.Transition
  nodes [Result(bCanEnterTransition=Get bWalk_UseFwdBlendspace())]
 }
 Transition {
  path=AnimGraph.AnimGraphNode_StateMachine_1.Locomotion.AnimStateNode_2.WalkStart.AnimGraphNode_StateMachine_0.WalkStart_SM.AnimStateNode_0.LeftFootStart.AnimGraphNode_StateMachine_0.LeftFoot Start SM.AnimStateTransitionNode_1.Transition
  nodes [Result(bCanEnterTransition=true)]
 }
 RightFootStart {
  path=AnimGraph.AnimGraphNode_StateMachine_1.Locomotion.AnimStateNode_2.WalkStart.AnimGraphNode_StateMachine_0.WalkStart_SM.AnimStateNode_1.RightFootStart
  nodes [
   AnimGraphNode_StateMachine_0 = RightFoot Start SM() {is=title AnimGraphNode_StateMachine, bAllowConduitEntryStates=true}
   Output Animation Pose(Result=@AnimGraphNode_StateMachine_0)
  ]
 }
 RightFoot Start SM {
  entry=Conduit
  path=AnimGraph.AnimGraphNode_StateMachine_1.Locomotion.AnimStateNode_2.WalkStart.AnimGraphNode_StateMachine_0.WalkStart_SM.AnimStateNode_1.RightFootStart.AnimGraphNode_StateMachine_0.RightFoot Start SM
  ids=id
  nodes [
   AnimStateConduitNode_0 = Conduit()
   AnimStateNode_0 = Fwd Start BS()
   AnimStateNode_1 = Bwd Start BS()
   AnimStateTransitionNode_0 = AnimStateTransitionNode()
   AnimStateTransitionNode_1 = AnimStateTransitionNode() {priorityOrder=100}
  ]
  links [
   {
    dst=Bwd Start BS
    rule=AnimGraph.AnimGraphNode_StateMachine_1.Locomotion.AnimStateNode_2.WalkStart.AnimGraphNode_StateMachine_0.WalkStart_SM.AnimStateNode_1.RightFootStart.AnimGraphNode_StateMachine_0.RightFoot Start SM.AnimStateTransitionNode_1.Transition
    src=Conduit
    via=AnimStateTransitionNode_1
   }
   {
    dst=Fwd Start BS
    rule=AnimGraph.AnimGraphNode_StateMachine_1.Locomotion.AnimStateNode_2.WalkStart.AnimGraphNode_StateMachine_0.WalkStart_SM.AnimStateNode_1.RightFootStart.AnimGraphNode_StateMachine_0.RightFoot Start SM.AnimStateTransitionNode_0.Transition
    src=Conduit
    via=AnimStateTransitionNode_0
   }
  ]
 }
 Conduit {
  path=AnimGraph.AnimGraphNode_StateMachine_1.Locomotion.AnimStateNode_2.WalkStart.AnimGraphNode_StateMachine_0.WalkStart_SM.AnimStateNode_1.RightFootStart.AnimGraphNode_StateMachine_0.RightFoot Start SM.AnimStateConduitNode_0.Conduit
  nodes [Result(bCanEnterTransition=true)]
 }
 Fwd Start BS {
  path=AnimGraph.AnimGraphNode_StateMachine_1.Locomotion.AnimStateNode_2.WalkStart.AnimGraphNode_StateMachine_0.WalkStart_SM.AnimStateNode_1.RightFootStart.AnimGraphNode_StateMachine_0.RightFoot Start SM.AnimStateNode_0.Fwd Start BS
  nodes [
   n1 = Blendspace Player 'ToyFox_NoWep_WalkStart_Fwd_LFF_RFoot_BS'!(X=0, Y=0, PlayRate:float=1) -> (Pose) {blendSpace=/ToyFoxGameplay/Assets/Animations/Locomotion/ToyFox_NoWep_WalkStart_Fwd_LFF_RFoot_BS, groupName=WalkSync, method=SyncGroup}
   Output Animation Pose(Result=@n1)
  ]
 }
 Bwd Start BS {
  path=AnimGraph.AnimGraphNode_StateMachine_1.Locomotion.AnimStateNode_2.WalkStart.AnimGraphNode_StateMachine_0.WalkStart_SM.AnimStateNode_1.RightFootStart.AnimGraphNode_StateMachine_0.RightFoot Start SM.AnimStateNode_1.Bwd Start BS
  nodes [Output Animation Pose(Result=Blendspace Player 'ToyFox_NoWep_WalkStart_Bwd_LFF_RFoot_BS'!(X=0, Y=0, PlayRate:float=1) -> (Pose) {blendSpace=/ToyFoxGameplay/Assets/Animations/Locomotion/ToyFox_NoWep_WalkStart_Bwd_LFF_RFoot_BS})]
 }
 Transition {
  path=AnimGraph.AnimGraphNode_StateMachine_1.Locomotion.AnimStateNode_2.WalkStart.AnimGraphNode_StateMachine_0.WalkStart_SM.AnimStateNode_1.RightFootStart.AnimGraphNode_StateMachine_0.RightFoot Start SM.AnimStateTransitionNode_0.Transition
  nodes [Result(bCanEnterTransition=Get bWalk_UseFwdBlendspace())]
 }
 Transition {
  path=AnimGraph.AnimGraphNode_StateMachine_1.Locomotion.AnimStateNode_2.WalkStart.AnimGraphNode_StateMachine_0.WalkStart_SM.AnimStateNode_1.RightFootStart.AnimGraphNode_StateMachine_0.RightFoot Start SM.AnimStateTransitionNode_1.Transition
  nodes [Result(bCanEnterTransition=true)]
 }
 Transition {
  path=AnimGraph.AnimGraphNode_StateMachine_1.Locomotion.AnimStateNode_2.WalkStart.AnimGraphNode_StateMachine_0.WalkStart_SM.AnimStateTransitionNode_0.Transition
  nodes [Result(bCanEnterTransition=Get bIsMoveStartLFoot() -> (bIsMoveStartLFoot:bool))]
 }
 Transition {
  path=AnimGraph.AnimGraphNode_StateMachine_1.Locomotion.AnimStateNode_2.WalkStart.AnimGraphNode_StateMachine_0.WalkStart_SM.AnimStateTransitionNode_1.Transition
  nodes [Result(bCanEnterTransition=true)]
 }
 WalkStop {
  path=AnimGraph.AnimGraphNode_StateMachine_1.Locomotion.AnimStateNode_3.WalkStop
  nodes [
   AnimGraphNode_StateMachine_0 = WalkStop_SM() {is=title AnimGraphNode_StateMachine, bAllowConduitEntryStates=true}
   Output Animation Pose(Result=Apply Additive(
     Base=@AnimGraphNode_StateMachine_0
     Additive="Sequence Player (additive) 'ToyFox_Idle_Pose_Add_NoWepRelative'"() {bLoopAnimation=false, playRate=0, sequence=&ToyFox_Idle_Pose_Add_NoWepRelative}
    ) {alphaBoolBlend {blendOption=Linear}}) {
    node {stateEntryFunction {functionName=OnStateEntry_0}}
    stateEntryFunction {bSelfContext=true, memberGuid=#22, memberName=WalkStop_OnEntry}
   }
  ]
 }
 WalkStop_SM {
  entry=Conduit
  path=AnimGraph.AnimGraphNode_StateMachine_1.Locomotion.AnimStateNode_3.WalkStop.AnimGraphNode_StateMachine_0.WalkStop_SM
  ids=id
  nodes [
   AnimStateConduitNode_0 = Conduit()
   AnimStateNode_0 = WalkStop_LFoot() {is=title AnimStateNode}
   AnimStateNode_1 = WalkStop_RFoot() {is=title AnimStateNode}
   AnimStateNode_2 = StopInPlace() {is=title AnimStateNode, bAlwaysResetOnEntry=true}
   AnimStateTransitionNode_0 = AnimStateTransitionNode() {blendMode=CubicInOut, crossfadeDuration=0.6}
   AnimStateTransitionNode_4 = AnimStateTransitionNode() {priorityOrder=10}
   AnimStateTransitionNode_5 = AnimStateTransitionNode() {priorityOrder=100}
  ]
  links [
   {
    dst=StopInPlace
    rule=AnimGraph.AnimGraphNode_StateMachine_1.Locomotion.AnimStateNode_3.WalkStop.AnimGraphNode_StateMachine_0.WalkStop_SM.AnimStateTransitionNode_0.Transition
    src=Conduit
    via=AnimStateTransitionNode_0
   }
   {
    dst=WalkStop_LFoot
    rule=AnimGraph.AnimGraphNode_StateMachine_1.Locomotion.AnimStateNode_3.WalkStop.AnimGraphNode_StateMachine_0.WalkStop_SM.AnimStateTransitionNode_4.Transition
    src=Conduit
    via=AnimStateTransitionNode_4
   }
   {
    dst=WalkStop_RFoot
    rule=AnimGraph.AnimGraphNode_StateMachine_1.Locomotion.AnimStateNode_3.WalkStop.AnimGraphNode_StateMachine_0.WalkStop_SM.AnimStateTransitionNode_5.Transition
    src=Conduit
    via=AnimStateTransitionNode_5
   }
  ]
 }
 Conduit {
  path=AnimGraph.AnimGraphNode_StateMachine_1.Locomotion.AnimStateNode_3.WalkStop.AnimGraphNode_StateMachine_0.WalkStop_SM.AnimStateConduitNode_0.Conduit
  nodes [Result(bCanEnterTransition=true)]
 }
 WalkStop_LFoot {
  path=AnimGraph.AnimGraphNode_StateMachine_1.Locomotion.AnimStateNode_3.WalkStop.AnimGraphNode_StateMachine_0.WalkStop_SM.AnimStateNode_0.WalkStop_LFoot
  nodes [
   AnimGraphNode_BlendSpaceGraph_1 = Blendspace 'WalkStop LFoot BlendSpace'!(X:float=0, Y:float=0) -> (Pose:Pose Link Structure) {is=title AnimGraphNode_BlendSpaceGraph}
   Output Animation Pose(Result=@AnimGraphNode_BlendSpaceGraph_1)
  ]
 }
 WalkStop LFoot BlendSpace {
  path=AnimGraph.AnimGraphNode_StateMachine_1.Locomotion.AnimStateNode_3.WalkStop.AnimGraphNode_StateMachine_0.WalkStop_SM.AnimStateNode_0.WalkStop_LFoot.AnimGraphNode_BlendSpaceGraph_1.WalkStop LFoot BlendSpace
  links []
  nodes []
 }
 NewSample {
  path=AnimGraph.AnimGraphNode_StateMachine_1.Locomotion.AnimStateNode_3.WalkStop.AnimGraphNode_StateMachine_0.WalkStop_SM.AnimStateNode_0.WalkStop_LFoot.AnimGraphNode_BlendSpaceGraph_1.WalkStop LFoot BlendSpace.NewSample
  nodes [
   Blendspace Player 'ToyFox_WalkStop_N_LFoot_BS'!(X=0, Y=0, PlayRate:float=1) -> (Pose) {bLoop=false, blendSpace=/ToyFoxGameplay/Assets/Animations/ToyFox_WalkStop_N_LFoot_BS}
   Output Animation Pose!(Result:Pose Link Structure=Sequence Player 'NoWep_Walk_Stop_N_Lfoot'!(PlayRate:float=1) -> (Pose) {bLoopAnimation=false, sequence=/ToyFoxGameplay/Assets/Animations/Locomotion/NoWep_Walk_Stop_N_Lfoot}) {is=title AnimGraphNode_BlendSpaceSampleResult}
  ]
 }
 NewSample1 {
  path=AnimGraph.AnimGraphNode_StateMachine_1.Locomotion.AnimStateNode_3.WalkStop.AnimGraphNode_StateMachine_0.WalkStop_SM.AnimStateNode_0.WalkStop_LFoot.AnimGraphNode_BlendSpaceGraph_1.WalkStop LFoot BlendSpace.NewSample1
  nodes [Output Animation Pose(Result=Sequence Player 'NoWep_Walk_Stop_S_Lfoot'!(PlayRate:float=1) -> (Pose) {bLoopAnimation=false, sequence=/ToyFoxGameplay/Assets/Animations/Locomotion/NoWep_Walk_Stop_S_Lfoot})]
 }
 NewSample2 {
  path=AnimGraph.AnimGraphNode_StateMachine_1.Locomotion.AnimStateNode_3.WalkStop.AnimGraphNode_StateMachine_0.WalkStop_SM.AnimStateNode_0.WalkStop_LFoot.AnimGraphNode_BlendSpaceGraph_1.WalkStop LFoot BlendSpace.NewSample2
  nodes [Output Animation Pose(Result=Sequence Player 'NoWep_Walk_Stop_S_Lfoot'(PlayRate=1) {bLoopAnimation=false, sequence=&NoWep_Walk_Stop_S_Lfoot})]
 }
 NewSample3 {
  path=AnimGraph.AnimGraphNode_StateMachine_1.Locomotion.AnimStateNode_3.WalkStop.AnimGraphNode_StateMachine_0.WalkStop_SM.AnimStateNode_0.WalkStop_LFoot.AnimGraphNode_BlendSpaceGraph_1.WalkStop LFoot BlendSpace.NewSample3
  nodes [
   n1 = Sequence Player 'NoWep_Walk_Stop_E_LFF_Rfoot'!(PlayRate:float=1) -> (Pose) {
    becomeRelevantFunction {bSelfContext=true, memberGuid=#23, memberName=WalkStop_LFoot_East_OnBecomeRelevant}
    node {
     bLoopAnimation=false
     sequence=/ToyFoxGameplay/Assets/Animations/Locomotion/NoWep_Walk_Stop_E_LFF_Rfoot
     becomeRelevantFunction {functionName=OnBecomeRelevant}
    }
   }
   Output Animation Pose(Result=@n1)
  ]
 }
 NewSample4 {
  path=AnimGraph.AnimGraphNode_StateMachine_1.Locomotion.AnimStateNode_3.WalkStop.AnimGraphNode_StateMachine_0.WalkStop_SM.AnimStateNode_0.WalkStop_LFoot.AnimGraphNode_BlendSpaceGraph_1.WalkStop LFoot BlendSpace.NewSample4
  nodes [
   n1 = Sequence Player!(PlayRate:float=1) -> (Pose) {
    becomeRelevantFunction {bSelfContext=true, memberGuid=#24, memberName=WalkStop_LFoot_West_OnBecomeRelevant}
    node {bLoopAnimation=false, becomeRelevantFunction {functionName=SelectWestWalkStop_OnBecomeRelevant}}
   }
   Output Animation Pose(Result=@n1)
  ]
 }
 NewSample5 {
  path=AnimGraph.AnimGraphNode_StateMachine_1.Locomotion.AnimStateNode_3.WalkStop.AnimGraphNode_StateMachine_0.WalkStop_SM.AnimStateNode_0.WalkStop_LFoot.AnimGraphNode_BlendSpaceGraph_1.WalkStop LFoot BlendSpace.NewSample5
  nodes [Output Animation Pose(Result=Sequence Player 'ToyFox_Walk_N_Stop_InPlace_LFoot_Trimmed'!(PlayRate:float=1) -> (Pose) {bLoopAnimation=false, sequence=/ToyFoxGameplay/Assets/Animations/ToyFox_Walk_N_Stop_InPlace_LFoot_Trimmed})]
 }
 NewSample6 {
  path=AnimGraph.AnimGraphNode_StateMachine_1.Locomotion.AnimStateNode_3.WalkStop.AnimGraphNode_StateMachine_0.WalkStop_SM.AnimStateNode_0.WalkStop_LFoot.AnimGraphNode_BlendSpaceGraph_1.WalkStop LFoot BlendSpace.NewSample6
  nodes [Output Animation Pose(Result=Sequence Player 'ToyFox_Walk_N_Stop_InPlace_LFoot_Trimmed'(PlayRate=1) {bLoopAnimation=false, method=Graph, sequence=&ToyFox_Walk_N_Stop_InPlace_LFoot_Trimmed})]
 }
 NewSample7 {
  path=AnimGraph.AnimGraphNode_StateMachine_1.Locomotion.AnimStateNode_3.WalkStop.AnimGraphNode_StateMachine_0.WalkStop_SM.AnimStateNode_0.WalkStop_LFoot.AnimGraphNode_BlendSpaceGraph_1.WalkStop LFoot BlendSpace.NewSample7
  nodes [Output Animation Pose(Result=Sequence Player 'ToyFox_Walk_N_Stop_InPlace_LFoot_Trimmed'(PlayRate=1) {method=Graph, sequence=&ToyFox_Walk_N_Stop_InPlace_LFoot_Trimmed})]
 }
 NewSample8 {
  path=AnimGraph.AnimGraphNode_StateMachine_1.Locomotion.AnimStateNode_3.WalkStop.AnimGraphNode_StateMachine_0.WalkStop_SM.AnimStateNode_0.WalkStop_LFoot.AnimGraphNode_BlendSpaceGraph_1.WalkStop LFoot BlendSpace.NewSample8
  nodes [Output Animation Pose(Result=Sequence Player 'ToyFox_Walk_N_Stop_InPlace_LFoot_Trimmed'(PlayRate=1) {bLoopAnimation=false, sequence=&ToyFox_Walk_N_Stop_InPlace_LFoot_Trimmed})]
 }
 NewSample9 {
  path=AnimGraph.AnimGraphNode_StateMachine_1.Locomotion.AnimStateNode_3.WalkStop.AnimGraphNode_StateMachine_0.WalkStop_SM.AnimStateNode_0.WalkStop_LFoot.AnimGraphNode_BlendSpaceGraph_1.WalkStop LFoot BlendSpace.NewSample9
  nodes [Output Animation Pose(Result=Sequence Player 'ToyFox_Walk_N_Stop_InPlace_LFoot_Trimmed'(PlayRate=1) {bLoopAnimation=false, sequence=&ToyFox_Walk_N_Stop_InPlace_LFoot_Trimmed})]
 }
 WalkStop_RFoot {
  path=AnimGraph.AnimGraphNode_StateMachine_1.Locomotion.AnimStateNode_3.WalkStop.AnimGraphNode_StateMachine_0.WalkStop_SM.AnimStateNode_1.WalkStop_RFoot
  nodes [
   AnimGraphNode_BlendSpaceGraph_1 = Blendspace 'WalkStop RFoot BlendSpace'!(X:float=0, Y:float=0) -> (Pose:Pose Link Structure) {is=title AnimGraphNode_BlendSpaceGraph}
   Output Animation Pose(Result=@AnimGraphNode_BlendSpaceGraph_1) {is=title AnimGraphNode_StateResult}
  ]
 }
 WalkStop RFoot BlendSpace {
  path=AnimGraph.AnimGraphNode_StateMachine_1.Locomotion.AnimStateNode_3.WalkStop.AnimGraphNode_StateMachine_0.WalkStop_SM.AnimStateNode_1.WalkStop_RFoot.AnimGraphNode_BlendSpaceGraph_1.WalkStop RFoot BlendSpace
  links []
  nodes []
 }
 NewSample {
  path=AnimGraph.AnimGraphNode_StateMachine_1.Locomotion.AnimStateNode_3.WalkStop.AnimGraphNode_StateMachine_0.WalkStop_SM.AnimStateNode_1.WalkStop_RFoot.AnimGraphNode_BlendSpaceGraph_1.WalkStop RFoot BlendSpace.NewSample
  nodes [Output Animation Pose(Result=Sequence Player 'NoWep_Walk_Stop_N_Rfoot'!(PlayRate:float=1) -> (Pose) {bLoopAnimation=false, sequence=/ToyFoxGameplay/Assets/Animations/Locomotion/NoWep_Walk_Stop_N_Rfoot}) {is=title AnimGraphNode_BlendSpaceSampleResult}]
 }
 NewSample1 {
  path=AnimGraph.AnimGraphNode_StateMachine_1.Locomotion.AnimStateNode_3.WalkStop.AnimGraphNode_StateMachine_0.WalkStop_SM.AnimStateNode_1.WalkStop_RFoot.AnimGraphNode_BlendSpaceGraph_1.WalkStop RFoot BlendSpace.NewSample1
  nodes [Output Animation Pose(Result=Sequence Player 'NoWep_Walk_Stop_S_Rfoot'!(PlayRate:float=1) -> (Pose) {bLoopAnimation=false, sequence=/ToyFoxGameplay/Assets/Animations/Locomotion/NoWep_Walk_Stop_S_Rfoot})]
 }
 NewSample2 {
  path=AnimGraph.AnimGraphNode_StateMachine_1.Locomotion.AnimStateNode_3.WalkStop.AnimGraphNode_StateMachine_0.WalkStop_SM.AnimStateNode_1.WalkStop_RFoot.AnimGraphNode_BlendSpaceGraph_1.WalkStop RFoot BlendSpace.NewSample2
  nodes [Output Animation Pose(Result=Sequence Player 'NoWep_Walk_Stop_S_Rfoot'(PlayRate=1) {bLoopAnimation=false, sequence=&NoWep_Walk_Stop_S_Rfoot})]
 }
 NewSample3 {
  path=AnimGraph.AnimGraphNode_StateMachine_1.Locomotion.AnimStateNode_3.WalkStop.AnimGraphNode_StateMachine_0.WalkStop_SM.AnimStateNode_1.WalkStop_RFoot.AnimGraphNode_BlendSpaceGraph_1.WalkStop RFoot BlendSpace.NewSample3
  nodes [
   n1 = Sequence Player 'NoWep_Walk_Stop_E_LFF_Rfoot'(PlayRate=1) {
    becomeRelevantFunction {bSelfContext=true, memberGuid=#25, memberName=WalkStop_RFoot_East_OnBecomeRelevant}
    node {
     bLoopAnimation=false
     sequence=&NoWep_Walk_Stop_E_LFF_Rfoot
     becomeRelevantFunction {functionName=OnBecomeRelevant}
    }
   }
   Output Animation Pose(Result=@n1)
  ]
 }
 NewSample4 {
  path=AnimGraph.AnimGraphNode_StateMachine_1.Locomotion.AnimStateNode_3.WalkStop.AnimGraphNode_StateMachine_0.WalkStop_SM.AnimStateNode_1.WalkStop_RFoot.AnimGraphNode_BlendSpaceGraph_1.WalkStop RFoot BlendSpace.NewSample4
  nodes [
   n1 = Sequence Player(PlayRate=1) {
    becomeRelevantFunction {bSelfContext=true, memberGuid=#26, memberName=WalkStop_RFoot_West_OnBecomeRelevant}
    node {bLoopAnimation=false, becomeRelevantFunction {functionName=OnBecomeRelevant}}
   }
   Output Animation Pose(Result=@n1)
  ]
 }
 NewSample5 {
  path=AnimGraph.AnimGraphNode_StateMachine_1.Locomotion.AnimStateNode_3.WalkStop.AnimGraphNode_StateMachine_0.WalkStop_SM.AnimStateNode_1.WalkStop_RFoot.AnimGraphNode_BlendSpaceGraph_1.WalkStop RFoot BlendSpace.NewSample5
  nodes [Output Animation Pose(Result=Sequence Player 'ToyFox_Walk_N_Stop_InPlace_RFoot_Trimmed'!(PlayRate:float=1) -> (Pose) {bLoopAnimation=false, sequence=/ToyFoxGameplay/Assets/Animations/ToyFox_Walk_N_Stop_InPlace_RFoot_Trimmed})]
 }
 NewSample6 {
  path=AnimGraph.AnimGraphNode_StateMachine_1.Locomotion.AnimStateNode_3.WalkStop.AnimGraphNode_StateMachine_0.WalkStop_SM.AnimStateNode_1.WalkStop_RFoot.AnimGraphNode_BlendSpaceGraph_1.WalkStop RFoot BlendSpace.NewSample6
  nodes [Output Animation Pose(Result=Sequence Player 'ToyFox_Walk_N_Stop_InPlace_RFoot_Trimmed'(PlayRate=1) {bLoopAnimation=false, sequence=&ToyFox_Walk_N_Stop_InPlace_RFoot_Trimmed})]
 }
 NewSample7 {
  path=AnimGraph.AnimGraphNode_StateMachine_1.Locomotion.AnimStateNode_3.WalkStop.AnimGraphNode_StateMachine_0.WalkStop_SM.AnimStateNode_1.WalkStop_RFoot.AnimGraphNode_BlendSpaceGraph_1.WalkStop RFoot BlendSpace.NewSample7
  nodes [Output Animation Pose(Result=Sequence Player 'ToyFox_Walk_N_Stop_InPlace_RFoot_Trimmed'(PlayRate=1) {bLoopAnimation=false, sequence=&ToyFox_Walk_N_Stop_InPlace_RFoot_Trimmed})]
 }
 NewSample8 {
  path=AnimGraph.AnimGraphNode_StateMachine_1.Locomotion.AnimStateNode_3.WalkStop.AnimGraphNode_StateMachine_0.WalkStop_SM.AnimStateNode_1.WalkStop_RFoot.AnimGraphNode_BlendSpaceGraph_1.WalkStop RFoot BlendSpace.NewSample8
  nodes [Output Animation Pose(Result=Sequence Player 'ToyFox_Walk_N_Stop_InPlace_RFoot_Trimmed'(PlayRate=1) {bLoopAnimation=false, sequence=&ToyFox_Walk_N_Stop_InPlace_RFoot_Trimmed})]
 }
 NewSample9 {
  path=AnimGraph.AnimGraphNode_StateMachine_1.Locomotion.AnimStateNode_3.WalkStop.AnimGraphNode_StateMachine_0.WalkStop_SM.AnimStateNode_1.WalkStop_RFoot.AnimGraphNode_BlendSpaceGraph_1.WalkStop RFoot BlendSpace.NewSample9
  nodes [Output Animation Pose(Result=Sequence Player 'ToyFox_Walk_N_Stop_InPlace_RFoot_Trimmed'(PlayRate=1) {bLoopAnimation=false, sequence=&ToyFox_Walk_N_Stop_InPlace_RFoot_Trimmed})]
 }
 StopInPlace {
  path=AnimGraph.AnimGraphNode_StateMachine_1.Locomotion.AnimStateNode_3.WalkStop.AnimGraphNode_StateMachine_0.WalkStop_SM.AnimStateNode_2.StopInPlace
  nodes [
   n1 = Sequence Player(PlayRate=1) {
    becomeRelevantFunction {bSelfContext=true, memberGuid=#27, memberName=WalkStop_InPlace_OnBecomeRelevant}
    node {bLoopAnimation=false, becomeRelevantFunction {functionName=OnBecomeRelevant}}
   }
   Output Animation Pose(Result=@n1) {is=title AnimGraphNode_StateResult}
  ]
 }
 Transition {
  path=AnimGraph.AnimGraphNode_StateMachine_1.Locomotion.AnimStateNode_3.WalkStop.AnimGraphNode_StateMachine_0.WalkStop_SM.AnimStateTransitionNode_0.Transition
  nodes [Result(bCanEnterTransition=Get bWalkStopInPlace() -> (bWalkStopInPlace:bool))]
 }
 Transition {
  path=AnimGraph.AnimGraphNode_StateMachine_1.Locomotion.AnimStateNode_3.WalkStop.AnimGraphNode_StateMachine_0.WalkStop_SM.AnimStateTransitionNode_4.Transition
  nodes [Result(bCanEnterTransition=Get bIsMoveStop_LFoot() -> (bIsMoveStop_LFoot:bool))]
 }
 Transition {
  path=AnimGraph.AnimGraphNode_StateMachine_1.Locomotion.AnimStateNode_3.WalkStop.AnimGraphNode_StateMachine_0.WalkStop_SM.AnimStateTransitionNode_5.Transition
  nodes [Result(bCanEnterTransition=true)]
 }
 Pivot {
  path=AnimGraph.AnimGraphNode_StateMachine_1.Locomotion.AnimStateNode_4.Pivot
  nodes [
   AnimGraphNode_StateMachine_0 = Pivot_SM() {is=title AnimGraphNode_StateMachine}
   Output Animation Pose(Result=Apply Additive(
     Base=@AnimGraphNode_StateMachine_0
     Additive="Sequence Player (additive) 'ToyFox_Idle_Pose_Add_NoWepRelative'"() {bLoopAnimation=false, playRate=0, sequence=&ToyFox_Idle_Pose_Add_NoWepRelative}
    ) {alphaBoolBlend {blendOption=Linear}}) {
    node {stateExitFunction {functionName=OnStateExit}}
    stateExitFunction {bSelfContext=true, memberGuid=#28, memberName=Pivot_OnStateExit}
   }
  ]
 }
 Pivot_SM {
  entry=A
  path=AnimGraph.AnimGraphNode_StateMachine_1.Locomotion.AnimStateNode_4.Pivot.AnimGraphNode_StateMachine_0.Pivot_SM
  ids=id
  nodes [
   AnimStateNode_0 = A() {is=title AnimStateNode}
   AnimStateNode_1 = B() {is=title AnimStateNode}
   AnimStateTransitionNode_1 = AnimStateTransitionNode() {
    blendMode=Cubic
    crossfadeDuration=0.5
    logicType=TLT_Custom
    blendProfileWrapper {
     bIsSkeletonBlendProfile=false
     blendProfile=&self:AnimGraph.AnimGraphNode_StateMachine_1.Locomotion.AnimStateNode_4.Pivot.AnimGraphNode_StateMachine_0.Pivot_SM.AnimStateTransitionNode_1.BlendProfile_0
     blendProfileProvider=BlendProfileStandaloneProvider'&self:AnimGraph.AnimGraphNode_StateMachine_1.Locomotion.AnimStateNode_4.Pivot.AnimGraphNode_StateMachine_0.Pivot_SM.AnimStateTransitionNode_1.BlendProfileStandaloneProvider_0'
    }
   }
   AnimStateTransitionNode_2 = AnimStateTransitionNode() {crossfadeDuration=0.5, logicType=TLT_Custom}
  ]
  links [
   {
    dst=B
    rule=AnimGraph.AnimGraphNode_StateMachine_1.Locomotion.AnimStateNode_4.Pivot.AnimGraphNode_StateMachine_0.Pivot_SM.AnimStateTransitionNode_1.Transition
    src=A
    via=AnimStateTransitionNode_1
   }
   {
    dst=A
    rule=AnimGraph.AnimGraphNode_StateMachine_1.Locomotion.AnimStateNode_4.Pivot.AnimGraphNode_StateMachine_0.Pivot_SM.AnimStateTransitionNode_1.Transition
    src=B
    via=AnimStateTransitionNode_2
   }
  ]
 }
 A {
  path=AnimGraph.AnimGraphNode_StateMachine_1.Locomotion.AnimStateNode_4.Pivot.AnimGraphNode_StateMachine_0.Pivot_SM.AnimStateNode_0.A
  nodes [
   n1 = Sequence Player!(PlayRate=1, StartPosition:float=0) -> (Pose) {
    becomeRelevantFunction {bSelfContext=true, memberGuid=#29, memberName=SetUpPivotAnims}
    node {becomeRelevantFunction {functionName=OnBecomeRelevant}}
   }
   Output Animation Pose(Result=@n1) {
    becomeRelevantFunction {bSelfContext=true, memberGuid=#30, memberName=SetUpPivotState}
    node {becomeRelevantFunction {functionName=SetUpPivotState}}
   }
  ]
 }
 B {
  path=AnimGraph.AnimGraphNode_StateMachine_1.Locomotion.AnimStateNode_4.Pivot.AnimGraphNode_StateMachine_0.Pivot_SM.AnimStateNode_1.B
  nodes [
   n1 = Sequence Player(PlayRate=1, StartPosition=0) {
    becomeRelevantFunction {bSelfContext=true, memberGuid=#29, memberName=SetUpPivotAnims}
    node {becomeRelevantFunction {functionName=OnBecomeRelevant}}
   }
   Output Animation Pose(Result=@n1) {
    becomeRelevantFunction {bSelfContext=true, memberGuid=#30, memberName=SetUpPivotState}
    node {becomeRelevantFunction {functionName=SetUpPivotState}}
   }
  ]
 }
 CustomTransition {
  path=AnimGraph.AnimGraphNode_StateMachine_1.Locomotion.AnimStateNode_4.Pivot.AnimGraphNode_StateMachine_0.Pivot_SM.AnimStateTransitionNode_1.CustomTransition
  nodes [Custom Transition Blend Result!(Result:Pose Link Structure=Two Way Blend(
     A=Input Pose Evaluator!() -> (Pose:Pose Link Structure) {is=title AnimGraphNode_TransitionPoseEvaluator, evaluatorMode=EM_Freeze}
     B=Input Pose Evaluator() {dataSource=EDS_DestinationPose}
     Alpha=Get PivotAToBCustomBlend() -> (PivotAToBCustomBlend:float)
    ) {blendNode {alphaBoolBlend {blendOption=Linear}}}) {is=title AnimGraphNode_CustomTransitionResult}]
 }
 Transition {
  path=AnimGraph.AnimGraphNode_StateMachine_1.Locomotion.AnimStateNode_4.Pivot.AnimGraphNode_StateMachine_0.Pivot_SM.AnimStateTransitionNode_1.Transition
  nodes [Result(bCanEnterTransition=Get bWantsOppositePivot() -> (bWantsOppositePivot:bool))]
 }
 CustomTransition {
  path=AnimGraph.AnimGraphNode_StateMachine_1.Locomotion.AnimStateNode_4.Pivot.AnimGraphNode_StateMachine_0.Pivot_SM.AnimStateTransitionNode_2.CustomTransition
  nodes [Custom Transition Blend Result(Result=Two Way Blend(
     A=Input Pose Evaluator() {evaluatorMode=EM_Freeze}
     B=Input Pose Evaluator() {dataSource=EDS_DestinationPose}
     Alpha=Get PivotBToACustomBlend() -> (PivotBToACustomBlend:float)
    ) {blendNode {alphaBoolBlend {blendOption=Linear}}})]
 }
 Transition {
  path=AnimGraph.AnimGraphNode_StateMachine_1.Locomotion.AnimStateTransitionNode_0.Transition
  nodes [Result(bCanEnterTransition="float > float"(A=Get Speed() -> (Speed:float), B=Get MinSpeed2DThreshold() -> (MinSpeed2DThreshold:double)))]
 }
 Transition {
  path=AnimGraph.AnimGraphNode_StateMachine_1.Locomotion.AnimStateTransitionNode_1.Transition
  nodes [Result(bCanEnterTransition="float >= float"(A=Get MinSpeed2DThreshold(), B=Get Speed()))]
 }
 Transition {
  path=AnimGraph.AnimGraphNode_StateMachine_1.Locomotion.AnimStateTransitionNode_10.Transition
  nodes [Result(bCanEnterTransition=Get bWalkStopAndTurn() -> (bWalkStopAndTurn:bool))]
 }
 Transition {
  path=AnimGraph.AnimGraphNode_StateMachine_1.Locomotion.AnimStateTransitionNode_11.Transition
  nodes [Result(bCanEnterTransition=Get bDoPivot() -> (bDoPivot:bool))]
 }
 Transition {
  path=AnimGraph.AnimGraphNode_StateMachine_1.Locomotion.AnimStateTransitionNode_12.Transition
  nodes [Result(bCanEnterTransition=Get bIsMoveStop() -> (bIsMoveStop:bool))]
 }
 Transition {
  path=AnimGraph.AnimGraphNode_StateMachine_1.Locomotion.AnimStateTransitionNode_13.Transition
  nodes [Result(bCanEnterTransition=Get bIsPivotExit() -> (bIsPivotExit:bool))]
 }
 Transition {
  path=AnimGraph.AnimGraphNode_StateMachine_1.Locomotion.AnimStateTransitionNode_15.Transition
  nodes [Result(bCanEnterTransition=Get bIsInputing() -> (bIsInputing:bool))]
 }
 Transition {
  path=AnimGraph.AnimGraphNode_StateMachine_1.Locomotion.AnimStateTransitionNode_2.Transition
  nodes [Result(bCanEnterTransition=Get bIsMoveStart() -> (bIsMoveStart:bool))]
 }
 Transition {
  path=AnimGraph.AnimGraphNode_StateMachine_1.Locomotion.AnimStateTransitionNode_3.Transition
  nodes [Result(bCanEnterTransition=Get bIsMoveStop())]
 }
 Transition {
  path=AnimGraph.AnimGraphNode_StateMachine_1.Locomotion.AnimStateTransitionNode_4.Transition
  nodes [Result(bCanEnterTransition=Get bIsMoveStop())]
 }
 Transition {
  path=AnimGraph.AnimGraphNode_StateMachine_1.Locomotion.AnimStateTransitionNode_5.Transition
  nodes [Result(bCanEnterTransition=Get bIsMoveStart())]
 }
 Transition {
  path=AnimGraph.AnimGraphNode_StateMachine_1.Locomotion.AnimStateTransitionNode_6.Transition
  nodes [Result(bCanEnterTransition=Get bIsMoveStartExit() -> (bIsMoveStartExit:bool))]
 }
 Transition {
  path=AnimGraph.AnimGraphNode_StateMachine_1.Locomotion.AnimStateTransitionNode_7.Transition
  nodes [Result(bCanEnterTransition="float < float"(A=Get Speed(), B=10))]
 }
 Transition {path=AnimGraph.AnimGraphNode_StateMachine_1.Locomotion.AnimStateTransitionNode_8.Transition, nodes [Result()]}
 Transition {
  path=AnimGraph.AnimGraphNode_StateMachine_1.Locomotion.AnimStateTransitionNode_9.Transition
  nodes [Result(bCanEnterTransition=Get bWalkStopToIdle() -> (bWalkStopToIdle:bool))]
 }
 Cannon Ability {
  entry=Passthrough
  path=AnimGraph.AnimGraphNode_StateMachine_2.Cannon Ability
  ids=id
  nodes [
   AnimStateNode_0 = Idle()
   AnimStateNode_1 = Fire() {is=title AnimStateNode}
   AnimStateNode_2 = Passthrough() {is=title AnimStateNode}
   AnimStateTransitionNode_0 = AnimStateTransitionNode() {blendMode=CubicInOut, crossfadeDuration=1}
   AnimStateTransitionNode_1 = AnimStateTransitionNode() {blendMode=CubicInOut, crossfadeDuration=0.7}
   AnimStateTransitionNode_2 = AnimStateTransitionNode() {blendMode=QuadraticInOut, crossfadeDuration=0.4}
   AnimStateTransitionNode_3 = AnimStateTransitionNode()
   AnimStateTransitionNode_4 = AnimStateTransitionNode() {bDisabled=true, blendMode=CubicInOut, crossfadeDuration=0.5, priorityOrder=100}
  ]
  links [
   {
    dst=Idle
    rule=AnimGraph.AnimGraphNode_StateMachine_2.Cannon Ability.AnimStateTransitionNode_4.Transition
    src=Fire
    via=AnimStateTransitionNode_4
   }
   {
    dst=Passthrough
    rule=AnimGraph.AnimGraphNode_StateMachine_2.Cannon Ability.AnimStateTransitionNode_1.Transition
    src=Fire
    via=AnimStateTransitionNode_1
   }
   {
    dst=Fire
    rule=AnimGraph.AnimGraphNode_StateMachine_2.Cannon Ability.AnimStateTransitionNode_3.Transition
    src=Idle
    via=AnimStateTransitionNode_3
   }
   {
    dst=Passthrough
    rule=AnimGraph.AnimGraphNode_StateMachine_2.Cannon Ability.AnimStateTransitionNode_0.Transition
    src=Idle
    via=AnimStateTransitionNode_0
   }
   {
    dst=Idle
    rule=AnimGraph.AnimGraphNode_StateMachine_2.Cannon Ability.AnimStateTransitionNode_2.Transition
    src=Passthrough
    via=AnimStateTransitionNode_2
   }
  ]
 }
 Idle {
  path=AnimGraph.AnimGraphNode_StateMachine_2.Cannon Ability.AnimStateNode_0.Idle
  nodes [
   Sequence Player 'ToyFox_LRG_Cannon_Idle'() {bLoopAnimation=false, sequence=/ToyFoxGameplay/Assets/Animations/ToyFox_LRG_Cannon_Idle}
   Output Animation Pose(Result=Apply Additive(
     Base=Use cached pose 'LocoPlusIdleAdditive'()
     Additive="Sequence Player (additive) 'ToyFox_LRG_Cannon_Idle_Additive'"() {bLoopAnimation=false, sequence=/ToyFoxGameplay/Assets/Animations/ToyFox_LRG_Cannon_Idle_Additive}
    ) {alphaBoolBlend {blendOption=Linear}})
  ]
 }
 Fire {
  path=AnimGraph.AnimGraphNode_StateMachine_2.Cannon Ability.AnimStateNode_1.Fire
  nodes [Output Animation Pose(Result=Apply Additive(
     Base=Use cached pose 'LocoPlusIdleAdditive'()
     Additive="Sequence Player (additive) 'ToyFox_LRG_Fire_Additive'"() {bLoopAnimation=false, sequence=/ToyFoxGameplay/Assets/Animations/ToyFox_LRG_Fire_Additive}
    ) {alphaBoolBlend {blendOption=Linear}})]
 }
 Passthrough {
  path=AnimGraph.AnimGraphNode_StateMachine_2.Cannon Ability.AnimStateNode_2.Passthrough
  nodes [Output Animation Pose(Result=Use cached pose 'LocoPlusIdleAdditive'())]
 }
 Transition {
  path=AnimGraph.AnimGraphNode_StateMachine_2.Cannon Ability.AnimStateTransitionNode_0.Transition
  nodes [Result(bCanEnterTransition=NOT Boolean(A=Get bAbilityCannonActive() -> (bAbilityCannonActive:bool)))]
 }
 Transition {
  path=AnimGraph.AnimGraphNode_StateMachine_2.Cannon Ability.AnimStateTransitionNode_1.Transition
  nodes [Result(bCanEnterTransition=Get bCannonCanTransitionFireToPassThrough() -> (bCannonCanTransitionFireToPassThrough:bool))]
 }
 Transition {
  path=AnimGraph.AnimGraphNode_StateMachine_2.Cannon Ability.AnimStateTransitionNode_2.Transition
  nodes [Result(bCanEnterTransition=Get bAbilityCannonActive())]
 }
 Transition {
  path=AnimGraph.AnimGraphNode_StateMachine_2.Cannon Ability.AnimStateTransitionNode_3.Transition
  nodes [Result(bCanEnterTransition=Get bAbilityCannonFire() -> (bAbilityCannonFire:bool))]
 }
 Transition {
  path=AnimGraph.AnimGraphNode_StateMachine_2.Cannon Ability.AnimStateTransitionNode_4.Transition
  nodes [
   n1 = "Get Relevant Anim Time Remaining (Fire)"!() -> (ReturnValue:float) {is=title K2Node_AnimGetter}
   Result(bCanEnterTransition=AND Boolean(A=Get bAbilityCannonActive(), B="float < float"(A=@n1, B=0.2)))
  ]
 }
 EventGraph {nodes [
   n1 = UpdateFootPhase!() -> (OutputDelegate, then, FootPhase:E_LocoFootPhase Enum=NewEnumerator0, ToyFoxFootPhase:EToyFoxGiantPawnLocoFootPhase Enum=LPass) {is=title K2Node_CustomEvent}
   > Set LocoFootPhase(LocoFootPhase:E_LocoFootPhase Enum=@n1.FootPhase) -> (Output_Get:E_LocoFootPhase Enum=NewEnumerator0)
   > Set ToyFoxLocoFootPhase(ToyFoxLocoFootPhase:EToyFoxGiantPawnLocoFootPhase Enum=@n1.ToyFoxFootPhase) -> (Output_Get:EToyFoxGiantPawnLocoFootPhase Enum=LPass)
   Event BlueprintInitializeAnimation()
   > Set SpeedMax(SpeedMax:double=GetCachedScalableFloatValue!(Input:"Fort Cached Scalable Float Structure (by ref)"=Get Hotfix_MaxSpeed_Walk() -> (Hotfix_MaxSpeed_Walk:Fort Cached Scalable Float Structure)) -> (ReturnValue:float)) -> (Output_Get:double)
  ]}
 OnStateEntry {nodes [
   OnStateEntry!() -> (then, Context:"Anim Update Context Structure (by ref)", Node:"Anim Node Reference Structure (by ref)") {is=title K2Node_FunctionEntry}
   > Set bIsMoveStart(bIsMoveStart:bool=false) -> (Output_Get:bool)
   > Return Node()
  ]}
 OnTurnInPlaceExit {nodes [
   OnTurnInPlaceExit!() -> (then, Context:"Anim Update Context Structure (by ref)", Node:"Anim Node Reference Structure (by ref)") {is=title K2Node_FunctionEntry}
   > Set bIsTurnInPlaceRight(bIsTurnInPlaceRight:bool=false) -> (Output_Get:bool)
   > Set bIsTurnInPlaceLeft(bIsTurnInPlaceLeft:bool=false) -> (Output_Get:bool)
   > Set bIsTurningInPlace(bIsTurningInPlace:bool=false) -> (Output_Get:bool)
   > Return Node()
  ]}
 Pivot_OnStateExit {nodes [
   Pivot_OnStateExit!() -> (then, Context:"Anim Update Context Structure (by ref)", Node:"Anim Node Reference Structure (by ref)") {is=title K2Node_FunctionEntry}
   > Set bWasInPivot(bWasInPivot:bool=true) -> (Output_Get:bool)
   > Return Node()
  ]}
 SetUpPivotAnims {nodes [
   n1 = SetUpPivotAnims!() -> (then, Context:"Anim Update Context Structure (by ref)", Node:"Anim Node Reference Structure (by ref)") {is=title K2Node_FunctionEntry}
   n2 = Get bIsRFFHips() -> (bIsRFFHips:bool)
   n3 = Select(
    Option 0:Anim Sequence Base Object Reference=/ToyFoxGameplay/Assets/Animations/Locomotion/NoWep_Walk_Pivot_W_E_LFF_Lfoot
    Option 1:Anim Sequence Base Object Reference=/ToyFoxGameplay/Assets/Animations/Locomotion/NoWep_Walk_Pivot_W_E_RFF_Lfoot
    Index=@n2
   ) -> (ReturnValue:Anim Sequence Base Object Reference)
   n4 = Select(
    Option 0=/ToyFoxGameplay/Assets/Animations/Locomotion/NoWep_Walk_Pivot_E_W_LFF_Lfoot
    Option 1=/ToyFoxGameplay/Assets/Animations/Locomotion/NoWep_Walk_Pivot_E_W_RFF_Lfoot
    Index=@n2
   )
   n5 = Get Pivot_CardinalDirection_Snap() -> (Pivot_CardinalDirection_Snap:ELocoCardinalDirection Enum=North)
   n6 = Select(
    Option 0=/ToyFoxGameplay/Assets/Animations/Locomotion/NoWep_Walk_Pivot_W_E_LFF_Rfoot
    Option 1=/ToyFoxGameplay/Assets/Animations/Locomotion/NoWep_Walk_Pivot_W_E_RFF_Rfoot
    Index=@n2
   )
   n7 = Select(
    Option 0=/ToyFoxGameplay/Assets/Animations/Locomotion/NoWep_Walk_Pivot_E_W_LFF_Rfoot
    Option 1=/ToyFoxGameplay/Assets/Animations/Locomotion/NoWep_Walk_Pivot_E_W_RFF_Rfoot
    Index=@n2
   )
   n8 = Get Pivot_CardinalDirection_Snap() -> (Pivot_CardinalDirection_Snap=North)
   > SetSequenceWithInertialBlending!(
    execute
    UpdateContext:"Anim Update Context Structure (by ref)"=@n1
    SequencePlayer:"Sequence Player Reference Structure (by ref)"=Convert to Sequence Player!(Node:"Anim Node Reference Structure (by ref)"=@n1.Node) -> (SequencePlayer:Sequence Player Reference Structure, Result:bool)
    Sequence:Anim Sequence Base Object Reference=Select(
     Option 0=Select!(
      North:Anim Sequence Base Object Reference=/ToyFoxGameplay/Assets/Animations/Locomotion/NoWep_Walk_Pivot_S_N_Rfoot
      South:Anim Sequence Base Object Reference=/ToyFoxGameplay/Assets/Animations/Locomotion/NoWep_Walk_Pivot_N_S_Rfoot
      East:Anim Sequence Base Object Reference=@n6
      West:Anim Sequence Base Object Reference=@n7
      Index:ELocoCardinalDirection Enum=@n8
     ) -> (ReturnValue)
     Option 1=Select(
      North=/ToyFoxGameplay/Assets/Animations/Locomotion/NoWep_Walk_Pivot_S_N_Lfoot
      South=/ToyFoxGameplay/Assets/Animations/Locomotion/NoWep_Walk_Pivot_N_S_Lfoot
      East=@n3
      West=@n4
      Index=@n5
     )
     Index=Get bIsMoveStop_LFoot()
    )
    BlendTime:float=0.2
   ) -> (then, ReturnValue:Sequence Player Reference Structure)
   > Return Node()
   Select!(
    NewEnumerator0:Anim Sequence Base Object Reference=&NoWep_Walk_Pivot_S_N_Rfoot
    NewEnumerator1:Anim Sequence Base Object Reference=&NoWep_Walk_Pivot_N_S_Rfoot
    NewEnumerator2:Anim Sequence Base Object Reference=@n6
    NewEnumerator3:Anim Sequence Base Object Reference=@n7
    Index:E_LocoCardinalDirections Enum=NewEnumerator0
   ) -> (ReturnValue)
   Select(
    NewEnumerator0=&NoWep_Walk_Pivot_S_N_Lfoot
    NewEnumerator1=&NoWep_Walk_Pivot_N_S_Lfoot
    NewEnumerator2=@n3
    NewEnumerator3=@n4
    Index=NewEnumerator0
   )
  ]}
 SetUpPivotState {nodes [
   SetUpPivotState!() -> (then, Context:"Anim Update Context Structure (by ref)", Node:"Anim Node Reference Structure (by ref)") {is=title K2Node_FunctionEntry}
   > Set PivotInput2D(PivotInput2D:Vector=Get Input2D() -> (Input2D:Vector)) -> (Output_Get:Vector)
   > Set bWantsOppositePivot(bWantsOppositePivot:bool=false) -> (Output_Get:bool)
   > Return Node()
  ]}
 WalkStart_OnEntry {nodes [
   WalkStart_OnEntry!() -> (then, Context:"Anim Update Context Structure (by ref)", Node:"Anim Node Reference Structure (by ref)") {is=title K2Node_FunctionEntry}
   > Set MovementDirectionDamped(MovementDirectionDamped:float=Get MovementDirectionInput() -> (MovementDirectionInput:float)) -> (Output_Get:float)
   > Return Node()
  ]}
 WalkStop_InPlace_OnBecomeRelevant {nodes [
   n1 = WalkStop_InPlace_OnBecomeRelevant!() -> (then, Context:"Anim Update Context Structure (by ref)", Node:"Anim Node Reference Structure (by ref)") {is=title K2Node_FunctionEntry}
   > SetSequenceWithInertialBlending(
    UpdateContext=@n1
    SequencePlayer=Convert to Sequence Player(Node=@n1.Node)
    Sequence=Select!(
     Option 0:Anim Sequence Base Object Reference=/ToyFoxGameplay/Assets/Animations/ToyFox_Walk_N_Stop_InPlace_LFoot_
     Option 1:Anim Sequence Base Object Reference=/ToyFoxGameplay/Assets/Animations/ToyFox_Walk_N_Stop_InPlace_RFoot_
     Index:bool=Get bIsMoveStartLFoot()
    ) -> (ReturnValue)
    BlendTime=0.2
   )
   > Set bWalkStopInPlace(bWalkStopInPlace:bool=false) -> (Output_Get:bool)
   > Return Node()
  ]}
 WalkStop_LFoot_East_OnBecomeRelevant {nodes [
   n1 = WalkStop_LFoot_East_OnBecomeRelevant!() -> (then, Context:"Anim Update Context Structure (by ref)", Node:"Anim Node Reference Structure (by ref)") {is=title K2Node_FunctionEntry}
   > SetSequenceWithInertialBlending(
    UpdateContext=@n1
    SequencePlayer=Convert to Sequence Player(Node=@n1.Node)
    Sequence=Select(
     Option 0=/ToyFoxGameplay/Assets/Animations/Locomotion/NoWep_Walk_Stop_E_LFF_Lfoot
     Option 1=/ToyFoxGameplay/Assets/Animations/Locomotion/NoWep_Walk_Stop_E_RFF_Lfoot
     Index=Get bIsRFFHips()
    )
    BlendTime=0.2
   )
   > Return Node()
  ]}
 WalkStop_LFoot_West_OnBecomeRelevant {nodes [
   n1 = WalkStop_LFoot_West_OnBecomeRelevant!() -> (then, Context:"Anim Update Context Structure (by ref)", Node:"Anim Node Reference Structure (by ref)") {is=title K2Node_FunctionEntry}
   > SetSequenceWithInertialBlending(
    UpdateContext=@n1
    SequencePlayer=Convert to Sequence Player(Node=@n1.Node)
    Sequence=Select(
     Option 0=/ToyFoxGameplay/Assets/Animations/Locomotion/NoWep_Walk_Stop_W_LFF_Lfoot
     Option 1=/ToyFoxGameplay/Assets/Animations/Locomotion/NoWep_Walk_Stop_W_RFF_Lfoot
     Index=Get bIsRFFHips()
    )
    BlendTime=0.2
   )
   > Return Node()
  ]}
 WalkStop_OnEntry {nodes [
   WalkStop_OnEntry!() -> (then, Context:"Anim Update Context Structure (by ref)", Node:"Anim Node Reference Structure (by ref)") {is=title K2Node_FunctionEntry}
   > n1 = Set MovementDirection_Stop(MovementDirection_Stop:float=Get MovementDirection() -> (MovementDirection:float)) -> (Output_Get:float)
   Set WalkStopPhase(execute=[@n2.then, @n1.then], WalkStopPhase:double=Get SpeedPercentage() -> (SpeedPercentage:double)) -> (Output_Get:double)
   > Return Node()
   n2 = Set VelocityCardinalDirection_Stop(VelocityCardinalDirection_Stop:E_LocoCardinalDirections Enum=Get VelocityCardinalDirection() -> (VelocityCardinalDirection:E_LocoCardinalDirections Enum=NewEnumerator0)) -> (Output_Get:E_LocoCardinalDirections Enum=NewEnumerator0)
  ]}
 WalkStop_RFoot_East_OnBecomeRelevant {nodes [
   n1 = WalkStop_RFoot_East_OnBecomeRelevant!() -> (then, Context:"Anim Update Context Structure (by ref)", Node:"Anim Node Reference Structure (by ref)") {is=title K2Node_FunctionEntry}
   > SetSequenceWithInertialBlending(
    UpdateContext=@n1
    SequencePlayer=Convert to Sequence Player(Node=@n1.Node)
    Sequence=Select(
     Option 0=&NoWep_Walk_Stop_E_LFF_Rfoot
     Option 1=/ToyFoxGameplay/Assets/Animations/Locomotion/NoWep_Walk_Stop_E_RFF_Rfoot
     Index=Get bIsRFFHips()
    )
    BlendTime=0.2
   )
   > Return Node()
  ]}
 WalkStop_RFoot_West_OnBecomeRelevant {nodes [
   n1 = WalkStop_RFoot_West_OnBecomeRelevant!() -> (then, Context:"Anim Update Context Structure (by ref)", Node:"Anim Node Reference Structure (by ref)") {is=title K2Node_FunctionEntry}
   > SetSequenceWithInertialBlending(
    UpdateContext=@n1
    SequencePlayer=Convert to Sequence Player(Node=@n1.Node)
    Sequence=Select(
     Option 0=/ToyFoxGameplay/Assets/Animations/Locomotion/NoWep_Walk_Stop_W_LFF_Rfoot
     Option 1=/ToyFoxGameplay/Assets/Animations/Locomotion/NoWep_Walk_Stop_W_RFF_Rfoot
     Index=Get bIsRFFHips()
    )
    BlendTime=0.2
   )
   > Return Node()
  ]}
]
variables [
 Hotfix_MaxSpeed_Walk {
  type=FortCachedScalableFloat
  default {scalableFloat {registryType=None, value=1, curve {curveTable=/ToyFoxGameplay/DataTables/ToyFoxGameplayGameData, rowName=ToyFox.MME.MaxWalkSpeed.Forward}}}
 }
 InputCardinalDirection {type=E_LocoCardinalDirections, default=N}
 LocoFootPhase {type=E_LocoFootPhase, default=LPass}
 PrintLocoDebug {type=boolean}
 VelocityCardinalDirection {type=E_LocoCardinalDirections, default=N}
 VelocityCardinalDirection_Stop {type=E_LocoCardinalDirections, default=N}
]
