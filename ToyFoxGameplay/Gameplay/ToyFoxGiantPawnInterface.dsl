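; Blueprint interface for giant pawns in the ToyFox gameplay code, declaring functions other systems call on them.
; Queries return two cannon muzzle locations, a head socket transform, latest landscape hit results, the budgeted skeletal mesh and the weapon mesh component.
; Commands destroy weak points and restore materials for an outro.
; It holds only signatures with default returns; implementing pawns supply the behavior.
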(fn DestroyWeakPoints)

(fn GetCannonMuzzleLocations
  (return "0, 0, 0" "0, 0, 0"))

(fn GetHeadSocketTrans
  (return 0))

(fn GetLatestLandscapeHitResults
  (return 0))

(fn GetSkeletalMeshBudgeted
  (return 0))

(fn GetWeaponMeshComponent
  (return 0))

(fn OutroMaterialRestore)
